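; Editor function library for FX Browser utilities, covering file and folder listing, login-name lookup, text file writing, BugItGo teleporting and dependency filtering.
; File and OS helpers run inline Python, parsing printed output into arrays where needed.
; File search is recursive but joins bare filenames to the root path; writing supports overwrite, unique append or plain append.
; IsADependencyBanned returns false for ban-list name matches or NiagaraEffectType assets, true otherwise.
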(fn BugItGo
  (params (Location :type Vector) (Rotation :type Rotator) (WorldContextObject :type Object%20Reference))
  (Development|ExecuteConsoleCommand WorldContextObject (Utilities|String|Append "BugItGo " (Utilities|String|ToString(Float) (.x Location)) " " (Utilities|String|ToString(Float) (.y Location)) " " (Utilities|String|ToString(Float) (.z Location)) " " (Utilities|String|ToString(Float) (.roll Rotation)) " " (Utilities|String|ToString(Float) (.pitch Rotation)) " " (Utilities|String|ToString(Float) (.yaw Rotation)))))

(fn GetFileListInPath
  (params (Path(/end) :type String) (ext :type String))
  (bind _tmplist (LocalVariables|GettmpList))
  (bind (_commandresult _logoutput) (Python|Execution|ExecutePythonCommand(Advanced) (Utilities|String|Append "from pathlib import Path


def list_files_with_extension(directory, extension):
    list = []
    for file_path in Path(directory).rglob(f\"*.{extension}\"):
        list.append(file_path.name)
    return list

list = list_files_with_extension(\"" (Utilities|String|Replace (Utilities|String|Replace (Utilities|String|Replace (Utilities|String|Replace (Utilities|String|Replace Path(/end) "\\" "\\\\") "\\\\\\" "\\\\") "\\\\\\\\" "\\\\") "\\\\\\\\\\" "\\\\") "\\\\\\\\\\\\" "\\\\") "\", \"" ext "\")
print(list)") "ExecuteFile" "Private"))
  (Utilities|Array|Clear (LocalVariables|GettmpList))
  (for _array_element (Utilities|String|ParseIntoArray (Utilities|String|Replace (Utilities|String|Replace (Utilities|String|Replace (Utilities|String|Replace (.Output (Utilities|Struct|BreakPythonLogOutputEntry (Utilities|Array|Get(aref) _logoutput))) "[") "]") "'") "
") ", " false)
    (Utilities|Array|Add _tmplist (Utilities|String|Append Path(/end) _array_element)))
  (return (Utilities|String|ParseIntoArray (Utilities|String|Replace (Utilities|String|Replace (Utilities|String|Replace (Utilities|String|Replace (.Output (Utilities|Struct|BreakPythonLogOutputEntry (Utilities|Array|Get(aref) _logoutput))) "[") "]") "'") "
") ", " false) _tmplist))

(fn GetFolderListInPath
  (params (Path(/end) :type String))
  (bind (_commandresult _logoutput) (Python|Execution|ExecutePythonCommand(Advanced) (Utilities|String|Append "import os


def get_directories(path):
    return [name for name in os.listdir(path) if os.path.isdir(os.path.join(path, name))]


directories = get_directories(\"" (Utilities|String|Replace (Utilities|String|Replace (Utilities|String|Replace (Utilities|String|Replace (Utilities|String|Replace Path(/end) "\\" "\\\\") "\\\\\\" "\\\\") "\\\\\\\\" "\\\\") "\\\\\\\\\\" "\\\\") "\\\\\\\\\\\\" "\\\\") "\")
print(directories)") "ExecuteFile" "Private"))
  (Utilities|Array|Clear (LocalVariables|GettmpList))
  (for _array_element (Utilities|String|ParseIntoArray (Utilities|String|Replace (Utilities|String|Replace (Utilities|String|Replace (Utilities|String|Replace (.Output (Utilities|Struct|BreakPythonLogOutputEntry (Utilities|Array|Get(aref) _logoutput))) "[") "]") "'") "
") ", " false)
    (Utilities|Array|Add (LocalVariables|GettmpList) (Utilities|String|Append Path(/end) _array_element "/")))
  (return (Utilities|String|ParseIntoArray (Utilities|String|Replace (Utilities|String|Replace (Utilities|String|Replace (Utilities|String|Replace (.Output (Utilities|Struct|BreakPythonLogOutputEntry (Utilities|Array|Get(aref) _logoutput))) "[") "]") "'") "
") ", " false) (LocalVariables|GettmpList)))

(fn GetOSName
  (bind (_commandresult _logoutput) (Python|Execution|ExecutePythonCommand(Advanced) "import os

username = os.getlogin()

print(username)" "ExecuteFile" "Private"))
  (return (Utilities|String|SelectString "anonymity" (Utilities|Array|Get(acopy) (Utilities|String|ParseIntoArray (Utilities|String|Replace (Utilities|String|Replace (Utilities|String|Replace (Utilities|String|Replace (.Output (Utilities|Struct|BreakPythonLogOutputEntry (Utilities|Array|Get(aref) _logoutput))) "[") "]") "'") "
") ", " false)) (Utilities|String|IsEmpty (Utilities|Array|Get(acopy) (Utilities|String|ParseIntoArray (Utilities|String|Replace (Utilities|String|Replace (Utilities|String|Replace (Utilities|String|Replace (.Output (Utilities|Struct|BreakPythonLogOutputEntry (Utilities|Array|Get(aref) _logoutput))) "[") "]") "'") "
") ", " false))))))

(fn IsADependencyBanned
  (params (AssetInput :type String) (AssetData :type Asset%20Data%20Structure))
  (for _array_element (LocalVariables|GetNameBanList)
    (LocalVariables|SetHitNameCheck (or (LocalVariables|GetHitNameCheck) (Utilities|String|Contains AssetInput _array_element))))
  (if (LocalVariables|GetHitNameCheck)
    (return false)
    (elif (Utilities|String|Contains (Utilities|GetClassDisplayName (AssetRegistry|FindAssetNativeClass AssetData)) "NiagaraEffectType")
      (return false)
      (else
        (return true)))))

(fn WriteStringIntoFile
  (params (WriteString :type String) (FilePath :type String) (Unique :type Boolean :default true) (Clear :type Boolean))
  (bind _returnvalue (Utilities|String|Replace FilePath "\\" "\\\\"))
  (bind _returnvalue_2 (Utilities|String|Replace _returnvalue "\\\\\\" "\\\\"))
  (bind _returnvalue_3 (Utilities|String|Replace _returnvalue_2 "\\\\\\\\" "\\\\"))
  (bind _returnvalue_4 (Utilities|String|Replace _returnvalue_3 "\\\\\\\\\\" "\\\\"))
  (bind _returnvalue_5 (Utilities|String|Replace _returnvalue_4 "\\\\\\\\\\\\" "\\\\"))
  (if Clear
    (Python|Execution|ExecutePythonCommand (Utilities|String|Append "import os
new_string = \"" WriteString "\\n\"
file_path = \"" _returnvalue_5 "\"
with open(file_path, 'w') as f:
    f.write(new_string)"))
    (elif Unique
      (Python|Execution|ExecutePythonCommand (Utilities|String|Append "import os
new_string = \"" WriteString "\\n\"
file_path = \"" _returnvalue_5 "\"

if not os.path.exists(file_path):
    with open(file_path, 'w') as f:
        f.write(new_string)
else:
    with open(file_path, 'r') as f:
        lines = f.readlines()
    if new_string not in lines:
        with open(file_path, 'a') as f:
            f.write(new_string)"))
      (else
        (Python|Execution|ExecutePythonCommand (Utilities|String|Append "import os
new_string = \"" WriteString "\\n\"
file_path = \"" _returnvalue_5 "\"

if not os.path.exists(file_path):
    with open(file_path, 'w') as f:
        f.write(new_string)
else:
    with open(file_path, 'a') as f:
        f.write(new_string)"))))))
