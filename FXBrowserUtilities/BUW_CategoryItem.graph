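# Editor utility widget for one row in the FX Browser category tree, with a checkbox, label and expand arrow.
# On construct it registers with its SubFrom parent, indents by depth, shows the name after the last dash and starts collapsed if a child.
# Clicking unchecks other rows, cascades down first children and updates ancestors, reporting changes to the owning BUW_Category.
# An invisible checkbox over the arrow rotates it and toggles direct children's visibility.
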EditorUtilityWidgetBlueprint /FXBrowserUtilities/BUW_CategoryItem
graphs [EventGraph {nodes [
   "On Check State Changed (EditorUtilityCheckBox_53)"!() -> (OutputDelegate, then, bIsChecked:bool) {is=title K2Node_ComponentBoundEvent}
   > n1 = Sequence!(execute) -> (then_0, then_1, then_2:Exec, then_3:Exec)
   n2 = Get CreateBy() -> (CreateBy:BUW Category Object Reference)
   Branch(execute=[@n19.then_0, @n1.then_0, @n21.then_0], Condition=IsValid(Object=@n2))
   n3 = IsChecked(self=Get EditorUtilityCheckBox_53() -> (EditorUtilityCheckBox_53:Editor Utility Check Box Object Reference))
   > "[UI]ReceiveCategoryChange"!(execute, self:BUW Category Object Reference=@n2, Check?:bool=@n3, CatName:String=Get CatName() -> (CatName:String)) -> (then)
   CleanUpAll!(execute=[@n1.then_1, @n32.then], self, Exc:BUW Category Item Object Reference=Self-Reference()) -> (then)
   n4 = For Each Loop(
    Exec=[@n19.then_1, @n1.then_2]
    Array:Array of BUW Category Item Object References=Get MyChilds() -> (MyChilds:Array of BUW Category Item Object References)
   ) -> (Array Element:BUW Category Item Object Reference)
   > n5 = Branch(Condition=@n3)
   > RunCheckFromParent!(execute, self=@n4, Check:bool="Equal (Integer)"(A=@n4.Array Index)) -> (then)
   RunCheckFromParent(execute=@n5.else, self=@n4, Check=false)
   n6 = Get SubFrom() -> (SubFrom:BUW Category Item Object Reference)
   Is Valid(exec=@n1.then_3, InputObject=@n6)
   > CleanUpCheck!(execute, self=@n6, Exc:BUW Category Item Object Reference=Self-Reference()) -> (then)
   > RunCheckFromSub(self=@n6)
   n7 = "On Check State Changed (EditorUtilityCheckBox_46)"!() -> (OutputDelegate, then, bIsChecked:bool) {is=title K2Node_ComponentBoundEvent}
   > n8 = Sequence!(execute) -> (then_0, then_1)
   > n9 = For Each Loop(Array=Get MyChilds())
   > SetVisibility(self=@n9, InVisibility=Select(Option 0=Collapsed, Option 1=Visible, Index=@n7.bIsChecked))
   SetRenderTransform!(
    execute=@n8.then_1
    self:Widget Object Reference=Get Image_86() -> (Image_86:Image Object Reference)
    InTransform:Widget Transform Structure=Make Widget Transform!(
     Translation:Vector 2D Structure=(X=0,Y=0)
     Scale:Vector 2D Structure=(X=1,Y=1)
     Shear:Vector 2D Structure=(X=0,Y=0)
     Angle:float=Select(Option 0:float=90, Option 1:float=0, Index=@n7.bIsChecked) -> (ReturnValue:float)
    ) -> (WidgetTransform:Widget Transform Structure) {is=title K2Node_MakeStruct}
   ) -> (then)
   n10 = CleanUpCheck!() -> (OutputDelegate, then, Exc:BUW Category Item Object Reference) {is=title K2Node_CustomEvent}
   > n11 = For Each Loop(Array=Get MyChilds())
   > n12 = Branch(Condition="Equal (Object)"(A=@n10.Exc, B=@n11))
   SetCheckedState!(
    execute=@n12.else
    self:Check Box Object Reference=Get EditorUtilityCheckBox_53!(self:BUW Category Item Object Reference=@n11) -> (EditorUtilityCheckBox_53)
    InCheckedState:ECheckBoxState Enum=Unchecked
   ) -> (then)
   > "[UI]ReceiveCategoryChange"(self=Get CreateBy(), Check?=false, CatName=Get CatName!(self:BUW Category Item Object Reference=@n11) -> (CatName))
   n13 = AddChild!() -> (OutputDelegate, then, NewParam:BUW Category Item Object Reference) {is=title K2Node_CustomEvent}
   > Add(TargetArray:Array of BUW Category Item Object References=Get MyChilds(), NewItem:"BUW Category Item Object Reference (by ref)"=@n13.NewParam) {is=title K2Node_CallArrayFunction}
   n14 = Get Overlay_26() -> (Overlay_26:Overlay Object Reference)
   > SetVisibility(self=@n14, InVisibility=Visible)
   > SetVisibility(self=@n13.NewParam, InVisibility=Collapsed)
   n15 = CleanUpAll!() -> (OutputDelegate, then, Exc:BUW Category Item Object Reference) {is=title K2Node_CustomEvent}
   > n16 = For Each Loop(Array=Get CategoryObject!(self:BUW Category Object Reference=Get CreateBy()) -> (CategoryObject:Array of BUW Category Item Object References))
   > n17 = Branch(Condition="Equal (Object)"(A=@n15.Exc, B=@n16))
   SetCheckedState(execute=@n17.else, self=Get EditorUtilityCheckBox_53(self=@n16), InCheckedState=Unchecked)
   > "[UI]ReceiveCategoryChange"(self=Get CreateBy(), Check?=false, CatName=Get CatName(self=@n16))
   n18 = RunCheckFromParent!() -> (OutputDelegate, then, Check:bool=~) {is=title K2Node_CustomEvent}
   > SetCheckedState(
    self=Get EditorUtilityCheckBox_53!() -> (EditorUtilityCheckBox_53)
    InCheckedState=Select(Option 0:ECheckBoxState Enum=Unchecked, Option 1:ECheckBoxState Enum=Checked, Index=@n18.Check) -> (ReturnValue:ECheckBoxState Enum)
   )
   > n19 = Sequence!(execute) -> (then_0, then_1, then_2:Exec)
   RunCheckFromSub() {is=title K2Node_CustomEvent}
   > Set tmpCheck(tmpCheck:bool=false) -> (Output_Get:bool)
   > n20 = For Each Loop(Array=Get MyChilds())
   SetCheckedState(
    execute=@n20.Completed
    self=Get EditorUtilityCheckBox_53()
    InCheckedState=Select(Option 0=Unchecked, Option 1=Checked, Index=Get tmpCheck() -> (tmpCheck:bool))
   )
   > n21 = Sequence!(execute) -> (then_0, then_1)
   n22 = Get SubFrom()
   Is Valid(exec=@n21.then_1, InputObject=@n22)
   > RunCheckFromSub(self=@n22) {is=title K2Node_CallFunction}
   n23 = Get EditorUtilityCheckBox_53!(self:BUW Category Item Object Reference=@n20) -> (EditorUtilityCheckBox_53)
   Set tmpCheck(execute=@n20.LoopBody, tmpCheck=OR Boolean(A=Get tmpCheck(), B=IsChecked(self=@n23)))
   Event Construct()
   > n24 = Sequence()
   n25 = Get SubFrom()
   > n26 = Branch(Condition=IsValid(Object=@n25))
   > n27 = Set MyLevel(MyLevel:int=int + int(A=Get MyLevel!(self:BUW Category Item Object Reference=@n25) -> (MyLevel:int=0))) -> (Output_Get:int=0)
   > AddChild!(execute, self:Self Object Reference=@n25, NewParam:BUW Category Item Object Reference=Self-Reference()) -> (then) {is=title K2Node_CallFunction}
   > n28 = For Loop(FirstIndex=1, LastIndex=@n27)
   n29 = "SetText (Text)"(
    execute=@n28.Completed
    self=Get Sub() -> (Sub:Text Object Reference)
    InText="To Text (String)"(InString=Get tmpBlank() -> (tmpBlank:String))
   )
   n30 = Get CatName!() -> (CatName)
   n31 = Split(SourceString=@n30, InStr=-, SearchDir=FromEnd)
   "SetText (Text)"(
    execute=[@n29.then, @n26.else]
    self=Get TextBlock_127() -> (TextBlock_127:Text Object Reference)
    InText="To Text (String)"(InString=SelectString(A=@n31.RightS, B=@n30, bPickA=@n31.ReturnValue))
   )
   Set tmpBlank(execute=@n28.LoopBody, tmpBlank:String=Append(A=Get tmpBlank(), B="    ")) -> (Output_Get:String)
   SetVisibility(execute=@n24.then_1, self=@n14, InVisibility=Collapsed)
   n32 = Branch(Condition="Equal (Integer)"(A=Get MyLevel!() -> (MyLevel=0)))
  ]}]
variables [
 CatName {type=string, default=Cat}
 CreateBy {type=BUW_Category_C, default=null}
 MyChilds {type="array<BUW_CategoryItem_C>", default []}
 MyLevel {type=integer}
 SubFrom {type=BUW_CategoryItem_C, default=null}
 tmpBlank {type=string, default=""}
 tmpCheck {type=boolean}
]
widgets {
 info {rootWidgetClass=&HorizontalBox, widgetCount=7}
 widgets [
  HorizontalBox_0 {widgetClassPath=&HorizontalBox, widget=HorizontalBox_0}
  Sub {
   widgetClassPath=&TextBlock
   bIsVariable=true
   parent=HorizontalBox_0
   widget=Sub
   properties {font {
     bForceMonospaced=false
     bMaterialIsStencil=false
     fontMaterial=null
     fontObject=/Engine/EngineFonts/RobotoTiny
     letterSpacing=0
     monospacedWidth=1
     size=18
     skewAmount=0
     typefaceFontName=Light
     outlineSettings {
      bApplyOutlineToDropShadows=false
      bMiteredCorners=false
      bSeparateFillAlpha=false
      outlineMaterial=null
      outlineSize=0
      outlineColor=<rgba 0 0 0 1>
     }
    }}
   slot {class=HorizontalBoxSlot, properties {horizontalAlignment=HAlign_Left, verticalAlignment=VAlign_Center}}
  }
  EditorUtilityCheckBox_53 {
   widgetClassPath=&EditorUtilityCheckBox
   bIsVariable=true
   parent=HorizontalBox_0
   widget=EditorUtilityCheckBox_53
   slot {class=HorizontalBoxSlot}
  }
  TextBlock_127 {
   widgetClassPath=&TextBlock
   bIsVariable=true
   parent=HorizontalBox_0
   widget=TextBlock_127
   properties {
    text=Text Block
    font {
     bForceMonospaced=false
     bMaterialIsStencil=false
     fontMaterial=null
     fontObject=&RobotoTiny
     letterSpacing=0
     monospacedWidth=1
     size=12
     skewAmount=0
     typefaceFontName=Light
     outlineSettings {
      bApplyOutlineToDropShadows=false
      bMiteredCorners=false
      bSeparateFillAlpha=false
      outlineMaterial=null
      outlineSize=0
      outlineColor=<rgba 0 0 0 1>
     }
    }
   }
   slot {class=HorizontalBoxSlot, properties {horizontalAlignment=HAlign_Left, verticalAlignment=VAlign_Center, size {sizeRule=Fill, value=1}}}
  }
  Overlay_26 {widgetClassPath=&Overlay, bIsVariable=true, parent=HorizontalBox_0, widget=Overlay_26, slot {class=HorizontalBoxSlot}}
  Image_86 {
   widgetClassPath=&Image
   bIsVariable=true
   parent=Overlay_26
   widget=Image_86
   properties {
    visibility=HitTestInvisible
    brush {
     bIsDynamicallyLoaded=false
     drawAs=Image
     imageType=FullColor
     mirroring=NoMirror
     resourceName=None
     resourceObject=&/M_Delta
     tiling=NoTile
     imageSize=<16 16>
     margin=<ltrb 0 0 0 0>
     outlineSettings {
      bUseBrushTransparency=false
      roundingType=HalfHeightRadius
      width=0
      color {colorUseRule=UseColor_Specified, specifiedColor=<rgba 0 0 0 0>}
      cornerRadii=<0 0 0 0>
     }
     tintColor {colorUseRule=UseColor_Specified, specifiedColor=<rgba 0.3698 0.3698 0.3698 1>}
     uVRegion {bIsValid=false, max=<0 0>, min=<0 0>}
    }
    renderTransform {angle=90, scale=<1 1>, shear=<0 0>, translation=<0 0>}
   }
   slot {class=OverlaySlot, properties {horizontalAlignment=HAlign_Center, verticalAlignment=VAlign_Center}}
  }
  EditorUtilityCheckBox_46 {
   widgetClassPath=&EditorUtilityCheckBox
   bIsVariable=true
   parent=Overlay_26
   widget=EditorUtilityCheckBox_46
   properties {widgetStyle {
     checkBoxType=CheckBox
     backgroundHoveredImage {
      bIsDynamicallyLoaded=false
      drawAs=RoundedBox
      imageType=NoImage
      mirroring=NoMirror
      resourceName=None
      resourceObject=null
      tiling=NoTile
      imageSize=<18 18>
      margin=<ltrb 0 0 0 0>
      outlineSettings {
       bUseBrushTransparency=false
       roundingType=FixedRadius
       width=1
       color {colorUseRule=UseColor_ColorTable, specifiedColor=<rgba 0 0 0 0>}
       cornerRadii=<3 3 3 3>
      }
      tintColor {colorUseRule=UseColor_Specified, specifiedColor=<rgba 0.0048 0.0048 0.0048 0>}
      uVRegion {bIsValid=false, max=<0 0>, min=<0 0>}
     }
     backgroundImage {
      bIsDynamicallyLoaded=false
      drawAs=RoundedBox
      imageType=NoImage
      mirroring=NoMirror
      resourceName=None
      resourceObject=null
      tiling=NoTile
      imageSize=<18 18>
      margin=<ltrb 0 0 0 0>
      outlineSettings {
       bUseBrushTransparency=false
       roundingType=FixedRadius
       width=1
       color {colorUseRule=UseColor_ColorTable, specifiedColor=<rgba 0 0 0 0>}
       cornerRadii=<3 3 3 3>
      }
      tintColor {colorUseRule=UseColor_Specified, specifiedColor=<rgba 0.0048 0.0048 0.0048 0>}
      uVRegion {bIsValid=false, max=<0 0>, min=<0 0>}
     }
     backgroundPressedImage {
      bIsDynamicallyLoaded=false
      drawAs=RoundedBox
      imageType=NoImage
      mirroring=NoMirror
      resourceName=None
      resourceObject=null
      tiling=NoTile
      imageSize=<18 18>
      margin=<ltrb 0 0 0 0>
      outlineSettings {
       bUseBrushTransparency=false
       roundingType=FixedRadius
       width=1
       color {colorUseRule=UseColor_ColorTable, specifiedColor=<rgba 0 0 0 0>}
       cornerRadii=<3 3 3 3>
      }
      tintColor {colorUseRule=UseColor_Specified, specifiedColor=<rgba 0.0048 0.0048 0.0048 0>}
      uVRegion {bIsValid=false, max=<0 0>, min=<0 0>}
     }
     borderBackgroundColor {colorUseRule=UseColor_Specified, specifiedColor=<rgba 0 0 0 0>}
     checkedForeground {colorUseRule=UseColor_Specified, specifiedColor=<rgba 0 0 0 0>}
     checkedHoveredForeground {colorUseRule=UseColor_Specified, specifiedColor=<rgba 0 0 0 0>}
     checkedHoveredImage {
      bIsDynamicallyLoaded=false
      drawAs=Image
      imageType=Vector
      mirroring=NoMirror
      resourceName=../../../Engine/Content/Slate/Starship/CoreWidgets/CheckBox/check.svg
      resourceObject=null
      tiling=NoTile
      imageSize=<16 16>
      margin=<ltrb 0 0 0 0>
      outlineSettings {
       bUseBrushTransparency=false
       roundingType=HalfHeightRadius
       width=0
       color {colorUseRule=UseColor_Specified, specifiedColor=<rgba 0 0 0 0>}
       cornerRadii=<0 0 0 0>
      }
      tintColor {colorUseRule=UseColor_Specified, specifiedColor=<rgba 0.0044 0.2384 1 0>}
      uVRegion {bIsValid=false, max=<0 0>, min=<0 0>}
     }
     checkedImage {
      bIsDynamicallyLoaded=false
      drawAs=Image
      imageType=Vector
      mirroring=NoMirror
      resourceName=../../../Engine/Content/Slate/Starship/CoreWidgets/CheckBox/check.svg
      resourceObject=null
      tiling=NoTile
      imageSize=<16 16>
      margin=<ltrb 0 0 0 0>
      outlineSettings {
       bUseBrushTransparency=false
       roundingType=HalfHeightRadius
       width=0
       color {colorUseRule=UseColor_Specified, specifiedColor=<rgba 0 0 0 0>}
       cornerRadii=<0 0 0 0>
      }
      tintColor {colorUseRule=UseColor_Specified, specifiedColor=<rgba 0 0.162 0.7454 0>}
      uVRegion {bIsValid=false, max=<0 0>, min=<0 0>}
     }
     checkedPressedForeground {colorUseRule=UseColor_Specified, specifiedColor=<rgba 0 0 0 0>}
     checkedPressedImage {
      bIsDynamicallyLoaded=false
      drawAs=Image
      imageType=Vector
      mirroring=NoMirror
      resourceName=../../../Engine/Content/Slate/Starship/CoreWidgets/CheckBox/check.svg
      resourceObject=null
      tiling=NoTile
      imageSize=<16 16>
      margin=<ltrb 0 0 0 0>
      outlineSettings {
       bUseBrushTransparency=false
       roundingType=HalfHeightRadius
       width=0
       color {colorUseRule=UseColor_Specified, specifiedColor=<rgba 0 0 0 0>}
       cornerRadii=<0 0 0 0>
      }
      tintColor {colorUseRule=UseColor_Specified, specifiedColor=<rgba 0 0.0802 0.3515 0>}
      uVRegion {bIsValid=false, max=<0 0>, min=<0 0>}
     }
     checkedSlateSound {resourceObject=null}
     foregroundColor {colorUseRule=UseColor_Specified, specifiedColor=<rgba 0 0 0 0>}
     hoveredForeground {colorUseRule=UseColor_Specified, specifiedColor=<rgba 0 0 0 0>}
     hoveredSlateSound {resourceObject=null}
     padding=<ltrb 2 0 0 0>
     pressedForeground {colorUseRule=UseColor_Specified, specifiedColor=<rgba 0 0 0 0>}
     uncheckedHoveredImage {
      bIsDynamicallyLoaded=false
      drawAs=NoDrawType
      imageType=NoImage
      mirroring=NoMirror
      resourceName=None
      resourceObject=null
      tiling=NoTile
      imageSize=<0 0>
      margin=<ltrb 0 0 0 0>
      outlineSettings {
       bUseBrushTransparency=false
       roundingType=HalfHeightRadius
       width=0
       color {colorUseRule=UseColor_Specified, specifiedColor=<rgba 0 0 0 0>}
       cornerRadii=<0 0 0 0>
      }
      tintColor {colorUseRule=UseColor_Specified, specifiedColor=<rgba 1 1 1 0>}
      uVRegion {bIsValid=false, max=<0 0>, min=<0 0>}
     }
     uncheckedImage {
      bIsDynamicallyLoaded=false
      drawAs=NoDrawType
      imageType=NoImage
      mirroring=NoMirror
      resourceName=None
      resourceObject=null
      tiling=NoTile
      imageSize=<0 0>
      margin=<ltrb 0 0 0 0>
      outlineSettings {
       bUseBrushTransparency=false
       roundingType=HalfHeightRadius
       width=0
       color {colorUseRule=UseColor_Specified, specifiedColor=<rgba 0 0 0 0>}
       cornerRadii=<0 0 0 0>
      }
      tintColor {colorUseRule=UseColor_Specified, specifiedColor=<rgba 1 1 1 0>}
      uVRegion {bIsValid=false, max=<0 0>, min=<0 0>}
     }
     uncheckedPressedImage {
      bIsDynamicallyLoaded=false
      drawAs=NoDrawType
      imageType=NoImage
      mirroring=NoMirror
      resourceName=None
      resourceObject=null
      tiling=NoTile
      imageSize=<0 0>
      margin=<ltrb 0 0 0 0>
      outlineSettings {
       bUseBrushTransparency=false
       roundingType=HalfHeightRadius
       width=0
       color {colorUseRule=UseColor_Specified, specifiedColor=<rgba 0 0 0 0>}
       cornerRadii=<0 0 0 0>
      }
      tintColor {colorUseRule=UseColor_Specified, specifiedColor=<rgba 1 1 1 0>}
      uVRegion {bIsValid=false, max=<0 0>, min=<0 0>}
     }
     uncheckedSlateSound {resourceObject=null}
     undeterminedForeground {colorUseRule=UseColor_Specified, specifiedColor=<rgba 0 0 0 0>}
     undeterminedHoveredImage {
      bIsDynamicallyLoaded=false
      drawAs=Image
      imageType=Vector
      mirroring=NoMirror
      resourceName=../../../Engine/Content/Slate/Starship/CoreWidgets/CheckBox/indeterminate.svg
      resourceObject=null
      tiling=NoTile
      imageSize=<16 16>
      margin=<ltrb 0 0 0 0>
      outlineSettings {
       bUseBrushTransparency=false
       roundingType=HalfHeightRadius
       width=0
       color {colorUseRule=UseColor_Specified, specifiedColor=<rgba 0 0 0 0>}
       cornerRadii=<0 0 0 0>
      }
      tintColor {colorUseRule=UseColor_Specified, specifiedColor=<rgba 0.0044 0.2384 1 0>}
      uVRegion {bIsValid=false, max=<0 0>, min=<0 0>}
     }
     undeterminedImage {
      bIsDynamicallyLoaded=false
      drawAs=Image
      imageType=Vector
      mirroring=NoMirror
      resourceName=../../../Engine/Content/Slate/Starship/CoreWidgets/CheckBox/indeterminate.svg
      resourceObject=null
      tiling=NoTile
      imageSize=<16 16>
      margin=<ltrb 0 0 0 0>
      outlineSettings {
       bUseBrushTransparency=false
       roundingType=HalfHeightRadius
       width=0
       color {colorUseRule=UseColor_Specified, specifiedColor=<rgba 0 0 0 0>}
       cornerRadii=<0 0 0 0>
      }
      tintColor {colorUseRule=UseColor_Specified, specifiedColor=<rgba 0 0.162 0.7454 0>}
      uVRegion {bIsValid=false, max=<0 0>, min=<0 0>}
     }
     undeterminedPressedImage {
      bIsDynamicallyLoaded=false
      drawAs=Image
      imageType=Vector
      mirroring=NoMirror
      resourceName=../../../Engine/Content/Slate/Starship/CoreWidgets/CheckBox/indeterminate.svg
      resourceObject=null
      tiling=NoTile
      imageSize=<16 16>
      margin=<ltrb 0 0 0 0>
      outlineSettings {
       bUseBrushTransparency=false
       roundingType=HalfHeightRadius
       width=0
       color {colorUseRule=UseColor_Specified, specifiedColor=<rgba 0 0 0 0>}
       cornerRadii=<0 0 0 0>
      }
      tintColor {colorUseRule=UseColor_Specified, specifiedColor=<rgba 0 0.0802 0.3515 0>}
      uVRegion {bIsValid=false, max=<0 0>, min=<0 0>}
     }
    }}
   slot {class=OverlaySlot, properties {horizontalAlignment=HAlign_Fill, verticalAlignment=VAlign_Fill}}
  }
 ]
}
