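# Editor utility widget that shows a category checklist for FX browser tools, with entries such as Pickaxe, Glider and Skin.
# On Construct it rebuilds CategoryBox from CategoryList, linking hyphenated names to parent items by splitting at the last hyphen.
# Item changes update a parallel checked array.
# A short retriggerable delay batches these changes into one CategoryChanged broadcast of both arrays, then prints each state.
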EditorUtilityWidgetBlueprint /FXBrowserUtilities/BUW_Category
graphs [
 CategoryChanged {nodes [CategoryChanged!() -> (then, CategoryList:Array of Strings, CheckedList:Array of Booleans, Shuffle:bool=~) {is=title K2Node_FunctionEntry}]}
 EventGraph {nodes [
   n1 = "[UI]ReceiveCategoryChange"!() -> (OutputDelegate, then, Check?:bool=~, CatName:String) {is=title K2Node_CustomEvent}
   > n2 = Sequence()
   n3 = Find Item!(TargetArray=Get CategoryList() -> (CategoryList:Array of Strings), ItemToFind:"String (by ref)"=@n1.CatName) -> (ReturnValue)
   > Branch(Condition="integer >= integer"(A=@n3))
   > Set Array Elem!(
    execute
    TargetArray:Array of Booleans=Get CategoryChecked() -> (CategoryChecked:Array of Booleans)
    Index:int=@n3
    Item:"Boolean (by ref)"=@n1.Check?
    bSizeToFit:bool=false
   ) -> (then)
   RetriggerableDelay!(execute=@n2.then_1, WorldContextObject:Object Reference, Duration=0.01) -> (then)
   > Call Category Changed!(
    execute:Exec
    self:BUW Category Object Reference
    CategoryList:Array of Strings=Get CategoryList()
    CheckedList:Array of Booleans=Get CategoryChecked()
    Shuffle:bool
   ) -> (then:Exec) {is=title K2Node_CallDelegate}
   > n4 = For Each Loop(Array:Array of Booleans=Get CategoryChecked()) -> (Array Element:bool)
   n5 = Get CategoryList()
   > PrintString(InString=Append!(
     A="Get (a ref)"(Array:Array of Strings=@n5, Dimension 1=@n4.Array Index) -> (Output:"String (by ref)")
     B=-----
     C:String="To String (Boolean)"(InBool=@n4)
    ) -> (ReturnValue))
   Event Construct()
   > "[UI]InitialCategoryList"()
   Event PreConstruct()
   Event Tick()
  ]}
 "[UI]InitialCategoryList" {nodes [
   "[UI]InitialCategoryList"() {is=title K2Node_FunctionEntry}
   > ClearChildren(self=Get CategoryBox() -> (CategoryBox:Vertical Box Object Reference))
   > n1 = For Each Loop(Array:Array of Strings=Get CategoryList()) -> (Array Element:String)
   > n2 = Branch(Condition="integer > integer"(A=FindSubstring(SearchIn=@n1, Substring=-, bSearchFromEnd=true), B=-1))
   n3 = Split(SourceString=@n1, InStr=-, SearchDir=FromEnd)
   n4 = Get CategoryList()
   n5 = Self-Reference()
   > n6 = Create Widget!(
    execute:Exec
    WorldContextObject:Object Reference
    Class:User Widget Class Reference=&/BUW_CategoryItem.BUW_CategoryItem_C
    OwningPlayer:Player Controller Object Reference
    SubFrom:BUW Category Item Object Reference="Get (a copy)"(
     Array:Array of BUW Category Item Object References=Get CategoryObject() -> (CategoryObject:Array of BUW Category Item Object References)
     Dimension 1=Find Item(TargetArray=@n4, ItemToFind=@n3)
    ) -> (Output:BUW Category Item Object Reference)
    CatName:String=@n1
    CreateBy:BUW Category Object Reference=@n5
   ) -> (then:Exec, ReturnValue:BUW Category Item Object Reference)
   n7 = Get CategoryBox()
   > AddChild(self=@n7, Content=@n6)
   > Add(TargetArray:Array of BUW Category Item Object References=Get CategoryObject(), NewItem:"BUW Category Item Object Reference (by ref)"=@n6) {is=title K2Node_CallArrayFunction}
   n8 = Get CategoryChecked()
   > Add(TargetArray:Array of Booleans=@n8, NewItem:"Boolean (by ref)")
   n9 = Create Widget(execute=@n2.else, Class=&BUW_CategoryItem_C, CatName=@n1, CreateBy=@n5)
   > AddChild(self=@n7, Content=@n9)
   > Add(TargetArray:Array of BUW Category Item Object References=Get CategoryObject(), NewItem:"BUW Category Item Object Reference (by ref)"=@n9)
   > Add(TargetArray:Array of Booleans=@n8, NewItem:"Boolean (by ref)")
  ]}
]
variables [
 CategoryChanged {default=()}
 CategoryChecked {type="array<boolean>", default []}
 CategoryList {
  type="array<string>"
  default [Pickaxe, Pickaxe-Idle, Pickaxe-Impact, Pickaxe-Impact-2, Pickaxe-Trail, Glider, Glider-Deploy, Emote, Skydive, Skin]
 }
 CategoryObject {type="array<BUW_CategoryItem_C>", default []}
]
widgets {
 info {rootWidgetClass=&VerticalBox, widgetCount=6}
 widgets [
  CategoryBox {widgetClassPath=&VerticalBox, bIsVariable=true, widget=CategoryBox}
  BUW_CategoryItem {widgetClassPath=&/BUW_CategoryItem, parent=CategoryBox, widget=BUW_CategoryItem, slot {class=VerticalBoxSlot}}
  BUW_CategoryItem_1 {widgetClassPath=&BUW_CategoryItem, parent=CategoryBox, widget=BUW_CategoryItem_1, slot {class=VerticalBoxSlot}}
  BUW_CategoryItem_2 {widgetClassPath=&BUW_CategoryItem, parent=CategoryBox, widget=BUW_CategoryItem_2, slot {class=VerticalBoxSlot}}
  BUW_CategoryItem_3 {widgetClassPath=&BUW_CategoryItem, parent=CategoryBox, widget=BUW_CategoryItem_3, slot {class=VerticalBoxSlot}}
  BUW_CategoryItem_4 {widgetClassPath=&BUW_CategoryItem, parent=CategoryBox, widget=BUW_CategoryItem_4, slot {class=VerticalBoxSlot}}
 ]
}
