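# Per-item editor utility object for the FX Browser, holding one preview entry's thumbnail, favorite state and tag editing.
# Initial loads the item's JPG from disk as a texture and derives Fav from a special marker tag.
# CreateNewTag and DeleteATag rewrite a per-user delimited tag text file line by line.
# Both then sync the SaveGame's tag-to-item index and the item's tags, save the database and refresh the parent widget's tag UI.
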EditorUtilityBlueprint /FXBrowserUtilities/EUO_PreviewItemObject
graphs [
 CreateNewTag {nodes [
   n1 = CreateNewTag!() -> (then, Tag:String, SpecialUser:String) {is=title K2Node_FunctionEntry}
   > n2 = Sequence!(execute) -> (then_0, then_1, then_2:Exec)
   n3 = Get ParentWidget() -> (ParentWidget:BUW FXBrowser Object Reference)
   n4 = SelectString(
    A=Get CurrentUser!(self:BUW FXBrowser Object Reference=@n3) -> (CurrentUser:String)
    B=@n1.SpecialUser
    bPickA=IsEmpty(InString=@n1.SpecialUser)
   )
   n5 = Append!(A=Get LibraryBasePath!(self:BUW FXBrowser Object Reference=@n3) -> (LibraryBasePath:String), B=Tag/, C:String=@n4, D:String=.tag.txt) -> (ReturnValue)
   > n6 = Read File to String!(execute, FilePath:String=@n5, bUseFullFilePath:bool=true) -> (then, ResultOut:String, ReturnValue:bool)
   > WriteStringIntoFile!(execute, WriteString:String, FilePath:String=@n5, Unique:bool=false, Clear:bool=true, __WorldContext:Object Reference=Get ParentWidget()) -> (then)
   > n7 = For Each Loop(Array=ParseIntoArray(SourceString=@n6, Delimiter="\r\n"))
   n8 = Branch(execute=@n7.Completed, Condition=Get tmphastag() -> (tmphastag:bool))
   n9 = Get ParentWidget()
   n10 = Break St FXB Item!(St_FXB_Item:"St FXB Item Structure (by ref)"="Get (a ref)"(
     Array:Array of St FXB Item Structures=Get CurrentPendingLoadItemList!(self:BUW FXBrowser Object Reference=@n9) -> (CurrentPendingLoadItemList:Array of St FXB Item Structures)
     Dimension 1=Get MyIndex() -> (MyIndex:int=0)
    ) -> (Output:"St FXB Item Structure (by ref)")) -> (
    AssetDataPath_12_#1:String
    EditorName_14_#2:String
    PublishName_16_#3:String
    PublishVersion_20_#4:int
    HasValidFX_22_#5:bool=~
    X_25_#6:int
    Y_27_#7:int
    VideoList_31_#8:Array of Strings
    ColorList_39_#9:Array of St FXB Color Array Structures
    Depend_38_#10:String
    Category_48_#11:String
    FolderPath_50_#12:String
    Tags_54_#13:Array of Strings
    MyIndexInDatabase_58_#14:int
    MyCategoryIndexInDataBase_61_#15:int
   ) {is=title K2Node_BreakStruct}
   WriteStringIntoFile(
    execute=@n8.else
    WriteString=Append(A=@n1, B="%&&%", C=@n10.EditorName_14_#2, D="%@%")
    FilePath=@n5
    Unique=true
    Clear=false
    __WorldContext=Get ParentWidget()
   )
   n11 = Append!(A=@n1, B="%&&%") -> (ReturnValue)
   n12 = Branch(execute=@n7.LoopBody, Condition="integer >= integer"(A=FindSubstring(SearchIn=@n7, Substring=@n11)))
   > Set tmphastag(tmphastag:bool=true) -> (Output_Get:bool)
   > n13 = Branch(Condition="integer >= integer"(A=FindSubstring(SearchIn=@n7, Substring=@n10.EditorName_14_#2)))
   n14 = Get ParentWidget()
   WriteStringIntoFile(execute=[@n12.else, @n13.then], WriteString=@n7, FilePath=@n5, Unique=true, Clear=false, __WorldContext=@n14)
   WriteStringIntoFile(
    execute=@n13.else
    WriteString=Append!(A=@n7, B=@n10.EditorName_14_#2, C:String="%@%") -> (ReturnValue)
    FilePath=@n5
    Unique=true
    Clear=false
    __WorldContext=@n14
   )
   n15 = Read File to String(execute=@n2.then_1, FilePath=@n5, bUseFullFilePath=true)
   n16 = Append!(A=@n4, B=.tag.txt) -> (ReturnValue)
   n17 = Get DataBaseSaveGame!(self:BUW FXBrowser Object Reference=Get ParentWidget()) -> (DataBaseSaveGame:FXB Save Game Object Reference)
   n18 = Get DataBase_TagFiles!(self:FXB Save Game Object Reference=@n17) -> (DataBase_TagFiles:Array of Strings)
   n19 = Find Item!(TargetArray=@n18, ItemToFind:"String (by ref)"=@n16) -> (ReturnValue)
   > n20 = Branch(Condition="integer >= integer"(A=@n19))
   n21 = Get DataBase_TagStrings!(self:FXB Save Game Object Reference=@n17) -> (DataBase_TagStrings:Array of Strings)
   > Set Array Elem!(execute, TargetArray=@n21, Index:int=@n19, Item:"String (by ref)"=@n15, bSizeToFit:bool=false) -> (then)
   Add(execute=@n20.else, TargetArray=@n21, NewItem=@n15) {is=title K2Node_CallArrayFunction}
   > Add(TargetArray=@n18, NewItem=@n16)
   n22 = Get DataBaseSaveGame(self=Get ParentWidget())
   n23 = Get DataBase_TagList!(self:FXB Save Game Object Reference=@n22) -> (DataBase_TagList:Array of Strings)
   n24 = Find Item(TargetArray=@n23, ItemToFind=@n1)
   n25 = Branch(execute=@n2.then_2, Condition="integer >= integer"(A=@n24))
   n26 = Get DataBase_Tag2ItemArray!(self:FXB Save Game Object Reference=@n22) -> (DataBase_Tag2ItemArray:Array of St FXB Tag 2Item Array Structures)
   n27 = "Get (a ref)"(Array:Array of St FXB Tag 2Item Array Structures=@n26, Dimension 1=@n24) -> (Output:"St FXB Tag 2Item Array Structure (by ref)")
   n28 = Break St FXB Tag 2Item Array!(St_FXB_Tag2ItemArray:"St FXB Tag 2Item Array Structure (by ref)"=@n27) -> (
    Tag2ItemList_3_#16:Array of St FXB Tag 2Item Structures
    ItemNameList_6_#17:Array of Strings
   ) {is=title K2Node_BreakStruct}
   > n29 = Set NewLocalVar(NewLocalVar:Array of St FXB Tag 2Item Structures=@n28) -> (Output_Get:Array of St FXB Tag 2Item Structures)
   > n30 = Set NewLocalVar_0(NewLocalVar_0:Array of Strings=@n28.ItemNameList_6_#17) -> (Output_Get:Array of Strings)
   n31 = Make St FXB Tag 2Item!(
    ItemEditorName_3_#18:String=@n10.EditorName_14_#2
    ItemCategoryIndex_7_#19:int=@n10.MyCategoryIndexInDataBase_61_#15
    ItemIndex_9_#20:int=@n10.MyIndexInDatabase_58_#14
   ) -> (St_FXB_Tag2Item:St FXB Tag 2Item Structure) {is=title K2Node_MakeStruct}
   > Add(TargetArray:Array of St FXB Tag 2Item Structures=@n29, NewItem:"St FXB Tag 2Item Structure (by ref)"=@n31)
   > Add(TargetArray:Array of Strings=@n30, NewItem:"String (by ref)"=@n10.EditorName_14_#2)
   > "Set St FXB Tag 2Item Array Structure (by ref)"!(
    execute:Exec
    Target:"St FXB Tag 2Item Array Structure (by ref)"=@n27
    Value:St FXB Tag 2Item Array Structure=Make St FXB Tag 2Item Array!(
     Tag2ItemList_3_#16:Array of St FXB Tag 2Item Structures=@n29
     ItemNameList_6_#17:Array of Strings=@n30
    ) -> (St_FXB_Tag2ItemArray:St FXB Tag 2Item Array Structure) {is=title K2Node_MakeStruct}
   ) -> (then:Exec) {is=title K2Node_VariableSetRef}
   > n32 = For Each Loop(Array=Break St FXB Tag 2Item Array(St_FXB_Tag2ItemArray="Get (a ref)"(Array=@n26, Dimension 1=@n24)).ItemNameList_6_#17)
   > PrintString(InString=@n32)
   n33 = Get ParentWidget()
   n34 = "Get (a ref)"(
    Array:Array of St FXB Item List Structures=Get DataBase_ItemList!(self:FXB Save Game Object Reference=Get DataBaseSaveGame(self=@n33)) -> (DataBase_ItemList:Array of St FXB Item List Structures)
    Dimension 1=@n10.MyCategoryIndexInDataBase_61_#15
   ) -> (Output:"St FXB Item List Structure (by ref)")
   n35 = Break St FXB Item List!(St_FXB_ItemList:"St FXB Item List Structure (by ref)"=@n34) -> (ItemList_16_#21:Array of St FXB Item Structures) {is=title K2Node_BreakStruct}
   n36 = "Get (a ref)"(Array:Array of St FXB Item Structures=@n35, Dimension 1=@n10.MyIndexInDatabase_58_#14) -> (Output:"St FXB Item Structure (by ref)")
   n37 = Set NewLocalVar_1(
    execute=[@n40.then, @n32.Completed]
    NewLocalVar_1:Array of Strings=Break St FXB Item(St_FXB_Item=@n36) -> (HasValidFX_22_#5=~).Tags_54_#13
   ) -> (Output_Get:Array of Strings)
   > Add Unique(TargetArray=@n37, NewItem=Get Tag() -> (Tag:String))
   > n38 = Set members in St FXB Item!(
    execute:Exec
    StructRef:"St FXB Item Structure (by ref)"=@n36
    Tags_54_#13:Array of Strings=Get NewLocalVar_1() -> (NewLocalVar_1:Array of Strings)
   ) -> (then:Exec, StructOut:"St FXB Item Structure (by ref)") {is=title K2Node_SetFieldsInStruct}
   > Set Array Elem(
    TargetArray:Array of St FXB Item Structures=@n35
    Index=@n10.MyIndexInDatabase_58_#14
    Item:"St FXB Item Structure (by ref)"=@n38
    bSizeToFit=false
   )
   > "Set St FXB Item List Structure (by ref)"!(
    execute:Exec
    Target:"St FXB Item List Structure (by ref)"=@n34
    Value:St FXB Item List Structure=Make St FXB Item List!(ItemList_16_#21:Array of St FXB Item Structures=@n35) -> (St_FXB_ItemList:St FXB Item List Structure) {is=title K2Node_MakeStruct}
   ) -> (then:Exec) {is=title K2Node_VariableSetRef}
   n39 = Get ParentWidget()
   > Set Array Elem(TargetArray=Get CurrentPendingLoadItemList(self=@n39), Index=Get MyIndex() -> (MyIndex=0), Item=@n38, bSizeToFit=false)
   > Set CurrentPlayItem!(execute, CurrentPlayItem:St FXB Item Structure=@n38, self:BUW FXBrowser Object Reference=@n39) -> (then, Output_Get:St FXB Item Structure)
   > SaveGameToSlot!(execute, SaveGameObject:Save Game Object Reference=@n22, SlotName:String=FXBrowser, UserIndex:int=0) -> (then, ReturnValue:bool)
   > UI_InitialTag(self:BUW FXBrowser Object Reference=Get ParentWidget())
   Add(execute=@n25.else, TargetArray=@n23, NewItem=@n1)
   > n40 = Add(
    TargetArray:Array of St FXB Tag 2Item Array Structures=@n26
    NewItem:"St FXB Tag 2Item Array Structure (by ref)"=Make St FXB Tag 2Item Array(
     Tag2ItemList_3_#16=Make Array("[0]":St FXB Tag 2Item Structure=@n31) -> (Array:Array of St FXB Tag 2Item Structures)
     ItemNameList_6_#17=Make Array("[0]":String=@n10.EditorName_14_#2) -> (Array:Array of Strings)
    )
   )
  ]}
 DeleteATag {nodes [
   n1 = DeleteATag!() -> (then, Tag:String, SpecialUser:String) {is=title K2Node_FunctionEntry}
   > n2 = Sequence()
   n3 = Get ParentWidget()
   n4 = SelectString(A=Get CurrentUser(self=@n3), B=@n1.SpecialUser, bPickA=IsEmpty(InString=@n1.SpecialUser))
   n5 = Append!(A=Get LibraryBasePath(self=@n3), B=Tag/, C:String=@n4, D:String=.tag.txt) -> (ReturnValue)
   > n6 = Read File to String(FilePath=@n5, bUseFullFilePath=true)
   > WriteStringIntoFile(FilePath=@n5, Unique=false, Clear=true, __WorldContext=Get ParentWidget())
   > n7 = For Each Loop(Array=ParseIntoArray(SourceString=@n6, Delimiter="\r\n"))
   n8 = Append!(A=@n1, B="%&&%") -> (ReturnValue)
   > n9 = Branch(Condition="integer >= integer"(A=FindSubstring(SearchIn=@n7, Substring=@n8)))
   n10 = Get ParentWidget()
   n11 = Break St FXB Item(St_FXB_Item="Get (a ref)"(Array=Get CurrentPendingLoadItemList(self=@n10), Dimension 1=Get MyIndex() -> (MyIndex=0))) -> (HasValidFX_22_#5=~)
   > n12 = Branch(Condition="integer >= integer"(A=FindSubstring(SearchIn=@n7, Substring=@n11.EditorName_14_#2)))
   > Set bFoundRemoveTag(bFoundRemoveTag:bool=true) -> (Output_Get:bool)
   n13 = Replace(SourceString=@n7, From=Append(A=@n11.EditorName_14_#2, B="%@%"))
   n14 = Get ParentWidget()
   > WriteStringIntoFile(WriteString=@n13, FilePath=@n5, Unique=true, Clear=false, __WorldContext=@n14)
   WriteStringIntoFile(execute=[@n9.else, @n12.else], WriteString=@n7, FilePath=@n5, Unique=true, Clear=false, __WorldContext=@n14)
   n15 = Read File to String(execute=@n2.then_1, FilePath=@n5, bUseFullFilePath=true)
   n16 = Append(A=@n4, B=.tag.txt)
   n17 = Get DataBaseSaveGame(self=Get ParentWidget())
   n18 = Get DataBase_TagFiles(self=@n17)
   n19 = Find Item(TargetArray=@n18, ItemToFind=@n16)
   > n20 = Branch(Condition="integer >= integer"(A=@n19))
   n21 = Get DataBase_TagStrings(self=@n17)
   > Set Array Elem(TargetArray:Array of Strings=@n21, Index=@n19, Item:"String (by ref)"=@n15, bSizeToFit=false)
   Add(execute=@n20.else, TargetArray:Array of Strings=@n21, NewItem:"String (by ref)"=@n15)
   > Add(TargetArray=@n18, NewItem=@n16)
   n22 = Get DataBaseSaveGame(self=Get ParentWidget())
   n23 = Find Item(TargetArray=Get DataBase_TagList(self=@n22), ItemToFind=@n1)
   Branch(execute=@n2.then_2, Condition="integer >= integer"(A=@n23))
   n24 = "Get (a ref)"(Array:Array of St FXB Tag 2Item Array Structures=Get DataBase_Tag2ItemArray(self=@n22), Dimension 1=@n23) -> (Output:"St FXB Tag 2Item Array Structure (by ref)")
   n25 = Break St FXB Tag 2Item Array(St_FXB_Tag2ItemArray=@n24)
   > n26 = Set NewLocalVar(NewLocalVar=@n25)
   > n27 = Set NewLocalVar_0(NewLocalVar_0=@n25.ItemNameList_6_#17)
   n28 = Find Item(TargetArray=@n25.ItemNameList_6_#17, ItemToFind=@n11.EditorName_14_#2)
   > Remove Index!(execute, TargetArray:Array of St FXB Tag 2Item Structures=@n26, IndexToRemove:int=@n28) -> (then)
   > Remove Index(TargetArray:Array of Strings=@n27, IndexToRemove=@n28)
   > "Set St FXB Tag 2Item Array Structure (by ref)"(
    Target=@n24
    Value=Make St FXB Tag 2Item Array(Tag2ItemList_3_#16=@n26, ItemNameList_6_#17=@n27)
   )
   n29 = Get ParentWidget()
   n30 = "Get (a ref)"(
    Array:Array of St FXB Item List Structures=Get DataBase_ItemList(self=Get DataBaseSaveGame(self=@n29))
    Dimension 1=@n11.MyCategoryIndexInDataBase_61_#15
   ) -> (Output:"St FXB Item List Structure (by ref)")
   n31 = Break St FXB Item List(St_FXB_ItemList=@n30)
   n32 = "Get (a ref)"(Array:Array of St FXB Item Structures=@n31, Dimension 1=@n11.MyIndexInDatabase_58_#14) -> (Output:"St FXB Item Structure (by ref)")
   > n33 = Set NewLocalVar_1(NewLocalVar_1=Break St FXB Item(St_FXB_Item=@n32) -> (HasValidFX_22_#5=~).Tags_54_#13)
   > Remove Item!(execute, TargetArray=@n33, Item:"String (by ref)"=Get Tag()) -> (then, ReturnValue:bool)
   > n34 = Set members in St FXB Item(StructRef=@n32, Tags_54_#13=Get NewLocalVar_1())
   > Set Array Elem(
    TargetArray:Array of St FXB Item Structures=@n31
    Index=@n11.MyIndexInDatabase_58_#14
    Item:"St FXB Item Structure (by ref)"=@n34
    bSizeToFit=false
   )
   > "Set St FXB Item List Structure (by ref)"(Target=@n30, Value=Make St FXB Item List(ItemList_16_#21=@n31))
   n35 = Get ParentWidget()
   > Set Array Elem(TargetArray=Get CurrentPendingLoadItemList(self=@n35), Index=Get MyIndex() -> (MyIndex=0), Item=@n34, bSizeToFit=false)
   > Set CurrentPlayItem(CurrentPlayItem=@n34, self=@n35)
   > SaveGameToSlot(SaveGameObject=@n22, SlotName=FXBrowser, UserIndex=0)
   > UI_InitialTag(self=Get ParentWidget())
   Branch(Condition=IsEmpty(InString=Split(SourceString=@n13, InStr="%&&%", SearchDir=FromStart).RightS))
  ]}
 EventGraph {nodes [
   Initial() {is=title K2Node_CustomEvent}
   > Is Valid(InputObject=Get ParentWidget())
   n1 = Append!(A=Get FolderPath() -> (FolderPath:String), B=Get LoadRuler() -> (LoadRuler:String), C:String=.jpg) -> (ReturnValue)
   > n2 = Create Texture from Image!(execute, TextureOuter:Object Reference=Self-Reference(), FilePath:String=@n1, bUseFullFilePath:bool=true) -> (then, OutTexture:Texture 2D Object Reference, ReturnValue:bool)
   > n3 = Branch(Condition=@n2.ReturnValue)
   Set LoadedImage(
    execute=[@n3.then, @n6.then]
    LoadedImage:Texture 2D Soft Object Reference=Make Soft Reference!(Input:Texture 2D Object Reference=@n2) -> (Output:Texture 2D Soft Object Reference)
   ) -> (Output_Get:Texture 2D Soft Object Reference)
   n4 = Get ParentWidget()
   n5 = Break St FXB Item(St_FXB_Item="Get (a ref)"(Array=Get CurrentPendingLoadItemList(self=@n4), Dimension 1=Get MyIndex() -> (MyIndex=0))) -> (HasValidFX_22_#5=~)
   > Set Fav(Fav:bool="integer >= integer"(A=Find Item(TargetArray=@n5.Tags_54_#13, ItemToFind=Favmark114514))) -> (Output_Get:bool)
   n6 = Branch(execute=@n3.else, Condition=FileExists!(InPath:String=@n1) -> (ReturnValue:bool))
   Set NeedHide(execute=@n6.else, NeedHide:bool=true) -> (Output_Get:bool)
  ]}
]
variables [
 Fav {type=boolean}
 FolderPath {type=string, default=""}
 LoadedImage {type=Texture2D}
 LoadRuler {type=string, default=""}
 MyIndex {type=integer}
 NeedHide {type=boolean}
 ParentWidget {type=BUW_FXBrowser_C, default=null}
 SubUVResX {type=integer, default=4}
 SubUVResY {type=integer, default=4}
]
