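# An Editor Utility Blueprint object for the FX Browser utilities that appears, from its name, to back a tag button.
# It holds a tag string T, a boolean H and a reference to a parent BUW_FXBrowser widget.
# Its event graph is empty, so it acts purely as a data container with no logic of its own.
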EditorUtilityBlueprint /FXBrowserUtilities/EUO_TagButtonWObject
graphs [EventGraph {links [], nodes []}]
variables [
 H {type=boolean}
 ParentWidget {type=BUW_FXBrowser_C, default=null}
 T {type=string, default=""}
]
