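# UI material that animates a flipbook preview from a sprite-sheet texture parameter, apparently for the FX browser's Finder.
# Pause-independent time, scaled by parameter D and the PlaySped value from MPC_PreviewImage, yields a whole-frame index.
# That index becomes column and row offsets in a square grid sized by S.R, remapping UVs for a mip-0 sample.
# Frames step without blending, despite the "Frame Lerp" label.
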Material /FXBrowserUtilities/FXRefLib/Finder/T_ShowPreview
graph {
 properties {materialDomain=MD_UI, pixelDepthOffsetMode=PDOM_Legacy, usedShadingModels=Default Lit}
 nodes [
  n1 = Multiply(
   A=Multiply(
    A=ScalarParameter() {defaultValue=18, parameterName=D}
    B=CollectionParameter() {collection=/FXBrowserUtilities/MPC_PreviewImage, parameterName=PlaySped}
   )
   B=Time() {bIgnorePause=true, bOverride_Period=true, period=666}
  )
  n2 = Subtract(A=@n1, B=Frac(None=@n1) {desc=Frame Lerp}) {desc=IndexA}
  n3 = VectorParameter() {parameterName=S, defaultValue=<rgba 4 4 0 1>}
  n4 = AppendVector(A=@n3.R, B=@n3.R)
  n5 = Divide(A=Constant2Vector() {g=1, r=1}, B=@n4) {desc=Scale}
  n6 = ComponentMask(None=@n5) {desc="1 / ## Horz Images", r=true}
  n7 = AppendVector(
   A=Fmod(A=@n2, B=ComponentMask(None=@n4) {desc="## Horz Images", r=true}) {desc=IndexA_Horz}
   B=Floor(None=Multiply(A=@n2, B=@n6)) {desc=IndexA_Vert}
  ) {desc=IndexA_Combined}
  n8 = TextureCoordinate()
  TextureSample!(
   UVs=Multiply(A=Add(A=@n7, B=@n8), B=@n5) {desc=UV_A}
   Tex=TextureObjectParameter() {parameterName=T, texture=/Game/Global/Textures/Default/Blanks/T_Blank_LegalWhite_sRGBColor}
   Level=Constant()
   Apply View MipBias
  ) {mipValueMode=TMVM_MipLevel}
 ]
}
