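# UI-domain translucent material for a heart icon, apparently the favorite marker in the FX browser utilities.
# It holds a constant white color and a linear-color sample of a heart texture named as a single-channel distance field.
# The graph shows no links from these nodes to material outputs and no thresholding math for the distance field.
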Material /FXBrowserUtilities/M_FavHeart
graph {
 properties {bAutomaticallySetUsageInEditor=false, blendMode=BLEND_Translucent, materialDomain=MD_UI, usedShadingModels=Default Lit}
 nodes [
  Constant3Vector() {constant=<rgba 1 1 1 0>}
  TextureSample() {samplerType=SAMPLERTYPE_LinearColor, texture=&/heart_SingleChannelDF}
 ]
}
