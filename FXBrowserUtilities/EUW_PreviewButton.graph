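# Editor utility list-entry widget for the FX Browser: a clickable preview tile for one effect item, with a favorite heart.
# On list item set it collapses hidden items and builds a dynamic preview material from the loaded image and SubUV grid, then restores favorite color.
# Clicking the tile calls ShowVideo on the parent browser; the heart adds or deletes a per-user favorite tag.
# System users also get NeedCheck and NeedCaptureAgain tag toggles with color-coded labels.
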EditorUtilityWidgetBlueprint /FXBrowserUtilities/EUW_PreviewButton
graphs [
 EventGraph {nodes [
   "On Clicked (EditorUtilityButton_225)"() {is=title K2Node_ComponentBoundEvent}
   n1 = Get MyItemObject() -> (MyItemObject:EUO Preview Item Object Object Reference)
   > n2 = Branch(Condition=Get Fav!(self:EUO Preview Item Object Object Reference=@n1) -> (Fav:bool))
   n3 = Append(
    A=Get CurrentUser!(self:BUW FXBrowser Object Reference=Get ParentWidget!(self:EUO Preview Item Object Object Reference=@n1) -> (ParentWidget:BUW FXBrowser Object Reference)) -> (CurrentUser:String)
    B=.Fav
   )
   > DeleteATag!(execute, self:EUO Preview Item Object Object Reference=@n1, Tag:String=Favmark114514, SpecialUser:String=@n3) -> (then)
   > Set Fav!(execute, Fav:bool=false, self:EUO Preview Item Object Object Reference=@n1) -> (then, Output_Get:bool)
   n4 = Get EditorUtilityButton_225() -> (EditorUtilityButton_225:Editor Utility Button Object Reference)
   > SetColorAndOpacity(self:Button Object Reference=@n4, InColorAndOpacity:Linear Color Structure=(R=1,G=1,B=1,A=0.5))
   CreateNewTag!(execute=@n2.else, self:EUO Preview Item Object Object Reference=@n1, Tag:String=Favmark114514, SpecialUser:String=@n3) -> (then)
   > Set Fav(Fav=true, self=@n1)
   > SetColorAndOpacity(self=@n4, InColorAndOpacity=(R=1,G=0,B=0,A=1))
   "On Clicked (EditorUtilityButton_46)"() {is=title K2Node_ComponentBoundEvent}
   n5 = Get MyItemObject()
   > ShowVideo!(
    execute
    self:BUW FXBrowser Object Reference=Get ParentWidget(self=@n5)
    Folder:String=Get FolderPath!(self:EUO Preview Item Object Object Reference=@n5) -> (FolderPath:String)
    SourceButton:EUW Preview Button Object Reference=Self-Reference()
    LoadRuler:String=Get LoadRuler!(self:EUO Preview Item Object Object Reference=@n5) -> (LoadRuler:String)
   ) -> (then)
   "On Clicked (NeedCheckBt)"() {is=title K2Node_ComponentBoundEvent}
   n6 = Get MyItemObject()
   n7 = Get ParentWidget(self=@n6)
   n8 = Break St FXB Item!(St_FXB_Item:"St FXB Item Structure (by ref)"="Get (a ref)"(
     Array:Array of St FXB Item Structures=Get CurrentPendingLoadItemList!(self:BUW FXBrowser Object Reference=@n7) -> (CurrentPendingLoadItemList:Array of St FXB Item Structures)
     Dimension 1=Get MyIndex!(self:EUO Preview Item Object Object Reference=@n6) -> (MyIndex:int=0)
    ) -> (Output:"St FXB Item Structure (by ref)")) -> (
    AssetDataPath_12_#1:String
    EditorName_14_#2:String
    PublishName_16_#3:String
    PublishVersion_20_#4:int
    HasValidFX_22_#5:bool=~
    X_25_#6:int
    Y_27_#7:int
    VideoList_31_#8:Array of Strings
    ColorList_39_#9:Array of St FXB Color Array Structures
    Depend_38_#10:String
    Category_48_#11:String
    FolderPath_50_#12:String
    Tags_54_#13:Array of Strings
    MyIndexInDatabase_58_#14:int
    MyCategoryIndexInDataBase_61_#15:int
   ) {is=title K2Node_BreakStruct}
   > n9 = Branch(Condition="integer >= integer"(A=Find Item!(TargetArray=@n8.Tags_54_#13, ItemToFind:"String (by ref)"=NeedCheck) -> (ReturnValue)))
   > DeleteATag(self=@n1, Tag=NeedCheck)
   n10 = Make Slate Color(SpecifiedColor=(R=1,G=0,B=1,A=1))
   n11 = Get NCT() -> (NCT:Text Object Reference)
   > Set ColorAndOpacity!(execute, ColorAndOpacity:Slate Color Structure=@n10, self:Text Object Reference=@n11) -> (then, Output_Get:Slate Color Structure)
   CreateNewTag(execute=@n9.else, self=@n1, Tag=NeedCheck)
   n12 = Make Slate Color(SpecifiedColor=(R=0.0305,G=1,B=0,A=1))
   > Set ColorAndOpacity(ColorAndOpacity=@n12, self=@n11)
   "On Clicked (NeedCaptureAgainBt)"() {is=title K2Node_ComponentBoundEvent}
   > n13 = Branch(Condition="integer >= integer"(A=Find Item(TargetArray=@n8.Tags_54_#13, ItemToFind=NeedCaptureAgain)))
   > DeleteATag(self=@n1, Tag=NeedCaptureAgain)
   n14 = Get NCAT() -> (NCAT:Text Object Reference)
   > Set ColorAndOpacity(ColorAndOpacity=@n10, self=@n14)
   CreateNewTag(execute=@n13.else, self=@n1, Tag=NeedCaptureAgain)
   > Set ColorAndOpacity(ColorAndOpacity=@n12, self=@n14)
   n15 = SwitchButtonBoard!() -> (OutputDelegate, then, ShowBoard:bool=~) {is=title K2Node_CustomEvent}
   n16 = Make Slate Color(SpecifiedColor=(R=0.0395,G=0.0395,B=0.0395,A=1))
   n17 = Make Slate Color(SpecifiedColor=(R=0,G=0.7117,B=1,A=1))
   > Set members in Button Style!(
    execute:Exec
    StructRef:"Button Style Structure (by ref)"=Get WidgetStyle!(self:Button Object Reference=Get EditorUtilityButton_46() -> (EditorUtilityButton_46:Editor Utility Button Object Reference)) -> (WidgetStyle:Button Style Structure)
    Normal:Slate Brush Structure=Select(
     Option 0:Slate Brush Structure=Make Slate Brush!(
      TintColor:Slate Color Structure=@n16
      DrawAs:ESlateBrushDrawType Enum=Image
      Tiling:ESlateBrushTileType Enum=NoTile
      Mirroring:ESlateBrushMirrorType Enum=NoMirror
      ImageSize:Vector 2D Structure=(X=32,Y=32)
      Margin:Margin Structure=(Left=0,Top=0,Right=0,Bottom=0)
      ResourceObject:Object Reference
      OutlineSettings:Slate Brush Outline Settings Structure=(CornerRadii=(X=0,Y=0,Z=0,W=0),Color=(SpecifiedColor=(R=0,G=0,B=0,A=0),ColorUseRule=UseColor_Specified),Width=0,RoundingType=HalfHeightRadius,bUseBrushTransparency=False)
     ) -> (SlateBrush:Slate Brush Structure) {is=title K2Node_MakeStruct}
     Option 1:Slate Brush Structure=Make Slate Brush(
      TintColor=@n17
      DrawAs=Image
      Tiling=NoTile
      Mirroring=NoMirror
      ImageSize=(X=32,Y=32)
      Margin=(Left=0,Top=0,Right=0,Bottom=0)
      OutlineSettings=(CornerRadii=(X=0,Y=0,Z=0,W=0),Color=(SpecifiedColor=(R=0,G=0,B=0,A=0),ColorUseRule=UseColor_Specified),Width=0,RoundingType=HalfHeightRadius,bUseBrushTransparency=False)
     )
     Index=@n15.ShowBoard
    ) -> (ReturnValue:Slate Brush Structure)
   ) -> (then:Exec, StructOut:"Button Style Structure (by ref)") {is=title K2Node_SetFieldsInStruct}
   n18 = Event OnListItemObjectSet()
   > n19 = Cast To EUO_PreviewItemObject!(execute, Object=@n18.ListItemObject) -> (then, CastFailed, AsEUO Preview Item Object:EUO Preview Item Object Object Reference) {is=title K2Node_DynamicCast}
   > n20 = Sequence()
   > Branch(Condition=Get NeedHide!(self:EUO Preview Item Object Object Reference=@n19) -> (NeedHide:bool))
   > SetVisibility(self=Self-Reference(), InVisibility=Collapsed)
   n21 = Set MyItemObject(execute=@n20.then_1, MyItemObject:EUO Preview Item Object Object Reference=@n19) -> (Output_Get:EUO Preview Item Object Object Reference)
   > n22 = CreateDynamicMaterialInstance(Parent=/FXBrowserUtilities/FXRefLib/Finder/T_ShowPreview, OptionalName=None, CreationFlags=None)
   n23 = Get SubUVResX!(self:EUO Preview Item Object Object Reference=@n19) -> (SubUVResX:int=0)
   n24 = Get SubUVResY!(self:EUO Preview Item Object Object Reference=@n19) -> (SubUVResY:int=0)
   > SetVectorParameterValue(
    self=@n22
    ParameterName=S
    Value=MakeColor!(R:float="To Float (Integer)"(InInt=@n23), G:float="To Float (Integer)"(InInt=@n24), B:float=0, A:float=1) -> (ReturnValue:Linear Color Structure)
   )
   > SetTextureParameterValue(
    self=@n22
    ParameterName=T
    Value=Resolve Soft Reference!(Input:Texture 2D Soft Object Reference=Get LoadedImage!(self:EUO Preview Item Object Object Reference=@n19) -> (LoadedImage:Texture 2D Soft Object Reference)) -> (Output:Texture 2D Object Reference)
   )
   > SetBrushFromMaterial(self=Get Image_70() -> (Image_70:Image Object Reference), Material=@n22)
   > Set MyPreviewMat(MyPreviewMat:Material Instance Dynamic Object Reference=@n22) -> (Output_Get:Material Instance Dynamic Object Reference)
   > n25 = Sequence()
   > n26 = Branch(Condition=Get Fav(self=@n21))
   n27 = Get EditorUtilityButton_225()
   > SetColorAndOpacity(self=@n27, InColorAndOpacity=(R=1,G=0,B=0,A=1))
   SetColorAndOpacity(execute=@n26.else, self=@n27, InColorAndOpacity=(R=1,G=1,B=1,A=0.5))
   n28 = Get ParentWidget(self=@n21)
   Branch(execute=@n25.then_1, Condition=Get SystemUser!(self:BUW FXBrowser Object Reference=@n28) -> (SystemUser:bool))
   > SetVisibility(
    self=[
     Get NeedCaptureAgainBt() -> (NeedCaptureAgainBt:Editor Utility Button Object Reference)
     Get NeedCheckBt() -> (NeedCheckBt:Editor Utility Button Object Reference)
    ]
    InVisibility=Visible
   )
   n29 = Make Slate Color(SpecifiedColor=(R=1,G=0,B=1,A=1))
   n30 = Make Slate Color(SpecifiedColor=(R=0.0305,G=1,B=0,A=1))
   n31 = Break St FXB Item(St_FXB_Item="Get (a ref)"(Array=Get CurrentPendingLoadItemList(self=@n28), Dimension 1=Get MyIndex(self=@n21) -> (MyIndex=0))) -> (HasValidFX_22_#5=~)
   n32 = Find Item(TargetArray=@n31.Tags_54_#13, ItemToFind=NeedCheck)
   > Set ColorAndOpacity(
    ColorAndOpacity=Select(Option 0:Slate Color Structure=@n29, Option 1:Slate Color Structure=@n30, Index="integer >= integer"(A=@n32)) -> (ReturnValue:Slate Color Structure)
    self=Get NCT()
   )
   n33 = Find Item(TargetArray=@n31.Tags_54_#13, ItemToFind=NeedCaptureAgain)
   > Set ColorAndOpacity(ColorAndOpacity=Select(Option 0=@n29, Option 1=@n30, Index="integer >= integer"(A=@n33)), self=Get NCAT())
  ]}
 GetBrush {nodes [
   GetBrush() {is=title K2Node_FunctionEntry}
   > Return Node!(
    execute
    ReturnValue:Slate Brush Structure=Make Slate Brush(
     TintColor=(SpecifiedColor=(R=1,G=1,B=1,A=1),ColorUseRule=UseColor_Specified)
     DrawAs=Image
     Tiling=NoTile
     Mirroring=NoMirror
     ImageSize=(X=32,Y=32)
     Margin=(Left=0,Top=0,Right=0,Bottom=0)
     OutlineSettings=(CornerRadii=(X=0,Y=0,Z=0,W=0),Color=(SpecifiedColor=(R=0,G=0,B=0,A=0),ColorUseRule=UseColor_Specified),Width=0,RoundingType=HalfHeightRadius,bUseBrushTransparency=False)
    )
   )
  ]}
 GetVisibility_0 {nodes [
   GetVisibility_0() {is=title K2Node_FunctionEntry}
   > Return Node(ReturnValue:ESlateVisibility Enum=Select(Option 0:ESlateVisibility Enum=Collapsed, Option 1:ESlateVisibility Enum=Visible, Index=Get bDebug() -> (bDebug:bool)) -> (ReturnValue:ESlateVisibility Enum))
  ]}
 GetVisibility_1 {nodes [
   GetVisibility_1() {is=title K2Node_FunctionEntry}
   > Return Node(ReturnValue=Select(Option 0=Collapsed, Option 1=SelfHitTestInvisible, Index=Get bDebug()))
  ]}
]
variables [
 bDebug {type=boolean, default=true}
 MyItemObject {type=EUO_PreviewItemObject_C, default=null}
 MyPreviewMat {type=MaterialInstanceDynamic}
]
widgets {
 info {rootWidgetClass=&Overlay, widgetCount=11}
 widgets [
  Overlay_0 {widgetClassPath=&Overlay, widget=Overlay_0}
  EditorUtilityButton_46 {
   widgetClassPath=&EditorUtilityButton
   parent=Overlay_0
   widget=EditorUtilityButton_46
   properties {widgetStyle {
     clickedSlateSound {resourceObject=null}
     disabled {
      bIsDynamicallyLoaded=false
      drawAs=RoundedBox
      imageType=NoImage
      mirroring=NoMirror
      resourceName=None
      resourceObject=null
      tiling=NoTile
      imageSize=<0 0>
      margin=<ltrb 0 0 0 0>
      outlineSettings {
       bUseBrushTransparency=false
       roundingType=FixedRadius
       width=1
       color {colorUseRule=UseColor_ColorTable, specifiedColor=<rgba 0 0 0 0>}
       cornerRadii=<4 4 4 4>
      }
      tintColor {colorUseRule=UseColor_Specified, specifiedColor=<rgba 0.0395 0.0395 0.0395 1>}
      uVRegion {bIsValid=false, max=<0 0>, min=<0 0>}
     }
     disabledForeground {colorUseRule=UseColor_ColorTable, specifiedColor=<rgba 0 0 0 0>}
     hovered {
      bIsDynamicallyLoaded=false
      drawAs=RoundedBox
      imageType=NoImage
      mirroring=NoMirror
      resourceName=None
      resourceObject=null
      tiling=NoTile
      imageSize=<32 32>
      margin=<ltrb 0 0 0 0>
      outlineSettings {
       bUseBrushTransparency=false
       roundingType=FixedRadius
       width=1
       color {colorUseRule=UseColor_ColorTable, specifiedColor=<rgba 0 0 0 0>}
       cornerRadii=<4 4 4 4>
      }
      tintColor {colorUseRule=UseColor_Specified, specifiedColor=<rgba 1 0 0.104 1>}
      uVRegion {bIsValid=false, max=<0 0>, min=<0 0>}
     }
     hoveredForeground {colorUseRule=UseColor_ColorTable, specifiedColor=<rgba 0 0 0 0>}
     hoveredSlateSound {resourceObject=null}
     normal {
      bIsDynamicallyLoaded=false
      drawAs=RoundedBox
      imageType=NoImage
      mirroring=NoMirror
      resourceName=None
      resourceObject=null
      tiling=NoTile
      imageSize=<32 32>
      margin=<ltrb 0 0 0 0>
      outlineSettings {
       bUseBrushTransparency=false
       roundingType=FixedRadius
       width=1
       color {colorUseRule=UseColor_ColorTable, specifiedColor=<rgba 0 0 0 0>}
       cornerRadii=<4 4 4 4>
      }
      tintColor {colorUseRule=UseColor_Specified, specifiedColor=<rgba 0.0395 0.0395 0.0395 1>}
      uVRegion {bIsValid=false, max=<0 0>, min=<0 0>}
     }
     normalForeground {colorUseRule=UseColor_ColorTable, specifiedColor=<rgba 0 0 0 0>}
     normalPadding=<ltrb 0 0 0 0>
     pressed {
      bIsDynamicallyLoaded=false
      drawAs=RoundedBox
      imageType=NoImage
      mirroring=NoMirror
      resourceName=None
      resourceObject=null
      tiling=NoTile
      imageSize=<32 32>
      margin=<ltrb 0 0 0 0>
      outlineSettings {
       bUseBrushTransparency=false
       roundingType=FixedRadius
       width=1
       color {colorUseRule=UseColor_ColorTable, specifiedColor=<rgba 0 0 0 0>}
       cornerRadii=<4 4 4 4>
      }
      tintColor {colorUseRule=UseColor_Specified, specifiedColor=<rgba 0.257 0 0.0021 1>}
      uVRegion {bIsValid=false, max=<0 0>, min=<0 0>}
     }
     pressedForeground {colorUseRule=UseColor_ColorTable, specifiedColor=<rgba 0 0 0 0>}
     pressedPadding=<ltrb 0 0 0 0>
     pressedSlateSound {resourceObject=null}
    }}
   slot {class=OverlaySlot, properties {horizontalAlignment=HAlign_Fill, verticalAlignment=VAlign_Fill}}
  }
  VerticalBox_0 {
   widgetClassPath=&VerticalBox
   parent=EditorUtilityButton_46
   widget=VerticalBox_0
   slot {class=ButtonSlot, properties {horizontalAlignment=HAlign_Fill, verticalAlignment=VAlign_Fill, padding=<ltrb 4 4 4 4>}}
  }
  Image_70 {
   widgetClassPath=&Image
   bIsVariable=true
   parent=VerticalBox_0
   widget=Image_70
   slot {class=VerticalBoxSlot, properties {size {sizeRule=Fill, value=1}}}
  }
  SizeBox_0 {
   widgetClassPath=&SizeBox
   parent=Overlay_0
   widget=SizeBox_0
   properties {bOverride_HeightOverride=true, bOverride_WidthOverride=true, heightOverride=15, widthOverride=15}
   slot {class=OverlaySlot, properties {padding=<ltrb 5 5 0 0>}}
  }
  EditorUtilityButton_225 {
   widgetClassPath=&EditorUtilityButton
   parent=SizeBox_0
   widget=EditorUtilityButton_225
   properties {
    backgroundColor=<rgba 1 1 1 0>
    colorAndOpacity=<rgba 1 1 1 0.5>
    widgetStyle {
     clickedSlateSound {resourceObject=null}
     disabled {
      bIsDynamicallyLoaded=false
      drawAs=RoundedBox
      imageType=NoImage
      mirroring=NoMirror
      resourceName=None
      resourceObject=null
      tiling=NoTile
      imageSize=<0 0>
      margin=<ltrb 0 0 0 0>
      outlineSettings {
       bUseBrushTransparency=false
       roundingType=FixedRadius
       width=1
       color {colorUseRule=UseColor_ColorTable, specifiedColor=<rgba 0 0 0 0>}
       cornerRadii=<4 4 4 4>
      }
      tintColor {colorUseRule=UseColor_Specified, specifiedColor=<rgba 0.0395 0.0395 0.0395 0>}
      uVRegion {bIsValid=false, max=<0 0>, min=<0 0>}
     }
     disabledForeground {colorUseRule=UseColor_ColorTable, specifiedColor=<rgba 0 0 0 0>}
     hovered {
      bIsDynamicallyLoaded=false
      drawAs=RoundedBox
      imageType=NoImage
      mirroring=NoMirror
      resourceName=None
      resourceObject=null
      tiling=NoTile
      imageSize=<32 32>
      margin=<ltrb 0 0 0 0>
      outlineSettings {
       bUseBrushTransparency=false
       roundingType=FixedRadius
       width=1
       color {colorUseRule=UseColor_ColorTable, specifiedColor=<rgba 0 0 0 0>}
       cornerRadii=<4 4 4 4>
      }
      tintColor {colorUseRule=UseColor_Specified, specifiedColor=<rgba 0.0953 0.0953 0.0953 0>}
      uVRegion {bIsValid=false, max=<0 0>, min=<0 0>}
     }
     hoveredForeground {colorUseRule=UseColor_ColorTable, specifiedColor=<rgba 0 0 0 0>}
     hoveredSlateSound {resourceObject=null}
     normal {
      bIsDynamicallyLoaded=false
      drawAs=RoundedBox
      imageType=NoImage
      mirroring=NoMirror
      resourceName=None
      resourceObject=null
      tiling=NoTile
      imageSize=<32 32>
      margin=<ltrb 0 0 0 0>
      outlineSettings {
       bUseBrushTransparency=false
       roundingType=FixedRadius
       width=1
       color {colorUseRule=UseColor_ColorTable, specifiedColor=<rgba 0 0 0 0>}
       cornerRadii=<4 4 4 4>
      }
      tintColor {colorUseRule=UseColor_Specified, specifiedColor=<rgba 0.0395 0.0395 0.0395 0>}
      uVRegion {bIsValid=false, max=<0 0>, min=<0 0>}
     }
     normalForeground {colorUseRule=UseColor_ColorTable, specifiedColor=<rgba 0 0 0 0>}
     normalPadding=<ltrb 0 0 0 0>
     pressed {
      bIsDynamicallyLoaded=false
      drawAs=RoundedBox
      imageType=NoImage
      mirroring=NoMirror
      resourceName=None
      resourceObject=null
      tiling=NoTile
      imageSize=<32 32>
      margin=<ltrb 0 0 0 0>
      outlineSettings {
       bUseBrushTransparency=false
       roundingType=FixedRadius
       width=1
       color {colorUseRule=UseColor_ColorTable, specifiedColor=<rgba 0 0 0 0>}
       cornerRadii=<0 0 0 0>
      }
      tintColor {colorUseRule=UseColor_Specified, specifiedColor=<rgba 0.0284 0.0284 0.0284 0>}
      uVRegion {bIsValid=false, max=<0 0>, min=<0 0>}
     }
     pressedForeground {colorUseRule=UseColor_ColorTable, specifiedColor=<rgba 0 0 0 0>}
     pressedPadding=<ltrb 0 0 0 0>
     pressedSlateSound {resourceObject=null}
    }
   }
   slot {class=SizeBoxSlot}
  }
  Image_47 {
   widgetClassPath=&Image
   bIsVariable=true
   parent=EditorUtilityButton_225
   widget=Image_47
   properties {brush {
     bIsDynamicallyLoaded=false
     drawAs=Image
     imageType=FullColor
     mirroring=NoMirror
     resourceName=None
     resourceObject=&/M_FavHeart
     tiling=NoTile
     imageSize=<64 64>
     margin=<ltrb 0 0 0 0>
     outlineSettings {
      bUseBrushTransparency=false
      roundingType=HalfHeightRadius
      width=0
      color {colorUseRule=UseColor_Specified, specifiedColor=<rgba 0 0 0 0>}
      cornerRadii=<0 0 0 0>
     }
     tintColor {colorUseRule=UseColor_Specified, specifiedColor=<rgba 1 1 1 1>}
     uVRegion {bIsValid=false, max=<0 0>, min=<0 0>}
    }}
   slot {class=ButtonSlot, properties {horizontalAlignment=HAlign_Fill, verticalAlignment=VAlign_Fill, padding=<ltrb 0 0 0 0>}}
  }
  NeedCheckBt {
   widgetClassPath=&EditorUtilityButton
   parent=Overlay_0
   widget=NeedCheckBt
   properties {visibility=Collapsed}
   slot {class=OverlaySlot, properties {verticalAlignment=VAlign_Bottom}}
  }
  NCT {widgetClassPath=&TextBlock, bIsVariable=true, parent=NeedCheckBt, widget=NCT, properties {text=NC}, slot {class=ButtonSlot}}
  NeedCaptureAgainBt {
   widgetClassPath=&EditorUtilityButton
   parent=Overlay_0
   widget=NeedCaptureAgainBt
   properties {visibility=Collapsed}
   slot {class=OverlaySlot, properties {horizontalAlignment=HAlign_Right, verticalAlignment=VAlign_Bottom}}
  }
  NCAT {
   widgetClassPath=&TextBlock
   bIsVariable=true
   parent=NeedCaptureAgainBt
   widget=NCAT
   properties {text=NCA}
   slot {class=ButtonSlot}
  }
 ]
}
