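# Editor utility list-entry widget that shows one tag as a toggle button in the FX Browser.
# When assigned a list item, it casts it to EUO_TagButtonWObject, copies the tag, toggle state and parent browser, then sets the label and color.
# Clicking calls UI_ChangeTag on the parent BUW_FXBrowser with the opposite state, then flips Has and recolors the button green for on, pale blue for off.
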EditorUtilityWidgetBlueprint /FXBrowserUtilities/EUW_TagButtonW
graphs [
 EventGraph {nodes [
   "On Clicked (EditorUtilityButton_0)"() {is=title K2Node_ComponentBoundEvent}
   > n1 = Branch(Condition=Get Has() -> (Has:bool))
   > UI_ChangeTag!(
    execute
    self:BUW FXBrowser Object Reference=Get Parent() -> (Parent:BUW FXBrowser Object Reference)
    Tag:String=Get Tag() -> (Tag:String)
    Check:bool=false
   ) -> (then)
   > SetBackgroundColor!(
    execute
    self:Button Object Reference=Get EditorUtilityButton_0() -> (EditorUtilityButton_0:Editor Utility Button Object Reference)
    InBackgroundColor:Linear Color Structure=(R=0.8153,G=0.8594,B=1,A=1)
   ) -> (then)
   > Set Has(Has:bool=false) -> (Output_Get:bool)
   UI_ChangeTag(execute=@n1.else, self=Get Parent(), Tag=Get Tag(), Check=true)
   > SetBackgroundColor(self=Get EditorUtilityButton_0(), InBackgroundColor=(R=0.3513,G=1,B=0,A=1))
   > Set Has(Has=true)
   n2 = Event OnListItemObjectSet()
   > n3 = Cast To EUO_TagButtonWObject!(execute, Object=@n2.ListItemObject) -> (then, CastFailed, AsEUO Tag Button WObject:EUO Tag Button WObject Object Reference) {is=title K2Node_DynamicCast}
   > Set Tag(Tag:String=Get T!(self:EUO Tag Button WObject Object Reference=@n3) -> (T:String)) -> (Output_Get:String)
   > Set Has(Has=Get H!(self:EUO Tag Button WObject Object Reference=@n3) -> (H:bool))
   > Set Parent(Parent:BUW FXBrowser Object Reference=Get ParentWidget!(self:EUO Tag Button WObject Object Reference=@n3) -> (ParentWidget:BUW FXBrowser Object Reference)) -> (Output_Get:BUW FXBrowser Object Reference)
   > Set Text(Text="To Text (String)"(InString=Get Tag()), self=Get ItemNameBox_3() -> (ItemNameBox_3:Text Object Reference))
   > n4 = Branch(Condition=Get Has())
   n5 = Get EditorUtilityButton_0()
   > SetBackgroundColor(self=@n5, InBackgroundColor=(R=0.3513,G=1,B=0,A=1))
   SetBackgroundColor(execute=@n4.else, self=@n5, InBackgroundColor=(R=0.8153,G=0.8594,B=1,A=1))
  ]}
 GetBrush {nodes [
   GetBrush() {is=title K2Node_FunctionEntry}
   > Return Node!(
    execute
    ReturnValue:Slate Brush Structure=Make Slate Brush!(
     TintColor:Slate Color Structure=(SpecifiedColor=(R=1,G=1,B=1,A=1),ColorUseRule=UseColor_Specified)
     DrawAs:ESlateBrushDrawType Enum=Image
     Tiling:ESlateBrushTileType Enum=NoTile
     Mirroring:ESlateBrushMirrorType Enum=NoMirror
     ImageSize:Vector 2D Structure=(X=32,Y=32)
     Margin:Margin Structure=(Left=0,Top=0,Right=0,Bottom=0)
     ResourceObject:Object Reference
     OutlineSettings:Slate Brush Outline Settings Structure=(CornerRadii=(X=0,Y=0,Z=0,W=0),Color=(SpecifiedColor=(R=0,G=0,B=0,A=0),ColorUseRule=UseColor_Specified),Width=0,RoundingType=HalfHeightRadius,bUseBrushTransparency=False)
    ) -> (SlateBrush:Slate Brush Structure) {is=title K2Node_MakeStruct}
   )
  ]}
]
variables [
 Has {type=boolean}
 Parent {type=BUW_FXBrowser_C, default=null}
 Tag {type=string, default=""}
]
widgets {
 info {rootWidgetClass=&EditorUtilityButton, widgetCount=3}
 widgets [
  EditorUtilityButton_0 {
   widgetClassPath=&EditorUtilityButton
   widget=EditorUtilityButton_0
   properties {backgroundColor=<rgba 0.8153 0.8594 1 1>}
   slot=~
  }
  HorizontalBox_0 {
   widgetClassPath=&HorizontalBox
   parent=EditorUtilityButton_0
   widget=HorizontalBox_0
   slot {class=ButtonSlot, properties {horizontalAlignment=HAlign_Fill, verticalAlignment=VAlign_Fill, padding=<ltrb 0 0 0 0>}}
  }
  ItemNameBox_3 {
   widgetClassPath=&TextBlock
   bIsVariable=true
   parent=HorizontalBox_0
   widget=ItemNameBox_3
   properties {
    text=AddTags
    font {
     bForceMonospaced=false
     bMaterialIsStencil=false
     fontMaterial=null
     fontObject=/Engine/EngineFonts/RobotoTiny
     letterSpacing=0
     monospacedWidth=1
     size=9
     skewAmount=0
     typefaceFontName=Light
     outlineSettings {
      bApplyOutlineToDropShadows=false
      bMiteredCorners=false
      bSeparateFillAlpha=false
      outlineMaterial=null
      outlineSize=0
      outlineColor=<rgba 0 0 0 1>
     }
    }
   }
   slot {class=HorizontalBoxSlot, properties {verticalAlignment=VAlign_Center, size {sizeRule=Fill, value=1}}}
  }
 ]
}
