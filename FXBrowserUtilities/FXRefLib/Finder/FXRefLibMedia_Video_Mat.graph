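# Unlit UI-domain material that presents the FXRefLibMedia_Video texture, apparently for video display in the FX Reference Library finder of FXBrowserUtilities.
# Its graph holds a single texture sample of that asset, with no other nodes, parameters or visible output links.
# The texture's name suggests it is a media texture fed by video playback.
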Material /FXBrowserUtilities/FXRefLib/Finder/FXRefLibMedia_Video_Mat
graph {
 properties {materialDomain=MD_UI, pixelDepthOffsetMode=PDOM_Legacy, shadingModel=MSM_Unlit, usedShadingModels=Unlit}
 nodes [TextureSample() {texture=&/FXRefLibMedia_Video}]
}
