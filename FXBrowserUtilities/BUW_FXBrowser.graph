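# Editor utility widget that browses an FX library with tags, tiled previews and video playback.
# DB_BuildDataBase parses category, item, color, dependency, media and tag files under the library path into the FXBrowser SaveGame.
# DB_BuildTag diffs tag files on disk against the saved copy, applying changes through CreateNewTag and DeleteATag, which update both tag-to-item and item-to-tag data.
# EventGraph creates preview objects for a TileView and drives a MediaPlayer.
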EditorUtilityWidgetBlueprint /FXBrowserUtilities/BUW_FXBrowser
graphs [
 BuildCategoryPath {nodes [BuildCategoryPath() {is=title K2Node_FunctionEntry}]}
 BuildItemPath {nodes [BuildItemPath() {is=title K2Node_FunctionEntry}]}
 CompareTwoTagData {nodes [
   n1 = CompareTwoTagData!() -> (
    then
    ATagList:Array of Strings
    BTagList:Array of Strings
    AItemList:Array of St FXB String Array Structures
    BItemList:Array of St FXB String Array Structures
    In_PendingTagList:Array of Strings
    In_PendingItemList:Array of St FXB String Array Structures
   ) {is=title K2Node_FunctionEntry}
   > n2 = For Each Loop(Array=@n1)
   Return Node(execute=@n2.Completed)
   n3 = Find Item!(TargetArray=@n1.BTagList, ItemToFind:"String (by ref)"=@n2) -> (ReturnValue)
   > n4 = Branch(Condition="integer >= integer"(A=@n3))
   n5 = "Get (a ref)"(Array:Array of St FXB String Array Structures=@n1.AItemList, Dimension 1=@n2.Array Index) -> (Output:"St FXB String Array Structure (by ref)")
   > n6 = For Each Loop(Array=Break St FXB String Array!(St_FXB_StringArray:"St FXB String Array Structure (by ref)"=@n5) -> (String_3_#1:Array of Strings) {is=title K2Node_BreakStruct})
   n7 = Break St FXB String Array(St_FXB_StringArray="Get (a ref)"(Array=@n1.BItemList, Dimension 1=@n3))
   > n8 = Branch(Condition="integer >= integer"(A=Find Item(TargetArray=@n7, ItemToFind=@n6)))
   n9 = Get In_PendingTagList() -> (In_PendingTagList:Array of Strings)
   n10 = Find Item(TargetArray=@n9, ItemToFind=@n2)
   n11 = Branch(execute=@n8.else, Condition="integer >= integer"(A=@n10))
   n12 = Get In_PendingItemList() -> (In_PendingItemList:Array of St FXB String Array Structures)
   > Add Unique(TargetArray=Break St FXB String Array(St_FXB_StringArray="Get (a ref)"(Array=@n12, Dimension 1=@n10)), NewItem=@n6)
   Add(execute=@n11.else, TargetArray=@n9, NewItem=@n2) {is=title K2Node_CallArrayFunction}
   > Add(
    TargetArray:Array of St FXB String Array Structures=@n12
    NewItem:"St FXB String Array Structure (by ref)"=Make St FXB String Array!(String_3_#1:Array of Strings=Make Array("[0]":String=@n6) -> (Array:Array of Strings)) -> (St_FXB_StringArray:St FXB String Array Structure) {is=title K2Node_MakeStruct}
   )
   n13 = Get In_PendingTagList()
   n14 = Find Item(TargetArray=@n13, ItemToFind=@n2)
   n15 = Branch(execute=@n4.else, Condition="integer >= integer"(A=@n14))
   > n16 = For Each Loop(Array=Break St FXB String Array(St_FXB_StringArray=@n5))
   n17 = Get In_PendingItemList()
   > Add Unique(TargetArray=Break St FXB String Array(St_FXB_StringArray="Get (a ref)"(Array=@n17, Dimension 1=@n14)), NewItem=@n16)
   Add(execute=@n15.else, TargetArray:Array of Strings=@n13, NewItem:"String (by ref)"=@n2)
   > Add(TargetArray:Array of St FXB String Array Structures=@n17, NewItem:"St FXB String Array Structure (by ref)"=@n5)
  ]}
 CreateNewTag {nodes [
   n1 = Get DataBaseSaveGame!(self:BUW FXBrowser Object Reference=Self-Reference()) -> (DataBaseSaveGame:FXB Save Game Object Reference)
   SaveGameToSlot!(execute, SaveGameObject:Save Game Object Reference=@n1, SlotName:String=FXBrowser, UserIndex:int=0) -> (then, ReturnValue:bool)
   > UI_InitialTag(self=Self-Reference())
   n2 = CreateNewTag!() -> (then, Tag:String, ItemName:String) {is=title K2Node_FunctionEntry}
   n3 = Self-Reference()
   n4 = Find(
    TargetMap:Map of Strings to Int Vector 2 Structures=Get DataBase_CategoryIndexAndItemName!(self:FXB Save Game Object Reference=Get DataBaseSaveGame(self=@n3)) -> (DataBase_CategoryIndexAndItemName:Map of Strings to Int Vector 2 Structures)
    Key=@n2.ItemName
   ) -> (Value:Int Vector 2 Structure)
   > Branch(Condition=@n4.ReturnValue)
   n5 = Get DataBase_TagList!(self:FXB Save Game Object Reference=@n1) -> (DataBase_TagList:Array of Strings)
   n6 = Find Item(TargetArray=@n5, ItemToFind=@n2)
   > n7 = Branch(Condition="integer >= integer"(A=@n6))
   n8 = Get DataBase_Tag2ItemArray!(self:FXB Save Game Object Reference=@n1) -> (DataBase_Tag2ItemArray:Array of St FXB Tag 2Item Array Structures)
   n9 = "Get (a ref)"(Array:Array of St FXB Tag 2Item Array Structures=@n8, Dimension 1=@n6) -> (Output:"St FXB Tag 2Item Array Structure (by ref)")
   n10 = Break St FXB Tag 2Item Array!(St_FXB_Tag2ItemArray:"St FXB Tag 2Item Array Structure (by ref)"=@n9) -> (
    Tag2ItemList_3_#2:Array of St FXB Tag 2Item Structures
    ItemNameList_6_#3:Array of Strings
   ) {is=title K2Node_BreakStruct}
   > n11 = Set NewLocalVar(NewLocalVar:Array of St FXB Tag 2Item Structures=@n10) -> (Output_Get:Array of St FXB Tag 2Item Structures)
   > n12 = Set NewLocalVar_0(NewLocalVar_0:Array of Strings=@n10.ItemNameList_6_#3) -> (Output_Get:Array of Strings)
   n13 = Break Int Vector 2!(IntVector2:"Int Vector 2 Structure (by ref)"=@n4) -> (X:int, Y:int) {is=title K2Node_BreakStruct}
   n14 = "Get (a ref)"(
    Array:Array of St FXB Item List Structures=Get DataBase_ItemList!(self:FXB Save Game Object Reference=@n1) -> (DataBase_ItemList:Array of St FXB Item List Structures)
    Dimension 1=@n13
   ) -> (Output:"St FXB Item List Structure (by ref)")
   n15 = Break St FXB Item!(St_FXB_Item:"St FXB Item Structure (by ref)"="Get (a ref)"(
     Array:Array of St FXB Item Structures=Break St FXB Item List!(St_FXB_ItemList:"St FXB Item List Structure (by ref)"=@n14) -> (ItemList_16_#4:Array of St FXB Item Structures) {is=title K2Node_BreakStruct}
     Dimension 1=@n13.Y
    ) -> (Output:"St FXB Item Structure (by ref)")) -> (
    AssetDataPath_12_#5:String
    EditorName_14_#6:String
    PublishName_16_#7:String
    PublishVersion_20_#8:int
    HasValidFX_22_#9:bool=~
    X_25_#10:int
    Y_27_#11:int
    VideoList_31_#12:Array of Strings
    ColorList_39_#13:Array of St FXB Color Array Structures
    Depend_38_#14:String
    Category_48_#15:String
    FolderPath_50_#16:String
    Tags_54_#17:Array of Strings
    MyIndexInDatabase_58_#18:int
    MyCategoryIndexInDataBase_61_#19:int
   ) {is=title K2Node_BreakStruct}
   n16 = Make St FXB Tag 2Item!(
    ItemEditorName_3_#20:String=@n15.EditorName_14_#6
    ItemCategoryIndex_7_#21:int=@n15.MyCategoryIndexInDataBase_61_#19
    ItemIndex_9_#22:int=@n15.MyIndexInDatabase_58_#18
   ) -> (St_FXB_Tag2Item:St FXB Tag 2Item Structure) {is=title K2Node_MakeStruct}
   > Add(TargetArray:Array of St FXB Tag 2Item Structures=@n11, NewItem:"St FXB Tag 2Item Structure (by ref)"=@n16)
   > Add(TargetArray:Array of Strings=@n12, NewItem:"String (by ref)"=@n15.EditorName_14_#6)
   > n17 = "Set St FXB Tag 2Item Array Structure (by ref)"!(
    execute:Exec
    Target:"St FXB Tag 2Item Array Structure (by ref)"=@n9
    Value:St FXB Tag 2Item Array Structure=Make St FXB Tag 2Item Array!(
     Tag2ItemList_3_#2:Array of St FXB Tag 2Item Structures=@n11
     ItemNameList_6_#3:Array of Strings=@n12
    ) -> (St_FXB_Tag2ItemArray:St FXB Tag 2Item Array Structure) {is=title K2Node_MakeStruct}
   ) -> (then:Exec) {is=title K2Node_VariableSetRef}
   n18 = Self-Reference()
   n19 = "Get (a ref)"(
    Array:Array of St FXB Item List Structures=Get DataBase_ItemList(self=Get DataBaseSaveGame(self=@n18))
    Dimension 1=@n15.MyCategoryIndexInDataBase_61_#19
   ) -> (Output:"St FXB Item List Structure (by ref)")
   n20 = Break St FXB Item List(St_FXB_ItemList=@n19)
   n21 = "Get (a ref)"(Array:Array of St FXB Item Structures=@n20, Dimension 1=@n15.MyIndexInDatabase_58_#18) -> (Output:"St FXB Item Structure (by ref)")
   n22 = Set NewLocalVar_1(
    execute=[@n24.then, @n17.then]
    NewLocalVar_1:Array of Strings=Break St FXB Item(St_FXB_Item=@n21) -> (HasValidFX_22_#9=~).Tags_54_#17
   ) -> (Output_Get:Array of Strings)
   > Add Unique(TargetArray=@n22, NewItem=Get Tag() -> (Tag:String))
   > n23 = Set members in St FXB Item!(
    execute:Exec
    StructRef:"St FXB Item Structure (by ref)"=@n21
    Tags_54_#17:Array of Strings=Get NewLocalVar_1() -> (NewLocalVar_1:Array of Strings)
   ) -> (then:Exec, StructOut:"St FXB Item Structure (by ref)") {is=title K2Node_SetFieldsInStruct}
   > Set Array Elem!(
    execute
    TargetArray:Array of St FXB Item Structures=@n20
    Index:int=@n15.MyIndexInDatabase_58_#18
    Item:"St FXB Item Structure (by ref)"=@n23
    bSizeToFit:bool=false
   ) -> (then)
   > "Set St FXB Item List Structure (by ref)"!(
    execute:Exec
    Target:"St FXB Item List Structure (by ref)"=@n19
    Value:St FXB Item List Structure=Make St FXB Item List!(ItemList_16_#4:Array of St FXB Item Structures=@n20) -> (St_FXB_ItemList:St FXB Item List Structure) {is=title K2Node_MakeStruct}
   ) -> (then:Exec) {is=title K2Node_VariableSetRef}
   Add(execute=@n7.else, TargetArray=@n5, NewItem=@n2)
   > n24 = Add(
    TargetArray:Array of St FXB Tag 2Item Array Structures=@n8
    NewItem:"St FXB Tag 2Item Array Structure (by ref)"=Make St FXB Tag 2Item Array(
     Tag2ItemList_3_#2=Make Array("[0]":St FXB Tag 2Item Structure=@n16) -> (Array:Array of St FXB Tag 2Item Structures)
     ItemNameList_6_#3=Make Array("[0]":String=@n15.EditorName_14_#6) -> (Array:Array of Strings)
    )
   )
  ]}
 DB_BuildDataBase {nodes [
   n1 = Get SaveGame() -> (SaveGame:FXB Save Game Object Reference)
   n2 = Keys!(
    execute=@n25.then_2
    TargetMap:Map of Strings to Strings=Get DataBase_Category!(self:FXB Save Game Object Reference=@n1) -> (DataBase_Category:Map of Strings to Strings)
   ) -> (then, Keys:Array of Strings)
   > n3 = For Each Loop(Array=@n2)
   > n4 = GetFolderListInPath!(
    execute
    "Path(/end)":String=Append!(A=Get LibraryBasePath() -> (LibraryBasePath:String), B=@n3, C:String=/) -> (ReturnValue)
    __WorldContext:Object Reference
   ) -> (then, FolderList:Array of Strings, FullPathList:Array of Strings)
   n5 = For Each Loop(Exec=@n74.then, Array=@n4.FullPathList)
   n6 = GetFileListInPath!(execute=@n83.then, "Path(/end)":String=@n5, ext:String=jpg, __WorldContext:Object Reference) -> (then, FileList:Array of Strings, FullPathList:Array of Strings)
   n7 = GetFileListInPath(execute=@n73.then, "Path(/end)"=@n5, ext=mp4)
   n8 = IsEmpty(InString="Get (a copy)"(Array=@n7))
   n9 = For Each Loop(
    Exec=[@n6.then, @n83.else]
    Array=Select(Option 0:Array of Strings=@n7, Option 1:Array of Strings=@n6, Index=@n8) -> (ReturnValue:Array of Strings)
   )
   n10 = "Get (a copy)"(Array=Select(Option 0=@n7.FullPathList, Option 1=@n6.FullPathList, Index=@n8), Dimension 1=@n9.Array Index)
   n11 = Select(Option 0:String=.mp4, Option 1:String=.jpg, Index=@n8) -> (ReturnValue:String)
   n12 = Read File to String!(
    execute=@n75.then
    FilePath:String=Append!(A=Replace(SourceString=@n10, From=@n11), B=_Color.txt) -> (ReturnValue)
    bUseFullFilePath:bool=true
   ) -> (then, ResultOut:String, ReturnValue:bool)
   n13 = For Each Loop(Exec=@n76.then, Array=ParseIntoArray(SourceString=@n12, Delimiter="\r\n"))
   n14 = ParseIntoArray(SourceString=Replace(SourceString=Replace(SourceString=@n13, From="["), From="]"), Delimiter=" ")
   n15 = String To Float(InString="Get (a ref)"(Array:Array of Strings=@n14, Dimension 1=2) -> (Output:"String (by ref)"))
   n16 = String To Float(InString="Get (a ref)"(Array=@n14, Dimension 1=0))
   n17 = String To Float(InString="Get (a ref)"(Array=@n14, Dimension 1=1))
   n18 = "Get (a copy)"(Array=@n4, Dimension 1=@n5.Array Index)
   n19 = PrintString(InString=Append!(
     A="To String (Float)"(InDouble=@n16)
     B=+
     C:String="To String (Float)"(InDouble=@n17)
     D:String=+
     E:String="To String (Float)"(InDouble=@n15)
     F:String="=="
     G:String=Append!(A=@n18, B=+, C=@n9) -> (ReturnValue)
    ) -> (ReturnValue))
   Add(
    execute=[@n19.then, @n13.LoopBody]
    TargetArray:Array of Linear Color Structures=Get tmpColorList() -> (tmpColorList:Array of Linear Color Structures)
    NewItem:"Linear Color Structure (by ref)"=Make Linear Color!(R:float=float / float(A=@n15, B=255), G:float=float / float(A=@n17, B=255), B:float=float / float(A=@n16, B=255), A:float=1) -> (LinearColor:Linear Color Structure) {is=title K2Node_MakeStruct}
   )
   DB_BuildDataBase() {is=title K2Node_FunctionEntry}
   n20 = Reverse!(SourceString:String=Get LibraryBasePath()) -> (ReturnValue:String)
   > n21 = Branch(Condition="Equal (Integer)"(A=FindSubstring(SearchIn=@n20, Substring="\\")))
   > n22 = Read File to String(FilePath=Append!(A=Get LibraryBasePath(), B=CategoryInformation.txt) -> (ReturnValue), bUseFullFilePath=true)
   > n23 = Branch(Condition=@n22.ReturnValue)
   n24 = Construct Object from Class(execute=[@n44.then, @n23.then], Class=&/FXB_SaveGame.FXB_SaveGame_C) -> (ReturnValue:FXB Save Game Object Reference)
   > Set SaveGame(SaveGame:FXB Save Game Object Reference=@n24) -> (Output_Get:FXB Save Game Object Reference)
   > n25 = Sequence!(execute) -> (then_0, then_1, then_2:Exec, then_3:Exec, then_4:Exec)
   > Clear!(execute, TargetMap:Map of Strings to Strings=Get DataBase_Category(self=Get SaveGame())) -> (then) {is=title K2Node_CallFunction}
   > n26 = Read File to String(FilePath=Append(A=Get LibraryBasePath(), B=CategoryInformation.txt), bUseFullFilePath=true)
   > n27 = For Each Loop(Array=ParseIntoArray(SourceString=@n26, Delimiter="\r\n"))
   n28 = ParseIntoArray(SourceString=@n27, Delimiter="&&")
   > Add!(
    execute
    TargetMap:Map of Strings to Strings=Get DataBase_Category(self=Get SaveGame())
    Key:"String (by ref)"="Get (a copy)"(Array=@n28)
    Value:"String (by ref)"="Get (a copy)"(Array=@n28, Dimension 1=1)
   ) -> (then) {is=title K2Node_CallFunction}
   Clear(execute=@n25.then_1, TargetArray=Get tmpTagFileContentList() -> (tmpTagFileContentList:Array of Strings)) {is=title K2Node_CallArrayFunction}
   > n29 = GetFileListInPath("Path(/end)"=Append(A=Get LibraryBasePath(), B=Tag/), ext=txt, __WorldContext=Self-Reference())
   > Set tmpTagFileList(tmpTagFileList:Array of Strings=@n29) -> (Output_Get:Array of Strings)
   > n30 = For Each Loop(Array=@n29.FullPathList)
   > n31 = Read File to String(FilePath=@n30, bUseFullFilePath=true)
   > Add(TargetArray:Array of Strings=Get tmpTagFileContentList(), NewItem:"String (by ref)"=@n31) {is=title K2Node_CallArrayFunction}
   > n32 = For Each Loop(Array=ParseIntoArray(SourceString=@n31, Delimiter="\r\n"))
   n33 = Split(SourceString=@n32, InStr="%&&%", SearchDir=FromStart)
   n34 = Find Item(TargetArray=Get ReadedTagList() -> (ReadedTagList:Array of Strings), ItemToFind=@n33)
   > n35 = Branch(Condition="integer >= integer"(A=@n34))
   n36 = ParseIntoArray(SourceString=@n33.RightS, Delimiter="%@%")
   > n37 = For Each Loop(Array=@n36)
   n38 = Get ReadedTagItemList() -> (ReadedTagItemList:Array of St FXB String Array Structures)
   > Add Unique(
    TargetArray=Break St FXB String Array(St_FXB_StringArray="Get (a ref)"(Array:Array of St FXB String Array Structures=@n38, Dimension 1=@n34) -> (Output:"St FXB String Array Structure (by ref)"))
    NewItem=@n37
   )
   Add(execute=@n35.else, TargetArray=Get ReadedTagList(), NewItem=@n33)
   > Add(
    TargetArray:Array of St FXB String Array Structures=Get ReadedTagItemList()
    NewItem:"St FXB String Array Structure (by ref)"=Make St FXB String Array(String_3_#1=@n36)
   )
   n39 = Get SaveGame()
   Set DataBase_TagFiles!(
    execute=@n25.then_3
    DataBase_TagFiles:Array of Strings=Get tmpTagFileList() -> (tmpTagFileList:Array of Strings)
    self:FXB Save Game Object Reference=@n39
   ) -> (then, Output_Get:Array of Strings)
   > Set DataBase_TagStrings!(execute, DataBase_TagStrings:Array of Strings=Get tmpTagFileContentList(), self:FXB Save Game Object Reference=@n39) -> (then, Output_Get:Array of Strings)
   > n40 = Read File to String(FilePath=Append(A=Get LibraryBasePath(), B=Version.txt), bUseFullFilePath=true)
   > Branch(Condition=@n40.ReturnValue)
   > Set DataBase_Version!(execute, DataBase_Version:int=String To Integer(InString=@n40), self:FXB Save Game Object Reference=Get SaveGame()) -> (then, Output_Get:int=0)
   SaveGameToSlot(execute=@n25.then_4, SaveGameObject=@n1, SlotName=FXBrowser, UserIndex=0)
   n41 = Read File to String(execute=@n21.else, FilePath=Append(A=Get LibraryBasePath(), B="\\CategoryInformation.txt"), bUseFullFilePath=true)
   > Branch(Condition=@n41.ReturnValue)
   n42 = Append(A=Get LibraryBasePath(), B="\\")
   > Set LibraryBasePath(LibraryBasePath:String=@n42) -> (Output_Get:String)
   n43 = Get SettingSaveObject() -> (SettingSaveObject:FXB Save Game Path Object Reference)
   > Set LibPath!(execute, LibPath:String=@n42, self:FXB Save Game Path Object Reference=@n43) -> (then, Output_Get:String)
   > n44 = SaveGameToSlot(SaveGameObject=@n43, SlotName=FXBrowserSetting, UserIndex=0)
   n45 = Get tmpItemList() -> (tmpItemList:Array of St FXB Item Structures)
   Add(
    execute=@n78.then
    TargetArray:Array of St FXB Item List Structures=Get DataBase_ItemList(self=Get SaveGame())
    NewItem:"St FXB Item List Structure (by ref)"=Make St FXB Item List(ItemList_16_#4=@n45)
   )
   n46 = Length(TargetArray:Array of St FXB Item List Structures=Get DataBase_ItemList(self=Get SaveGame()))
   n47 = Read File to String(execute=@n79.then_1, FilePath=Append(A=@n5, B=Dependencies.txt), bUseFullFilePath=true)
   n48 = Read File to String(execute=@n80.then_0, FilePath=Append(A=@n5, B=BasicInformation.txt), bUseFullFilePath=true)
   > n49 = Set tmpBasicInformationArray(tmpBasicInformationArray:Array of Strings=ParseIntoArray(SourceString=@n48, Delimiter=",&&,", CullEmptyStrings=false)) -> (Output_Get:Array of Strings)
   n50 = "Get (a copy)"(Array=@n49, Dimension 1=1)
   n51 = Make St FXB Item!(
    AssetDataPath_12_#5:String="Get (a copy)"(Array=@n49)
    EditorName_14_#6:String=@n50
    PublishName_16_#7:String="Get (a copy)"(Array=@n49, Dimension 1=2)
    PublishVersion_20_#8:int=String To Integer(InString="Get (a copy)"(Array=@n49, Dimension 1=3))
    HasValidFX_22_#9:bool="Equal Exactly (String)"(A="Get (a copy)"(Array=@n49, Dimension 1=4), B=true)
    X_25_#10:int=String To Integer(InString="Get (a copy)"(Array=@n49, Dimension 1=5))
    Y_27_#11:int=String To Integer(InString="Get (a copy)"(Array=@n49, Dimension 1=6))
    VideoList_31_#12:Array of Strings=Get tmpVideoList() -> (tmpVideoList:Array of Strings)
    ColorList_39_#13:Array of St FXB Color Array Structures=Get tmpColorArrayList() -> (tmpColorArrayList:Array of St FXB Color Array Structures)
    Depend_38_#14:String=@n47
    Category_48_#15:String=@n3
    FolderPath_50_#16:String=@n5
    Tags_54_#17:Array of Strings=Get TagArray() -> (TagArray:Array of Strings)
    MyIndexInDatabase_58_#18:int=Length(TargetArray:Array of St FXB Item Structures=Get tmpItemList())
    MyCategoryIndexInDataBase_61_#19:int=@n46
   ) -> (St_FXB_Item:St FXB Item Structure) {is=title K2Node_MakeStruct}
   n52 = Break St FXB Item(St_FXB_Item=@n51) -> (HasValidFX_22_#9=~)
   n53 = Make St FXB Tag 2Item(
    ItemEditorName_3_#20=@n52.EditorName_14_#6
    ItemCategoryIndex_7_#21=@n46
    ItemIndex_9_#22=@n52.MyIndexInDatabase_58_#18
   )
   n54 = Get DataBase_Tag2ItemArray(self=Get SaveGame())
   Add(
    execute=@n67.then
    TargetArray:Array of St FXB Tag 2Item Array Structures=@n54
    NewItem:"St FXB Tag 2Item Array Structure (by ref)"=Make St FXB Tag 2Item Array(
     Tag2ItemList_3_#2=Make Array("[0]":St FXB Tag 2Item Structure=@n53) -> (Array:Array of St FXB Tag 2Item Structures)
     ItemNameList_6_#3=Make Array("[0]":String=@n52.EditorName_14_#6) -> (Array:Array of Strings)
    )
   )
   Is Empty!(TargetArray=@n7) -> (ReturnValue:bool)
   n55 = For Each Loop(Exec=@n72.then, Array:Array of St FXB String Array Structures=Get ReadedTagItemList()) -> (Array Element:St FXB String Array Structure)
   Add Unique(execute=@n82.then, TargetArray=Get TagArray(), NewItem="Get (a copy)"(Array=Get ReadedTagList(), Dimension 1=@n55.Array Index))
   Add(
    execute=@n13.Completed
    TargetArray:Array of St FXB Color Array Structures=Get tmpColorArrayList()
    NewItem:"St FXB Color Array Structure (by ref)"=Make St FXB Color Array!(Array_9_#23:Array of Linear Color Structures=Get tmpColorList()) -> (St_FXB_ColorArray:St FXB Color Array Structure) {is=title K2Node_MakeStruct}
   )
   n56 = Break St FXB Item(St_FXB_Item=@n51) -> (HasValidFX_22_#9=~)
   Add(
    execute=@n70.then
    TargetMap:Map of Strings to Int Vector 2 Structures=Get DataBase_CategoryIndexAndItemName(self=Get SaveGame())
    Key=@n56.EditorName_14_#6
    Value:"Int Vector 2 Structure (by ref)"=Make Int Vector 2!(X:int=@n56.MyCategoryIndexInDataBase_61_#19, Y:int=@n56.MyIndexInDatabase_58_#18) -> (IntVector2:Int Vector 2 Structure) {is=title K2Node_MakeStruct}
   ) {is=title K2Node_CallFunction}
   Get LibraryBasePath()
   n57 = For Each Loop(Exec=@n79.then_4, Array:Array of Strings=Get TagArray()) -> (Array Element:String)
   n58 = Get DataBase_TagList(self=Get SaveGame())
   n59 = Find Item(TargetArray=@n58, ItemToFind=@n57)
   n60 = "Get (a ref)"(Array:Array of St FXB Tag 2Item Array Structures=@n54, Dimension 1=@n59) -> (Output:"St FXB Tag 2Item Array Structure (by ref)")
   n61 = Break St FXB Tag 2Item Array(St_FXB_Tag2ItemArray=@n60)
   n62 = Set tmpTag2ItemArray(execute=@n81.else, tmpTag2ItemArray:Array of St FXB Tag 2Item Structures=@n61) -> (Output_Get:Array of St FXB Tag 2Item Structures)
   > n63 = Set tmpItemNameList(tmpItemNameList:Array of Strings=@n61.ItemNameList_6_#3) -> (Output_Get:Array of Strings)
   n64 = Set members in St FXB Tag 2Item Array!(
    execute:Exec=@n68.then
    StructRef:"St FXB Tag 2Item Array Structure (by ref)"=@n60
    Tag2ItemList_3_#2:Array of St FXB Tag 2Item Structures=@n62
    ItemNameList_6_#3:Array of Strings=@n63
   ) -> (then:Exec, StructOut:"St FXB Tag 2Item Array Structure (by ref)") {is=title K2Node_SetFieldsInStruct}
   > "Set St FXB Tag 2Item Array Structure (by ref)"(Target=@n60, Value=@n64)
   n65 = Break St FXB String Array(St_FXB_StringArray=@n55)
   n66 = Length(TargetArray=@n45)
   n67 = Add(execute=@n81.then, TargetArray:Array of Strings=@n58, NewItem:"String (by ref)"=@n57) {is=title K2Node_CallArrayFunction}
   n68 = Add(execute=@n69.then, TargetArray=@n63, NewItem=@n52.EditorName_14_#6)
   n69 = Add(execute=@n63.then, TargetArray:Array of St FXB Tag 2Item Structures=@n62, NewItem:"St FXB Tag 2Item Structure (by ref)"=@n53)
   n70 = Add(execute=@n79.then_3, TargetArray:Array of St FXB Item Structures=Get tmpItemList(), NewItem:"St FXB Item Structure (by ref)"=@n51)
   n71 = Find Item(TargetArray=@n65, ItemToFind=@n50)
   n72 = Clear(execute=@n80.then_1, TargetArray=Get TagArray())
   n73 = Clear(execute=@n77.then, TargetArray=Get tmpVideoList())
   n74 = Clear(execute=@n4.then, TargetArray:Array of St FXB Item Structures=Get tmpItemList())
   n75 = Add(execute=@n9.LoopBody, TargetArray:Array of Strings=Get tmpVideoList(), NewItem:"String (by ref)"=Replace(SourceString=@n9, From=@n11))
   n76 = Clear(execute=@n12.then, TargetArray:Array of Linear Color Structures=Get tmpColorList())
   n77 = Clear(execute=@n79.then_2, TargetArray:Array of St FXB Color Array Structures=Get tmpColorArrayList())
   n78 = PrintString(execute=@n5.Completed, InString="To String (Integer)"(InInt=@n66))
   n79 = Sequence(execute=@n5.LoopBody)
   > n80 = Sequence!(execute) -> (then_0, then_1)
   n81 = Branch(execute=@n57.LoopBody, Condition="integer < integer"(A=@n59))
   n82 = Branch(execute=@n55.LoopBody, Condition="integer >= integer"(A=@n71))
   n83 = Branch(execute=@n7.then, Condition=@n8)
  ]}
 DB_BuildTag {nodes [
   DB_BuildTag() {is=title K2Node_FunctionEntry}
   n1 = FindSubstring(SearchIn=Reverse(SourceString=Get LibraryBasePath()), Substring="\\")
   > n2 = Branch(Condition="Equal (Integer)"(A=@n1))
   > n3 = Read File to String(FilePath=Append(A=Get LibraryBasePath(), B=CategoryInformation.txt), bUseFullFilePath=true)
   > n4 = Branch(Condition=@n3.ReturnValue)
   n5 = Sequence!(execute=[@n61.then, @n4.then]) -> (then_0, then_1, then_2:Exec, then_3:Exec, then_4:Exec)
   > Clear(TargetArray:Array of Strings=Get PendingAddTagList() -> (PendingAddTagList:Array of Strings))
   > Clear(TargetArray:Array of St FXB String Array Structures=Get PendingAddTagItemList() -> (PendingAddTagItemList:Array of St FXB String Array Structures))
   > Clear(TargetArray:Array of Strings=Get tmpTagFileContentList())
   > n6 = GetFileListInPath("Path(/end)"=Append(A=Get LibraryBasePath(), B=Tag/), ext=txt, __WorldContext=Self-Reference())
   > Set tmpTagFileList(tmpTagFileList=@n6)
   > n7 = For Each Loop(Array=@n6.FullPathList)
   > n8 = Read File to String(FilePath=@n7, bUseFullFilePath=true)
   > Add(TargetArray=Get tmpTagFileContentList(), NewItem=@n8)
   n9 = For Each Loop(Exec=@n5.then_1, Array=Get tmpTagFileList())
   > Clear(TargetArray=Get InDiskTagList() -> (InDiskTagList:Array of Strings))
   > Clear(TargetArray:Array of St FXB String Array Structures=Get InDiskTagItems() -> (InDiskTagItems:Array of St FXB String Array Structures))
   > Clear(TargetArray:Array of Strings=Get InSaveGameTagList() -> (InSaveGameTagList:Array of Strings))
   > Clear(TargetArray:Array of St FXB String Array Structures=Get InSaveGameTagItems() -> (InSaveGameTagItems:Array of St FXB String Array Structures))
   > n10 = Sequence!(execute) -> (then_0, then_1)
   n11 = Get DataBaseSaveGame!() -> (DataBaseSaveGame)
   n12 = Get DataBase_TagFiles!(self:FXB Save Game Object Reference=@n11) -> (DataBase_TagFiles:Array of Strings)
   n13 = Find Item(TargetArray=@n12, ItemToFind=@n9)
   n14 = Branch(execute=@n10.then_1, Condition="integer >= integer"(A=@n13))
   n15 = "Get (a ref)"(Array:Array of Strings=Get tmpTagFileContentList(), Dimension 1=@n9.Array Index) -> (Output:"String (by ref)")
   n16 = Get DataBase_TagStrings!(self:FXB Save Game Object Reference=@n11) -> (DataBase_TagStrings:Array of Strings)
   n17 = "Get (a ref)"(Array=@n16, Dimension 1=@n13)
   > n18 = Branch(Condition="Equal, Case Insensitive (String)"(A=@n17, B=@n15))
   n19 = For Each Loop(Exec=@n18.else, Array=ParseIntoArray(SourceString=@n15, Delimiter="\r\n"))
   n20 = For Each Loop(Exec=@n19.Completed, Array=ParseIntoArray(SourceString=@n17, Delimiter="\r\n"))
   n21 = Get InDiskTagItems()
   n22 = Get InDiskTagList()
   n23 = Get InSaveGameTagList()
   n24 = Get InSaveGameTagItems()
   CompareTwoTagData!(
    execute=@n20.Completed
    self
    ATagList:Array of Strings=@n22
    BTagList:Array of Strings=@n23
    AItemList:Array of St FXB String Array Structures=@n21
    BItemList:Array of St FXB String Array Structures=@n24
    In_PendingTagList:Array of Strings=Get PendingAddTagList()
    In_PendingItemList:Array of St FXB String Array Structures=Get PendingAddTagItemList()
   ) -> (then) {is=title K2Node_CallFunction}
   > CompareTwoTagData(
    ATagList=@n23
    BTagList=@n22
    AItemList=@n24
    BItemList=@n21
    In_PendingTagList=Get PendingRemoveTagList() -> (PendingRemoveTagList:Array of Strings)
    In_PendingItemList=Get PendingRemoveTagItemList() -> (PendingRemoveTagItemList:Array of St FXB String Array Structures)
   )
   n25 = Split(SourceString=@n20, InStr="%&&%", SearchDir=FromStart)
   Add(execute=@n20.LoopBody, TargetArray=Get InSaveGameTagList(), NewItem=@n25)
   > Add(
    TargetArray:Array of St FXB String Array Structures=Get InSaveGameTagItems()
    NewItem:"St FXB String Array Structure (by ref)"=Make St FXB String Array(String_3_#1=ParseIntoArray(SourceString=@n25.RightS, Delimiter="%@%"))
   )
   n26 = Split(SourceString=@n19, InStr="%&&%", SearchDir=FromStart)
   Add(execute=@n19.LoopBody, TargetArray:Array of Strings=Get InDiskTagList(), NewItem:"String (by ref)"=@n26)
   > Add(
    TargetArray:Array of St FXB String Array Structures=Get InDiskTagItems()
    NewItem:"St FXB String Array Structure (by ref)"=Make St FXB String Array(String_3_#1=ParseIntoArray(SourceString=@n26.RightS, Delimiter="%@%"))
   )
   n27 = For Each Loop(Exec=@n14.else, Array=ParseIntoArray(SourceString=@n15, Delimiter="\r\n"))
   n28 = Split(SourceString=@n27, InStr="%&&%", SearchDir=FromStart)
   n29 = Find Item(TargetArray=Get PendingAddTagList(), ItemToFind=@n28)
   > n30 = Branch(Condition="integer >= integer"(A=@n29))
   n31 = ParseIntoArray(SourceString=@n28.RightS, Delimiter="%@%")
   > n32 = For Each Loop(Array=@n31)
   n33 = Get PendingAddTagItemList()
   > Add Unique(
    TargetArray=Break St FXB String Array(St_FXB_StringArray="Get (a ref)"(Array:Array of St FXB String Array Structures=@n33, Dimension 1=@n29) -> (Output:"St FXB String Array Structure (by ref)"))
    NewItem=@n32
   )
   Add(execute=@n30.else, TargetArray:Array of Strings=Get PendingAddTagList(), NewItem:"String (by ref)"=@n28)
   > Add(
    TargetArray:Array of St FXB String Array Structures=Get PendingAddTagItemList()
    NewItem:"St FXB String Array Structure (by ref)"=Make St FXB String Array(String_3_#1=@n31)
   )
   n34 = For Each Loop(Exec=@n5.then_2, Array=@n12)
   n35 = Get tmpTagFileList()
   > n36 = Branch(Condition="integer >= integer"(A=Find Item(TargetArray=@n35, ItemToFind=@n34)))
   n37 = For Each Loop(
    Exec=@n36.else
    Array=ParseIntoArray(SourceString="Get (a ref)"(Array:Array of Strings=@n16, Dimension 1=@n34.Array Index) -> (Output:"String (by ref)"), Delimiter="\r\n")
   )
   n38 = Split(SourceString=@n37, InStr="%&&%", SearchDir=FromStart)
   n39 = Get PendingRemoveTagList()
   n40 = Find Item(TargetArray=@n39, ItemToFind=@n38)
   > n41 = Branch(Condition="integer >= integer"(A=@n40))
   n42 = ParseIntoArray(SourceString=@n38.RightS, Delimiter="%@%")
   > n43 = For Each Loop(Array=@n42)
   n44 = Get PendingRemoveTagItemList()
   > Add Unique(
    TargetArray=Break St FXB String Array(St_FXB_StringArray="Get (a ref)"(Array:Array of St FXB String Array Structures=@n44, Dimension 1=@n40) -> (Output:"St FXB String Array Structure (by ref)"))
    NewItem=@n43
   )
   Add(execute=@n41.else, TargetArray:Array of Strings=@n39, NewItem:"String (by ref)"=@n38)
   > Add(
    TargetArray:Array of St FXB String Array Structures=@n44
    NewItem:"St FXB String Array Structure (by ref)"=Make St FXB String Array(String_3_#1=@n42)
   )
   PrintString(execute=@n5.then_3, InString=AddTagList:)
   > n45 = For Each Loop(Array=Get PendingAddTagList())
   PrintString(execute=@n45.Completed, InString=RemoveTagList:)
   > n46 = For Each Loop(Array=Get PendingRemoveTagList())
   > PrintString(InString=@n46)
   n47 = Get PendingRemoveTagItemList()
   > n48 = For Each Loop(Array=Break St FXB String Array(St_FXB_StringArray="Get (a ref)"(Array=@n47, Dimension 1=@n46.Array Index)))
   > PrintString(InString=Append(A=---, B=@n48))
   PrintString(execute=@n45.LoopBody, InString=@n45)
   n49 = Get PendingAddTagItemList()
   > n50 = For Each Loop(Array=Break St FXB String Array(St_FXB_StringArray="Get (a ref)"(Array=@n49, Dimension 1=@n45.Array Index)))
   > PrintString(InString=Append(A=---, B=@n50))
   n51 = For Each Loop(Exec=@n5.then_4, Array=Get PendingAddTagList())
   n52 = For Each Loop(Exec=@n51.Completed, Array=Get PendingRemoveTagList())
   n53 = Get DataBaseSaveGame()
   Set DataBase_TagFiles(execute=@n52.Completed, DataBase_TagFiles=Get tmpTagFileList(), self=@n53)
   > Set DataBase_TagStrings(DataBase_TagStrings=Get tmpTagFileContentList(), self=@n53)
   > SaveGameToSlot(SaveGameObject=Get DataBaseSaveGame(), SlotName=FXBrowser, UserIndex=0)
   > UI_InitialTag(self=Self-Reference())
   > UI_HandleCategoryChange!(
    execute
    self
    CategoryList:Array of Strings=Get CurrentCategoryList() -> (CurrentCategoryList:Array of Strings)
    CheckedList:Array of Booleans=Get CurrentCheckedList() -> (CurrentCheckedList:Array of Booleans)
    Shuffle:bool=false
   ) -> (then)
   n54 = Get PendingRemoveTagItemList()
   n55 = For Each Loop(Exec=@n52.LoopBody, Array=Break St FXB String Array(St_FXB_StringArray="Get (a ref)"(Array=@n54, Dimension 1=@n52.Array Index)))
   > DeleteATag!(execute, self, Tag:String=@n52, ItemName:String=@n55) -> (then)
   n56 = Get PendingAddTagItemList()
   n57 = For Each Loop(Exec=@n51.LoopBody, Array=Break St FXB String Array(St_FXB_StringArray="Get (a ref)"(Array=@n56, Dimension 1=@n51.Array Index)))
   > CreateNewTag!(execute, self, Tag:String=@n51, ItemName:String=@n57) -> (then) {is=title K2Node_CallFunction}
   PrintString(execute=@n2.else, InString="To String (Integer)"(InInt=@n1))
   > n58 = Read File to String(FilePath=Append(A=Get LibraryBasePath(), B="\\CategoryInformation.txt"), bUseFullFilePath=true)
   > Branch(Condition=@n58.ReturnValue)
   > n59 = Set LibraryBasePath(LibraryBasePath=Append(A=Get LibraryBasePath(), B="\\"))
   n60 = Get SettingSaveObject()
   > Set LibPath(LibPath=@n59, self=@n60)
   > n61 = SaveGameToSlot(SaveGameObject=@n60, SlotName=FXBrowserSetting, UserIndex=0)
  ]}
 DB_LoadDataBase {nodes [
   DB_LoadDataBase() {is=title K2Node_FunctionEntry}
   > n1 = LoadGameFromSlot!(execute, SlotName:String=FXBrowser, UserIndex:int=0) -> (then, ReturnValue:Save Game Object Reference)
   > n2 = Cast To FXB_SaveGame!(execute, Object=@n1) -> (then, CastFailed, AsFXB Save Game:FXB Save Game Object Reference) {is=title K2Node_DynamicCast}
   > Set DataBaseSaveGame(DataBaseSaveGame:FXB Save Game Object Reference=@n2) -> (Output_Get:FXB Save Game Object Reference)
   > Return Node!(execute, Valid:bool)
   Return Node(execute=@n2.CastFailed, Valid=false)
  ]}
 DeleteATag {nodes [
   n1 = DeleteATag!() -> (then, Tag:String, ItemName:String) {is=title K2Node_FunctionEntry}
   n2 = Self-Reference()
   n3 = Find(
    TargetMap=Get DataBase_CategoryIndexAndItemName(self=Get DataBaseSaveGame!(self:BUW FXBrowser Object Reference=@n2) -> (DataBaseSaveGame))
    Key=@n1.ItemName
   )
   > Branch(Condition=@n3.ReturnValue)
   n4 = Get DataBaseSaveGame(self=Self-Reference())
   n5 = Find Item(TargetArray=Get DataBase_TagList(self=@n4), ItemToFind=@n1)
   > Branch(Condition="integer >= integer"(A=@n5))
   n6 = "Get (a ref)"(Array:Array of St FXB Tag 2Item Array Structures=Get DataBase_Tag2ItemArray(self=@n4), Dimension 1=@n5) -> (Output:"St FXB Tag 2Item Array Structure (by ref)")
   n7 = Break St FXB Tag 2Item Array(St_FXB_Tag2ItemArray=@n6)
   > n8 = Set NewLocalVar(NewLocalVar=@n7)
   > n9 = Set NewLocalVar_0(NewLocalVar_0=@n7.ItemNameList_6_#3)
   n10 = Break Int Vector 2(IntVector2=@n3)
   n11 = "Get (a ref)"(Array:Array of St FXB Item List Structures=Get DataBase_ItemList(self=@n4), Dimension 1=@n10) -> (Output:"St FXB Item List Structure (by ref)")
   n12 = Break St FXB Item(St_FXB_Item="Get (a ref)"(Array:Array of St FXB Item Structures=Break St FXB Item List(St_FXB_ItemList=@n11), Dimension 1=@n10.Y) -> (Output:"St FXB Item Structure (by ref)")) -> (HasValidFX_22_#9=~)
   n13 = Find Item(TargetArray=@n7.ItemNameList_6_#3, ItemToFind=@n12.EditorName_14_#6)
   > Remove Index!(execute, TargetArray:Array of St FXB Tag 2Item Structures=@n8, IndexToRemove:int=@n13) -> (then)
   > Remove Index(TargetArray:Array of Strings=@n9, IndexToRemove=@n13)
   > "Set St FXB Tag 2Item Array Structure (by ref)"(
    Target=@n6
    Value=Make St FXB Tag 2Item Array(Tag2ItemList_3_#2=@n8, ItemNameList_6_#3=@n9)
   )
   n14 = Self-Reference()
   n15 = "Get (a ref)"(
    Array:Array of St FXB Item List Structures=Get DataBase_ItemList(self=Get DataBaseSaveGame(self=@n14))
    Dimension 1=@n12.MyCategoryIndexInDataBase_61_#19
   ) -> (Output:"St FXB Item List Structure (by ref)")
   n16 = Break St FXB Item List(St_FXB_ItemList=@n15)
   n17 = "Get (a ref)"(Array:Array of St FXB Item Structures=@n16, Dimension 1=@n12.MyIndexInDatabase_58_#18) -> (Output:"St FXB Item Structure (by ref)")
   > n18 = Set NewLocalVar_1(NewLocalVar_1=Break St FXB Item(St_FXB_Item=@n17) -> (HasValidFX_22_#9=~).Tags_54_#17)
   > Remove Item!(execute, TargetArray=@n18, Item:"String (by ref)"=Get Tag()) -> (then, ReturnValue:bool)
   > n19 = Set members in St FXB Item(StructRef=@n17, Tags_54_#17=Get NewLocalVar_1())
   > Set Array Elem(TargetArray=@n16, Index=@n12.MyIndexInDatabase_58_#18, Item=@n19, bSizeToFit=false)
   > "Set St FXB Item List Structure (by ref)"(Target=@n15, Value=Make St FXB Item List(ItemList_16_#4=@n16))
  ]}
 EventGraph {nodes [
   n1 = Append!(
    A="H:\\Capture\\Backup\\Arranged\\"
    B=LoadAPickaxe!() -> (OutputDelegate, then, FolderName:String) {is=title K2Node_CustomEvent}.FolderName
    C:String="\\"
   ) -> (ReturnValue)
   PrintString(InString=@n1)
   n2 = Self-Reference()
   > n3 = Construct Object from Class!(
    execute
    Class=&/EUO_PreviewItemObject.EUO_PreviewItemObject_C
    self=@n2
    FolderPath:String=@n1
    ParentWidget:BUW FXBrowser Object Reference=@n2
    LoadRuler:String
    MyIndex:int=0
    SubUVResX:int=4
    SubUVResY:int=4
   ) -> (then, ReturnValue:EUO Preview Item Object Object Reference)
   > Add(
    TargetArray:Array of EUO Preview Item Object Object References=Get CurrentPageObjectList() -> (CurrentPageObjectList:Array of EUO Preview Item Object Object References)
    NewItem:"EUO Preview Item Object Object Reference (by ref)"=@n3
   )
   > Initial(self:EUO Preview Item Object Object Reference=@n3)
   > AddItem!(execute, self:List View Object Reference=Get TileView_218() -> (TileView_218:Tile View Object Reference), Item:Object Reference=@n3) -> (then)
   n4 = "Execute Python Command (Advanced)"!(
    execute
    PythonCommand:String="import os\r\n\r\n\r\ndef get_directories(path):\r\n    return [name for name in os.listdir(path) if os.path.isdir(os.path.join(path, name))]\r\n\r\n\r\ndirectories = get_directories(\"H:/Capture/Backup/Arranged\")\r\nprint(directories)"
    ExecutionMode:EPythonCommandExecutionMode Enum=ExecuteFile
    FileExecutionScope:EPythonFileExecutionScope Enum=Private
   ) -> (then, CommandResult:String, LogOutput:Array of Python Log Output Entry Structures, ReturnValue:bool)
   n5 = Break Python Log Output Entry!(PythonLogOutputEntry:"Python Log Output Entry Structure (by ref)"="Get (a ref)"(Array:Array of Python Log Output Entry Structures=@n4.LogOutput, Dimension 1=0) -> (Output:"Python Log Output Entry Structure (by ref)")) -> (Type:EPythonLogOutputType Enum, Output:String) {is=title K2Node_BreakStruct}
   n6 = Replace(SourceString=Replace(SourceString=Replace(SourceString=@n5.Output, From="["), From="]"), From=')
   > Set FolderList(FolderList:Array of Strings=ParseIntoArray(SourceString=Replace(SourceString=@n6, From="\r\n"), Delimiter=", ", CullEmptyStrings=false)) -> (Output_Get:Array of Strings)
   "On Clicked (CosmeticTypeButton)"() {is=title K2Node_ComponentBoundEvent}
   n7 = Get CosmeticCategorySlot() -> (CosmeticCategorySlot:Overlay Object Reference)
   > n8 = Branch(Condition="Equal (Enum)"(
     A:ESlateVisibility Enum=GetVisibility!(self:Widget Object Reference=@n7) -> (ReturnValue:ESlateVisibility Enum=Visible)
     B:ESlateVisibility Enum=SelfHitTestInvisible
    ))
   > SetVisibility(self=@n7, InVisibility=Collapsed)
   SetVisibility(execute=@n8.else, self=@n7, InVisibility=SelfHitTestInvisible)
   "On Clicked (EditorUtilityButton_0)"() {is=title K2Node_ComponentBoundEvent}
   n9 = Get MediaPlayer() -> (MediaPlayer:Media Player Object Reference)
   > Reopen!(execute, self:Media Player Object Reference=@n9) -> (then, ReturnValue:bool)
   > n10 = Play!(execute, self:Media Player Object Reference=@n9) -> (then, ReturnValue:bool)
   n11 = ShowVideo!() -> (OutputDelegate, then, Folder:String, SourceButton:EUW Preview Button Object Reference, LoadRuler:String) {is=title K2Node_CustomEvent}
   n12 = "Get (a ref)"(
    Array:Array of St FXB Item Structures=Get CurrentPendingLoadItemList() -> (CurrentPendingLoadItemList:Array of St FXB Item Structures)
    Dimension 1=Get MyIndex!(self:EUO Preview Item Object Object Reference=Get MyItemObject!(self:EUW Preview Button Object Reference=@n11.SourceButton) -> (MyItemObject:EUO Preview Item Object Object Reference)) -> (MyIndex:int=0)
   ) -> (Output:"St FXB Item Structure (by ref)")
   n13 = Break St FXB Item(St_FXB_Item=@n12) -> (HasValidFX_22_#9=~)
   RetriggerableDelay!(
    execute=[@n10.then, @n179.then]
    WorldContextObject:Object Reference
    Duration=SelectFloat(A=50, B=20, bPickA="Equal Exactly (String)"(A=@n13.Category_48_#15, B=Gameplay))
   ) -> (then)
   > Pause!(execute, self:Media Player Object Reference=@n9) -> (then, ReturnValue:bool)
   > RetriggerableDelay(Duration=50)
   > Close(self:Media Player Object Reference=@n9)
   n14 = "On Text Changed (EditorUtilityEditableTextBox_25)"!() -> (OutputDelegate, then, Text:"Text (by ref)") {is=title K2Node_ComponentBoundEvent}
   > Set CurrentUser(CurrentUser:String="To String (Text)"(InText=@n14.Text)) -> (Output_Get:String)
   "On Clicked (EditorUtilityButton_203)"() {is=title K2Node_ComponentBoundEvent}
   > n15 = Sequence()
   n16 = Get DetailsView_172() -> (DetailsView_172:Details View Object Reference)
   > n17 = Branch(Condition="Equal (Byte)"(A:Byte=GetVisibility(self=@n16) -> (ReturnValue=Visible), B:Byte) -> (ReturnValue:bool=~) {is=title K2Node_PromotableOperator})
   > SetVisibility(self=@n16, InVisibility=Collapsed)
   n18 = Get EditorUtilityButton_203() -> (EditorUtilityButton_203:Editor Utility Button Object Reference)
   > SetBackgroundColor!(execute, self:Button Object Reference=@n18, InBackgroundColor:Linear Color Structure=(R=0.8153,G=0.8594,B=1,A=1)) -> (then)
   SetVisibility(execute=@n17.else, self=@n16, InVisibility=Visible)
   > SetBackgroundColor(self=@n18, InBackgroundColor=(R=0.2906,G=1,B=0,A=1))
   n19 = Break St FXB Item(St_FXB_Item=Get CurrentPlayItem() -> (CurrentPlayItem:St FXB Item Structure)) -> (HasValidFX_22_#9=~)
   n20 = Branch(execute=@n15.then_1, Condition=IsEmpty(InString=@n19.EditorName_14_#6))
   n21 = Get RelatedAssetsList() -> (RelatedAssetsList:Array of Object References)
   Clear(execute=@n20.else, TargetArray:Array of Object References=@n21)
   > n22 = LoadAsset(AssetPath=Split(SourceString=@n19, InStr=" ", SearchCase=CaseSensitive, SearchDir=FromStart).RightS)
   > Add Unique(TargetArray:Array of Object References=@n21, NewItem:"Object Reference (by ref)"=@n22)
   > n23 = For Each Loop(Array=ParseIntoArray(SourceString=@n19.Depend_38_#14, Delimiter="\r\n"))
   > n24 = LoadAsset(AssetPath=@n23)
   > n25 = Cast To NiagaraSystem!(execute, Object=@n24) -> (then, CastFailed, AsNiagara System:Niagara System Object Reference) {is=title K2Node_DynamicCast}
   > Add Unique(TargetArray=@n21, NewItem=@n25)
   "On Clicked (ClearSearchbt)"() {is=title K2Node_ComponentBoundEvent}
   > Set Text(self:Text Box Object Reference=Get EditorUtilityEditableTextBox_65() -> (EditorUtilityEditableTextBox_65:Editor Utility Editable Text Box Object Reference))
   > Set Search(Search:String) -> (Output_Get:String)
   > UI_HandleCategoryChange(CategoryList=Get CurrentCategoryList(), CheckedList=Get CurrentCheckedList(), Shuffle=false)
   "On Text Changed (EditorUtilityEditableTextBox_413)"!() -> (OutputDelegate, then, Text:"Text (by ref)") {is=title K2Node_ComponentBoundEvent}
   > Is Valid(InputObject=Get LastShowSource() -> (LastShowSource:EUW Preview Button Object Reference))
   > UI_InitialRightTagPanel()
   "On Clicked (EditorUtilityButton_485)"() {is=title K2Node_ComponentBoundEvent}
   > Is Valid(InputObject=Get LastShowSource())
   n26 = "To String (Text)"(InText=Get Text(self=Get EditorUtilityEditableTextBox_413() -> (EditorUtilityEditableTextBox_413:Editor Utility Editable Text Box Object Reference)))
   > n27 = Branch(Condition=IsEmpty(InString=@n26))
   UI_ChangeTag!(execute=@n27.else, self, Tag:String=@n26, Check:bool=true) -> (then)
   > UI_InitialRightTagPanel()
   n28 = "On Clicked (EditorUtilityButton_5)"() {is=title K2Node_ComponentBoundEvent}
   n29 = Get CurrentCategoryList()
   Branch(execute=[@n28.then, @n95.then], Condition="integer >= integer"(A=Length(TargetArray:Array of Strings=@n29)))
   > UI_HandleCategoryChange(CategoryList=Get CurrentCategoryList(), CheckedList=Get CurrentCheckedList(), Shuffle=false)
   "On Clicked (EditorUtilityButton_166)"() {is=title K2Node_ComponentBoundEvent}
   > n30 = Sequence()
   n31 = Get Text(self=Get MinVersionInput() -> (MinVersionInput:Editor Utility Editable Text Box Object Reference))
   > Set MinVersion(MinVersion:int=String To Integer(InString="To String (Text)"(InText=@n31))) -> (Output_Get:int=0)
   n32 = Get Text(self=Get MaxVersionInput() -> (MaxVersionInput:Editor Utility Editable Text Box Object Reference))
   > n33 = Set MaxVersion(MaxVersion:int=String To Integer(InString="To String (Text)"(InText=@n32))) -> (Output_Get:int=0)
   UI_HandleCategoryChange(execute=[@n43.then, @n33.then], CategoryList=Get CurrentCategoryList(), CheckedList=Get CurrentCheckedList(), Shuffle=false)
   HighLightDisableFilterButton!(execute=@n30.then_1, self, HighLight?:bool=true) -> (then)
   "On Clicked (EditorUtilityButton_584)"() {is=title K2Node_ComponentBoundEvent}
   n34 = Get EditorUtilityEditableTextBox_65()
   > Set Search(Search="To String (Text)"(InText=Get Text(self=@n34)))
   > UI_HandleCategoryChange(CategoryList=Get CurrentCategoryList(), CheckedList=Get CurrentCheckedList(), Shuffle=false)
   n35 = "On Value Changed (AnalogSlider_1)"!() -> (OutputDelegate, then, Value:float) {is=title K2Node_ComponentBoundEvent}
   > Set ColorF(ColorF:Linear Color Structure=HSV to RGB!(H:float=@n35.Value, S:float=1, V:float=1, A:float=1) -> (ReturnValue:Linear Color Structure)) -> (Output_Get:Linear Color Structure)
   "On Clicked (EditorUtilityButton_6)"() {is=title K2Node_ComponentBoundEvent}
   > Set ColorFApply(ColorFApply:Linear Color Structure=Get ColorF() -> (ColorF:Linear Color Structure)) -> (Output_Get:Linear Color Structure)
   > UI_HandleCategoryChange(CategoryList=Get CurrentCategoryList(), CheckedList=Get CurrentCheckedList(), Shuffle=false)
   > HighLightDisableFilterButton(HighLight?=true)
   n36 = "On Value Changed (AnalogSlider_219)"!() -> (OutputDelegate, then, Value:float) {is=title K2Node_ComponentBoundEvent}
   n37 = float * float(A=@n36.Value, B=400) {is=title K2Node_PromotableOperator}
   n38 = Get TileView_218()
   > SetEntryHeight!(execute, self:Tile View Object Reference=@n38, NewHeight:float=@n37) -> (then)
   > SetEntryWidth!(execute, self:Tile View Object Reference=@n38, NewWidth:float=@n37) -> (then)
   n39 = "On Text Changed (EditorUtilityEditableTextBox)"!() -> (OutputDelegate, then, Text:"Text (by ref)") {is=title K2Node_ComponentBoundEvent}
   > Set ColorFT(ColorFT:int=String To Integer(InString="To String (Text)"(InText=@n39.Text))) -> (Output_Get:int=0)
   n40 = "On Text Committed (EditorUtilityEditableTextBox)"!() -> (OutputDelegate, then, Text:"Text (by ref)", CommitMethod:ETextCommit Enum=Default) {is=title K2Node_ComponentBoundEvent}
   > Set ColorFT(ColorFT=String To Integer(InString="To String (Text)"(InText=@n40.Text))) -> (Output_Get=0)
   n41 = "On Clicked (EditorUtilityButton_7)"() {is=title K2Node_ComponentBoundEvent}
   Set ColorFApply(execute=[@n41.then, @n62.then_0], ColorFApply=(R=0,G=0,B=0,A=1))
   > UI_HandleCategoryChange(CategoryList=Get CurrentCategoryList(), CheckedList=Get CurrentCheckedList(), Shuffle=false)
   > UpdateDisableFilterButtonHighLight()
   n42 = "On Clicked (EditorUtilityButton_9)"() {is=title K2Node_ComponentBoundEvent}
   Set Text(execute=[@n42.then, @n62.then_1], self=Get MinVersionInput())
   > Set Text(self=Get MaxVersionInput())
   > Set MinVersion(MinVersion=0) -> (Output_Get=0)
   > Set MaxVersion(MaxVersion=99) -> (Output_Get=0)
   > n43 = UpdateDisableFilterButtonHighLight()
   "On Clicked (EditorUtilityButton_8)"() {is=title K2Node_ComponentBoundEvent}
   > UI_HandleCategoryChange(CategoryList=Get CurrentCategoryList(), CheckedList=Get CurrentCheckedList(), Shuffle=true)
   n44 = "On Check State Changed (EditorUtilityCheckBox_1)"!() -> (OutputDelegate, then, bIsChecked:bool) {is=title K2Node_ComponentBoundEvent}
   > Set SearchDepend(SearchDepend:bool=@n44.bIsChecked) -> (Output_Get:bool)
   "On Clicked (EditorUtilityButton_91)"() {is=title K2Node_ComponentBoundEvent}
   > UI_HandleCategoryChange(CategoryList=Get CurrentCategoryList(), CheckedList=Get CurrentCheckedList(), Shuffle=false)
   "On Clicked (EditorUtilityButton_10)"() {is=title K2Node_ComponentBoundEvent}
   > LaunchURL!(execute, URL:String) -> (then)
   "On Clicked (RebuildTagButton)"() {is=title K2Node_ComponentBoundEvent}
   > DB_BuildTag() {is=title K2Node_CallFunction}
   n45 = "On Clicked (EditorUtilityButton_299)"() {is=title K2Node_ComponentBoundEvent}
   n46 = Append!(A=GetProjectDirectory!() -> (ReturnValue:String), B=Plugins/FXBrowserUtilities/Resources/Library) -> (ReturnValue)
   PrintString(execute=[@n45.then, @n55.then], InString=Append(A="Will build database from : ", B=@n46))
   > Sequence()
   > SetUpLibPath!(execute, self, Path:String=@n46) -> (then)
   > RetriggerableDelay(Duration=0.5)
   > Rebuild Database()
   "On Clicked (EditorUtilityButton_62)"() {is=title K2Node_ComponentBoundEvent}
   > n47 = Flip Flop!(None:Exec) -> (A:Exec, B:Exec, IsA:bool=~) {is=title K2Node_MacroInstance}
   n48 = Get LeftPanel() -> (LeftPanel:Overlay Object Reference)
   > SetVisibility(self=@n48, InVisibility=Collapsed)
   SetVisibility(execute=@n47.B, self=@n48, InVisibility=SelfHitTestInvisible)
   "On Clicked (EditorUtilityButton_11)"() {is=title K2Node_ComponentBoundEvent}
   n49 = Get EditorUtilityButton_12() -> (EditorUtilityButton_12:Editor Utility Button Object Reference)
   n50 = Get Image_0() -> (Image_0:Image Object Reference)
   n51 = Get TextBlock_138() -> (TextBlock_138:Text Object Reference)
   > SetVisibility(self=[@n49, @n50, @n51], InVisibility=Visible)
   n52 = Get EditorUtilityButton_11() -> (EditorUtilityButton_11:Editor Utility Button Object Reference)
   n53 = Get TextBlock_11() -> (TextBlock_11:Text Object Reference)
   n54 = Get EditorUtilityButton_408() -> (EditorUtilityButton_408:Editor Utility Button Object Reference)
   > SetVisibility(self=[@n52, @n53, @n54], InVisibility=Collapsed)
   "On Clicked (EditorUtilityButton_12)"() {is=title K2Node_ComponentBoundEvent}
   > SetVisibility(self=[@n52, @n53, @n54], InVisibility=Visible)
   > SetVisibility(self=[@n49, @n50, @n51], InVisibility=Collapsed)
   n55 = "On Clicked (EditorUtilityButton_408)"() {is=title K2Node_ComponentBoundEvent}
   "On Clicked (EditorUtilityButton_13)"() {is=title K2Node_ComponentBoundEvent}
   > Clear(TargetArray:Array of EUO Preview Item Object Object References=Get CurrentPageObjectList())
   > ClearListItems(self=Get TileView_218())
   > n56 = For Each Loop(Array:Array of St FXB Item Structures=Get CurrentPendingLoadItemList()) -> (Array Element:St FXB Item Structure)
   n57 = Break St FXB Item(St_FXB_Item=@n56) -> (HasValidFX_22_#9=~)
   > n58 = Test_GetTagFromAssetPath!(execute, self, AssetPath:Array of Strings=ParseIntoArray(SourceString=@n57.Depend_38_#14, Delimiter="\r\n")) -> (then, Tags:Array of Strings)
   > Branch(Condition=Contains Item(TargetArray:Array of Strings=@n58, ItemToFind:"String (by ref)"=Fire))
   n59 = Find(
    TargetMap:Map of Strings to Strings=Get CurrentLoadItemRuler() -> (CurrentLoadItemRuler:Map of Strings to Strings)
    Key=@n57.Category_48_#15
   ) -> (Value:String)
   n60 = Self-Reference()
   > n61 = Construct Object from Class(
    Class=&EUO_PreviewItemObject_C
    self=@n60
    FolderPath=@n57.FolderPath_50_#16
    ParentWidget=@n60
    LoadRuler=SelectString(A=@n59, B="Get (a copy)"(Array=@n57.VideoList_31_#12), bPickA=@n59.ReturnValue)
    MyIndex=@n56.Array Index
    SubUVResX=@n57.X_25_#10
    SubUVResY=@n57.Y_27_#11
   )
   > Add(TargetArray=Get CurrentPageObjectList(), NewItem=@n61)
   > Initial(self=@n61)
   > AddItem(self=Get TileView_218(), Item=@n61)
   "On Clicked (EditorUtilityButton_14)"() {is=title K2Node_ComponentBoundEvent}
   > n62 = Sequence!(execute) -> (then_0, then_1, then_2:Exec)
   n63 = "On Text Changed (LibPath)"!() -> (OutputDelegate, then, Text:"Text (by ref)") {is=title K2Node_ComponentBoundEvent}
   > n64 = Set LibraryBasePath(LibraryBasePath="To String (Text)"(InText=@n63.Text))
   RetriggerableDelay(execute=[@n170.then, @n169.then, @n129.then, @n64.then], Duration=0.1)
   n65 = Get SettingSaveObject()
   n66 = Get LibraryBasePath()
   > Set LibPath(LibPath=@n66, self=@n65)
   > n67 = SaveGameToSlot(SaveGameObject=@n65, SlotName=FXBrowserSetting, UserIndex=0)
   > PrintString(InString="To String (Boolean)"(InBool=@n67))
   n68 = "On Clicked (EditorUtilityButton_2)"() {is=title K2Node_ComponentBoundEvent}
   n69 = Get WaitBuildCanvas() -> (WaitBuildCanvas:Canvas Panel Object Reference)
   SetVisibility(execute=[@n68.then, @n121.then], self=@n69, InVisibility=SelfHitTestInvisible)
   > RetriggerableDelay(Duration=0.1)
   > DB_BuildDataBase() {is=title K2Node_CallFunction}
   > RetriggerableDelay(Duration=0.1)
   > SetVisibility(self=@n69, InVisibility=Collapsed)
   n70 = Get TutorialCanvas() -> (TutorialCanvas:Canvas Panel Object Reference)
   > n71 = Branch(Condition="Equal (Enum)"(A=GetVisibility(self=@n70) -> (ReturnValue=Visible), B=SelfHitTestInvisible))
   n72 = GetOSName!(execute=[@n164.then, @n71.else, @n71.then], __WorldContext:Object Reference) -> (then, Name:String)
   > Set CurrentUser(CurrentUser=SelectString(A=system, B=@n72, bPickA=Get SystemUser() -> (SystemUser:bool)))
   > n73 = LoadGameFromSlot(SlotName=FXBrowserSetting, UserIndex=0)
   > n74 = Cast To FXB_SaveGame_Path!(execute, Object=@n73) -> (then, CastFailed, AsFXB Save Game Path:FXB Save Game Path Object Reference) {is=title K2Node_DynamicCast}
   > n75 = Is Valid(InputObject=@n73)
   n76 = Construct Object from Class!(execute=[@n74.CastFailed, @n75.Is Not Valid], Class=&/FXB_SaveGame_Path.FXB_SaveGame_Path_C, self) -> (then, ReturnValue:FXB Save Game Path Object Reference)
   > n77 = Set SettingSaveObject(SettingSaveObject:FXB Save Game Path Object Reference=@n76) -> (Output_Get:FXB Save Game Path Object Reference)
   n78 = Get LibPath!(self:FXB Save Game Path Object Reference=Get SettingSaveObject()) -> (LibPath:String)
   Set Text(
    execute=[@n77.then, @n83.then]
    Text="To Text (String)"(InString=@n78)
    self=Get LibPath!() -> (LibPath:Editor Utility Editable Text Box Object Reference)
   )
   > Set LibraryBasePath(LibraryBasePath=@n78)
   > n79 = DB_LoadDataBase!(execute, self) -> (then, Valid:bool) {is=title K2Node_CallFunction}
   > n80 = Branch(Condition=@n79)
   n81 = Get TutorialCanvas()
   > SetVisibility(self=@n81, InVisibility=Collapsed)
   n82 = Get RebuildTagButton() -> (RebuildTagButton:Editor Utility Button Object Reference)
   > SetVisibility(self=@n82, InVisibility=Visible)
   > UI_InitialCategory()
   > UI_InitialTag()
   > SetObject(self=Get DetailsView_172(), NewObject=Self-Reference())
   SetVisibility(execute=@n80.else, self=@n81, InVisibility=SelfHitTestInvisible)
   > SetVisibility(self=@n82, InVisibility=Collapsed)
   n83 = Set SettingSaveObject(execute=@n75.Is Valid, SettingSaveObject=@n74)
   "On Clicked (EditorUtilityButton_1)"() {is=title K2Node_ComponentBoundEvent}
   n84 = Get RightPanel() -> (RightPanel:Overlay Object Reference)
   > n85 = Branch(Condition="Equal (Enum)"(A=GetVisibility(self=@n84) -> (ReturnValue=Visible), B=SelfHitTestInvisible))
   > SetVisibility(self=@n84, InVisibility=Collapsed)
   SetVisibility(execute=@n85.else, self=@n84, InVisibility=SelfHitTestInvisible)
   n86 = "On Value Changed (AnalogSlider)"!() -> (OutputDelegate, then, Value:float) {is=title K2Node_ComponentBoundEvent}
   > SetScalarParameterValue!(
    execute
    WorldContextObject=Self-Reference()
    Collection:Material Parameter Collection Object Reference=&/MPC_PreviewImage
    ParameterName:Name=PlaySped
    ParameterValue:float=@n86.Value
   ) -> (then) {is=title K2Node_CallMaterialParameterCollectionFunction}
   "On Clicked (EditorUtilityButton_3)"() {is=title K2Node_ComponentBoundEvent}
   n87 = Get CurrentPage() -> (CurrentPage:int=0)
   > n88 = Set CurrentPage(CurrentPage:int="Clamp (Integer)"(Value=int + int(A=@n87, B=-1), Min=0, Max=999) -> (ReturnValue=0)) -> (Output_Get:int=0)
   UI_LoadCurrentPage(execute=[@n94.then, @n88.then])
   "On Clicked (EditorUtilityButton_4)"() {is=title K2Node_ComponentBoundEvent}
   n89 = Length(TargetArray:Array of St FXB Item Structures=Get CurrentPendingLoadItemList())
   n90 = int + int(A=@n87)
   n91 = Get LoadPageItemNumber() -> (LoadPageItemNumber:int=0)
   n92 = int * int(A=int + int(A=@n90), B=@n91)
   > n93 = Branch(Condition="integer > integer"(A=int - int(A=@n92, B=@n89), B=@n91))
   n94 = Set CurrentPage(execute=@n93.else, CurrentPage=@n90) -> (Output_Get=0)
   n95 = ApplyTags() {is=title K2Node_CustomEvent}
   n96 = UI_ChangeTag!() -> (OutputDelegate, then, Tag:String, Check:bool=~) {is=title K2Node_CustomEvent}
   n97 = Get LastShowSource()
   > Is Valid(InputObject=@n97)
   > n98 = Branch(Condition=@n96.Check)
   n99 = Get MyItemObject(self=@n97)
   > CreateNewTag!(execute, self:EUO Preview Item Object Object Reference=@n99, Tag=@n96.Tag, SpecialUser:String) -> (then)
   DeleteATag!(execute=@n98.else, self:EUO Preview Item Object Object Reference=@n99, Tag=@n96.Tag, SpecialUser:String) -> (then) {is=title K2Node_CallFunction}
   n100 = HighLightDisableFilterButton!() -> (OutputDelegate, then, HighLight?:bool=~) {is=title K2Node_CustomEvent}
   > n101 = Branch(Condition=@n100.HighLight?)
   n102 = Get EditorUtilityButton_14() -> (EditorUtilityButton_14:Editor Utility Button Object Reference)
   > SetBackgroundColor(self=@n102, InBackgroundColor=(R=10,G=0,B=0,A=1))
   SetBackgroundColor(execute=@n101.else, self=@n102, InBackgroundColor=(R=0.5972,G=0.6308,B=0.7305,A=1))
   UpdateDisableFilterButtonHighLight() {is=title K2Node_CustomEvent}
   n103 = Get ColorFApply() -> (ColorFApply:Linear Color Structure)
   n104 = Get MinVersion() -> (MinVersion:int=0)
   n105 = Get MaxVersion() -> (MaxVersion:int=0)
   > Branch(Condition=AND Boolean!(
     A="Equal (LinearColor)"(A:Linear Color Structure=@n103, B:Linear Color Structure=(R=0,G=0,B=0,A=1)) -> (ReturnValue:bool=~) {is=title K2Node_PromotableOperator}
     B="Equal (Integer)"(A=@n104)
     C:bool="Equal (Integer)"(A=@n105, B=99)
    ) -> (ReturnValue))
   > HighLightDisableFilterButton(HighLight?=false) {is=title K2Node_CallFunction}
   UI_LoadCurrentPage() {is=title K2Node_CustomEvent}
   > Clear(TargetArray=Get CurrentPageObjectList())
   > ClearListItems(self=Get TileView_218())
   n106 = Get CurrentPage() -> (CurrentPage=0)
   n107 = int + int(A=@n106)
   n108 = Get LoadPageItemNumber() -> (LoadPageItemNumber=0)
   > n109 = For Loop(FirstIndex=int * int(A=@n106, B=@n108), LastIndex=int - int(A=int * int(A=@n107, B=@n108)))
   n110 = Get CurrentPendingLoadItemList()
   > Branch(Condition=Is Valid Index!(TargetArray:Array of St FXB Item Structures=@n110, IndexToTest:int=@n109) -> (ReturnValue:bool))
   n111 = Break St FXB Item(St_FXB_Item="Get (a ref)"(Array=@n110, Dimension 1=@n109)) -> (HasValidFX_22_#9=~)
   n112 = Find(TargetMap=Get CurrentLoadItemRuler(), Key=@n111.Category_48_#15)
   n113 = Self-Reference()
   > n114 = Construct Object from Class!(
    execute
    Class=&EUO_PreviewItemObject_C
    self=@n113
    FolderPath:String=@n111.FolderPath_50_#16
    ParentWidget:BUW FXBrowser Object Reference=@n113
    LoadRuler:String=SelectString(A=@n112, B="Get (a copy)"(Array=@n111.VideoList_31_#12), bPickA=@n112.ReturnValue)
    MyIndex:int=@n109
    SubUVResX:int=@n111.X_25_#10
    SubUVResY:int=@n111.Y_27_#11
   ) -> (then, ReturnValue:EUO Preview Item Object Object Reference)
   > Add(TargetArray=Get CurrentPageObjectList(), NewItem=@n114)
   > Initial(self=@n114)
   > AddItem(self=Get TileView_218(), Item=@n114)
   n115 = UI_ChangeTagListHide!() -> (OutputDelegate, then, Tag:String, Add:bool=~) {is=title K2Node_CustomEvent}
   > n116 = Branch(Condition=@n115.Add)
   n117 = SelectString(A=Favmark114514, B=@n115.Tag, bPickA="Equal Exactly (String)"(A=@n115.Tag, B="❤❤❤❤❤Favorite❤❤❤❤❤"))
   n118 = Get CurrentHideTag() -> (CurrentHideTag:Array of Strings)
   > n119 = Add Unique(TargetArray:Array of Strings=@n118, NewItem:"String (by ref)"=@n117)
   RetriggerableDelay(execute=[@n120.then, @n119.then, @n127.then, @n126.then], Duration=0.5)
   > ApplyTags() {is=title K2Node_CallFunction}
   n120 = Remove Item(execute=@n116.else, TargetArray=@n118, Item=@n117)
   n121 = Rebuild Database() {is=title K2Node_CustomEvent}
   n122 = UI_ChangeTagListShow!() -> (OutputDelegate, then, Tag:String, Add:bool=~) {is=title K2Node_CustomEvent}
   > n123 = Branch(Condition=@n122.Add)
   n124 = SelectString(A=Favmark114514, B=@n122.Tag, bPickA="Equal Exactly (String)"(A=@n122.Tag, B="❤❤❤❤❤Favorite❤❤❤❤❤"))
   n125 = Get CurrentShowTag() -> (CurrentShowTag:Array of Strings)
   > n126 = Add Unique(TargetArray=@n125, NewItem=@n124)
   n127 = Remove Item(execute=@n123.else, TargetArray=@n125, Item=@n124)
   n128 = SetUpLibPath!() -> (OutputDelegate, then, Path:String) {is=title K2Node_CustomEvent}
   > n129 = Set LibraryBasePath(LibraryBasePath=@n128.Path)
   DB_BuildTagOnly() {is=title K2Node_CustomEvent}
   n130 = Get DataBaseSaveGame!() -> (DataBaseSaveGame)
   > Is Valid(InputObject=@n130)
   > Clear(TargetArray:Array of Strings=Get DataBase_TagList(self=@n130))
   > Clear(TargetArray:Array of St FXB Tag 2Item Array Structures=Get DataBase_Tag2ItemArray(self=@n130))
   > n131 = Sequence!(execute) -> (then_0, then_1, then_2, then_3:Exec, then_4:Exec)
   > n132 = GetFileListInPath("Path(/end)"=Append(A=Get LibraryBasePath(), B=Tag/), ext=txt, __WorldContext=Self-Reference())
   > n133 = For Each Loop(Array:Array of Strings=@n132.FullPathList) -> (Array Element:String)
   > n134 = Read File to String(FilePath=@n133, bUseFullFilePath=true)
   > n135 = For Each Loop(Array=ParseIntoArray(SourceString=@n134, Delimiter="\r\n"))
   n136 = Split(SourceString=@n135, InStr="%&&%", SearchDir=FromStart)
   n137 = Find Item(TargetArray=Get tmpRefreshTag_ReadedTagList() -> (tmpRefreshTag_ReadedTagList:Array of Strings), ItemToFind=@n136)
   > n138 = Branch(Condition="integer >= integer"(A=@n137))
   n139 = ParseIntoArray(SourceString=@n136.RightS, Delimiter="%@%")
   > n140 = For Each Loop(Array=@n139)
   n141 = Get tmpRefreshTag_ReadedTagItemList() -> (tmpRefreshTag_ReadedTagItemList:Array of St FXB String Array Structures)
   > Add Unique(
    TargetArray=Break St FXB String Array(St_FXB_StringArray="Get (a ref)"(Array:Array of St FXB String Array Structures=@n141, Dimension 1=@n137) -> (Output:"St FXB String Array Structure (by ref)"))
    NewItem=@n140
   )
   Add(execute=@n138.else, TargetArray:Array of Strings=Get tmpRefreshTag_ReadedTagList(), NewItem:"String (by ref)"=@n136)
   > Add(
    TargetArray:Array of St FXB String Array Structures=Get tmpRefreshTag_ReadedTagItemList()
    NewItem:"St FXB String Array Structure (by ref)"=Make St FXB String Array(String_3_#1=@n139)
   )
   n142 = Get DataBase_ItemList(self=Get DataBaseSaveGame())
   n143 = For Each Loop(Exec=@n131.then_1, Array:Array of St FXB Item List Structures=@n142) -> (Array Element:St FXB Item List Structure)
   n144 = Break St FXB Item List(St_FXB_ItemList=@n143)
   > Set tmpRefreshTag_TmpItemListData(tmpRefreshTag_TmpItemListData:Array of St FXB Item Structures=@n144) -> (Output_Get:Array of St FXB Item Structures)
   > n145 = For Each Loop(Array:Array of St FXB Item Structures=@n144) -> (Array Element:St FXB Item Structure)
   Set Array Elem(
    execute=@n145.Completed
    TargetArray:Array of St FXB Item List Structures=@n142
    Index=@n143.Array Index
    Item:"St FXB Item List Structure (by ref)"=Make St FXB Item List(ItemList_16_#4=Get tmpRefreshTag_TmpItemListData() -> (tmpRefreshTag_TmpItemListData:Array of St FXB Item Structures))
    bSizeToFit=false
   )
   n146 = Set tmpRefreshTag_TmpItem(execute=@n145.LoopBody, tmpRefreshTag_TmpItem:St FXB Item Structure=@n145) -> (Output_Get:St FXB Item Structure)
   > Clear(TargetArray:Array of Strings=Get tmpRefreshTag_TagArray() -> (tmpRefreshTag_TagArray:Array of Strings))
   > n147 = For Each Loop(Array:Array of St FXB String Array Structures=Get tmpRefreshTag_ReadedTagItemList()) -> (Array Element:St FXB String Array Structure)
   n148 = Set members in St FXB Item(execute=@n147.Completed, StructRef=@n146, Tags_54_#17=Get tmpRefreshTag_TagArray())
   > Set Array Elem(
    TargetArray:Array of St FXB Item Structures=Get tmpRefreshTag_TmpItemListData()
    Index=@n145.Array Index
    Item:"St FXB Item Structure (by ref)"=@n148
    bSizeToFit=false
   )
   > n149 = For Each Loop(Array:Array of Strings=Get tmpRefreshTag_TagArray()) -> (Array Element:String)
   n150 = Get DataBaseSaveGame()
   n151 = Get DataBase_TagList(self=@n150)
   n152 = Find Item(TargetArray=@n151, ItemToFind=@n149)
   > n153 = Branch(Condition="integer < integer"(A=@n152))
   > Add(TargetArray:Array of Strings=@n151, NewItem:"String (by ref)"=@n149)
   n154 = Break St FXB Item(St_FXB_Item=@n146) -> (HasValidFX_22_#9=~)
   n155 = Make St FXB Tag 2Item(
    ItemEditorName_3_#20=@n154.EditorName_14_#6
    ItemCategoryIndex_7_#21=@n154.MyCategoryIndexInDataBase_61_#19
    ItemIndex_9_#22=@n154.MyIndexInDatabase_58_#18
   )
   n156 = Get DataBase_Tag2ItemArray(self=@n150)
   > Add(
    TargetArray:Array of St FXB Tag 2Item Array Structures=@n156
    NewItem:"St FXB Tag 2Item Array Structure (by ref)"=Make St FXB Tag 2Item Array(
     Tag2ItemList_3_#2=Make Array("[0]":St FXB Tag 2Item Structure=@n155) -> (Array:Array of St FXB Tag 2Item Structures)
     ItemNameList_6_#3=Make Array("[0]":String=@n154.EditorName_14_#6) -> (Array:Array of Strings)
    )
   )
   n157 = "Get (a ref)"(Array:Array of St FXB Tag 2Item Array Structures=@n156, Dimension 1=@n152) -> (Output:"St FXB Tag 2Item Array Structure (by ref)")
   n158 = Break St FXB Tag 2Item Array(St_FXB_Tag2ItemArray=@n157)
   n159 = Set tmpRefreshTag_tmpTag2ItemArray(execute=@n153.else, tmpRefreshTag_tmpTag2ItemArray:Array of St FXB Tag 2Item Structures=@n158) -> (Output_Get:Array of St FXB Tag 2Item Structures)
   > n160 = Set tmpRefreshTag_tmpItemNameList(tmpRefreshTag_tmpItemNameList:Array of Strings=@n158.ItemNameList_6_#3) -> (Output_Get:Array of Strings)
   > Add(TargetArray:Array of St FXB Tag 2Item Structures=@n159, NewItem:"St FXB Tag 2Item Structure (by ref)"=@n155)
   > Add(TargetArray:Array of Strings=@n160, NewItem:"String (by ref)"=@n154.EditorName_14_#6)
   > n161 = Set members in St FXB Tag 2Item Array(StructRef=@n157, Tag2ItemList_3_#2=@n159, ItemNameList_6_#3=@n160)
   > "Set St FXB Tag 2Item Array Structure (by ref)"(Target=@n157, Value=@n161)
   n162 = Break St FXB Item(St_FXB_Item=@n146) -> (HasValidFX_22_#9=~)
   n163 = Break St FXB String Array(St_FXB_StringArray=@n147)
   Branch(
    execute=@n147.LoopBody
    Condition="integer >= integer"(A=Find Item(TargetArray=@n163, ItemToFind=@n162.EditorName_14_#6))
   )
   > Add Unique(TargetArray=Get tmpRefreshTag_TagArray(), NewItem="Get (a copy)"(Array=Get tmpRefreshTag_ReadedTagList(), Dimension 1=@n147.Array Index))
   PrintString(execute=@n131.then_3, InString=Finished)
   SaveGameToSlot(execute=@n131.then_4, SaveGameObject=Get DataBaseSaveGame(), SlotName=FXBrowser, UserIndex=0)
   n164 = Event Construct()
   Event Destruct()
   > ClearListItems(self=Get TileView_218())
   > Clear(TargetArray:Array of EUO Preview Item Object Object References=Get CurrentPageObjectList())
   > Set LastShowSource(LastShowSource:EUW Preview Button Object Reference) -> (Output_Get:EUW Preview Button Object Reference)
   > Close(self=Get MediaPlayer())
   Event Tick()
   n165 = Get RightPanel()
   n166 = GetAbsoluteSize!(Geometry:"Geometry Structure (by ref)"=GetTickSpaceGeometry!(self:Widget Object Reference=@n165) -> (ReturnValue:"Geometry Structure (by ref)")) -> (ReturnValue:Vector 2D Structure)
   > SetHeightOverride(self=Get SizeBox_3() -> (SizeBox_3:Size Box Object Reference), InHeightOverride=float * float(A=BreakVector2D(InVec=@n166), B=0.82))
   n167 = Branch(Condition="integer > integer"(A=FindSubstring(SearchIn=@n66, Substring="\\", bSearchFromEnd=true)))
   n168 = Append(A=@n66, B="\\")
   > Set LibraryBasePath(LibraryBasePath=@n168)
   > n169 = Set Text(Text="To Text (String)"(InString=@n168), self=Get LibPath())
   n170 = Set LibraryBasePath(execute=@n167.else, LibraryBasePath=@n66)
   SetVisibility(execute=@n180.then_0, self=Get RightPanel(), InVisibility=SelfHitTestInvisible)
   n171 = Get VideoImage() -> (VideoImage:Image Object Reference)
   SetBrushFromMaterial(
    execute=@n178.else
    self=@n171
    Material=Get MyPreviewMat!(self:EUW Preview Button Object Reference=@n11.SourceButton) -> (MyPreviewMat:Material Instance Dynamic Object Reference)
   )
   Scroll Item Into View!(execute=@n180.then_1, self:List View Object Reference=Get TileView_218(), Item:Object Reference=Get MyItemObject(self=Get LastShowSource())) -> (then)
   tmpArrangeTag()
   UI_InitialRightTagPanel(execute=@n182.then)
   GetDesiredSize!(self:Widget Object Reference=@n165) -> (ReturnValue:Vector 2D Structure)
   ReadFolders() {is=title K2Node_CustomEvent}
   Event PreConstruct()
   Reroute Node(InputPin:Wildcard) -> (OutputPin:Wildcard)
   n172 = Find(TargetMap=Get CurrentLoadItemRuler(), Key=@n13.Category_48_#15)
   n173 = Find Item(TargetArray=@n13.VideoList_31_#12, ItemToFind=@n172)
   n174 = Break St FXB Color Array!(St_FXB_ColorArray:"St FXB Color Array Structure (by ref)"="Get (a copy)"(
     Array:Array of St FXB Color Array Structures=@n13.ColorList_39_#13
     Dimension 1=SelectInt(A=@n173, B=0, bPickA=@n172.ReturnValue)
    ) -> (Output:St FXB Color Array Structure)) -> (Array_9_#23:Array of Linear Color Structures) {is=title K2Node_BreakStruct}
   n175 = "To Text (String)"(InString=@n13.PublishName_16_#7)
   n176 = Append!(A=@n11.Folder, B=@n11.LoadRuler, C:String=.mp4) -> (ReturnValue)
   n177 = FileExists!(InPath:String=@n176) -> (ReturnValue:bool)
   RetriggerableDelay(execute=@n181.then, Duration=0.1)
   > n178 = Branch(Condition=@n177)
   > SetBrushFromMaterial(self=@n171, Material=/FXBrowserUtilities/FXRefLib/Finder/FXRefLibMedia_Video_Mat)
   > n179 = OpenFile!(execute, self:Media Player Object Reference=@n9, FilePath:String=@n176) -> (then, ReturnValue:bool)
   SwitchButtonBoard!(execute=@n183.else, self:EUW Preview Button Object Reference=Get LastShowSource(), ShowBoard:bool=false) -> (then)
   > Set LastShowSource(LastShowSource=@n11.SourceButton)
   > SwitchButtonBoard(self=@n11.SourceButton, ShowBoard=true)
   > n180 = Sequence!(execute) -> (then_0, then_1, then_2, then_3)
   n181 = Close(execute=@n180.then_2, self=@n9)
   Set Text(
    execute=@n180.then_3
    Text="To Text (String)"(InString=@n13.EditorName_14_#6)
    self:Text Object Reference=Get ItemNameBox() -> (ItemNameBox:Text Object Reference)
   )
   > Set Text(Text=@n175, self=Get PublishName() -> (PublishName:Text Object Reference))
   > Set BrushColor!(
    execute
    BrushColor:Linear Color Structure="Get (a copy)"(Array:Array of Linear Color Structures=@n174) -> (Output:Linear Color Structure)
    self:Border Object Reference=Get 1stColor() -> (1stColor:Border Object Reference)
   ) -> (then, Output_Get:Linear Color Structure)
   > Set BrushColor(BrushColor="Get (a copy)"(Array=@n174, Dimension 1=1), self=Get 2ndColor() -> (2ndColor:Border Object Reference))
   > Set BrushColor(BrushColor="Get (a copy)"(Array=@n174, Dimension 1=2), self=Get 3rdColor() -> (3rdColor:Border Object Reference))
   > Set BrushColor(BrushColor="Get (a copy)"(Array=@n174, Dimension 1=3), self=Get 4thColor() -> (4thColor:Border Object Reference))
   > n182 = Set CurrentPlayItem(CurrentPlayItem:St FXB Item Structure=@n12) -> (Output_Get:St FXB Item Structure)
   n183 = Branch(execute=@n11.then, Condition="Equal (Object)"(A=@n11.SourceButton, B=Get LastShowSource()))
  ]}
 GetText {nodes [
   GetText() {is=title K2Node_FunctionEntry}
   > Return Node!(execute, ReturnValue:Text="To Text (String)"(InString=Get CurrentUser() -> (CurrentUser:String)))
  ]}
 GetText_0 {nodes [
   GetText_0() {is=title K2Node_FunctionEntry}
   > Return Node(ReturnValue="To Text (Integer)"(Value=Get ColorFT() -> (ColorFT:int=0)))
  ]}
 Get_ImageSize_6_Text {nodes [
   Get_ImageSize_6_Text() {is=title K2Node_FunctionEntry}
   n1 = "To String (Integer)"(InInt=Get MinVersion() -> (MinVersion=0))
   n2 = "To String (Integer)"(InInt=Get MaxVersion() -> (MaxVersion=0))
   n3 = Get Search() -> (Search:String)
   > Return Node(ReturnValue="To Text (String)"(InString=Append!(A="MinVer : ", B=@n1, C=" MaxVer : ", D:String=@n2, E:String=" Search : ", F:String=@n3, G:String=" Color : ") -> (ReturnValue)))
  ]}
 Get_PageNum_Text {nodes [
   Get_PageNum_Text() {is=title K2Node_FunctionEntry}
   n1 = "To String (Integer)"(InInt=int + int(A=Get CurrentPage() -> (CurrentPage=0)))
   n2 = Length(TargetArray=Get CurrentPendingLoadItemList())
   n3 = Get LoadPageItemNumber() -> (LoadPageItemNumber=0)
   n4 = "To String (Integer)"(InInt=int + int(A=int / int(A=@n2, B=@n3)))
   > Return Node(ReturnValue="To Text (String)"(InString=Append!(A=@n1, B=/, C=@n4) -> (ReturnValue)))
  ]}
 Test_GetTagFromAssetPath {nodes [
   Test_GetTagFromAssetPath!() -> (then, AssetPath:Array of Strings) {is=title K2Node_FunctionEntry}
   > Return Node!(
    execute
    Tags:Array of Strings=Select!(
     Option 0:Array of Strings=Make Array("[0]"=Fire)
     Option 1:Array of Strings=Make Array("[0]"=Ice)
     Option 2:Array of Strings=Make Array("[0]"=Wind)
     Index:int=RandomInteger!(Max:int=3) -> (ReturnValue:int=0)
    ) -> (ReturnValue:Array of Strings)
   )
  ]}
 UI_HandleCategoryChange {nodes [
   n1 = UI_HandleCategoryChange!() -> (then, CategoryList:Array of Strings, CheckedList:Array of Booleans, Shuffle:bool=~) {is=title K2Node_FunctionEntry}
   > Set CurrentCategoryList(CurrentCategoryList:Array of Strings=@n1) -> (Output_Get:Array of Strings)
   > Set CurrentCheckedList(CurrentCheckedList:Array of Booleans=@n1.CheckedList) -> (Output_Get:Array of Booleans)
   > Set CurrentPage(CurrentPage=0) -> (Output_Get=0)
   > Clear(TargetArray:Array of St FXB Item Structures=Get CurrentPendingLoadItemList())
   > Clear(TargetMap=Get CurrentLoadItemRuler()) {is=title K2Node_CallFunction}
   n2 = Get DataBaseSaveGame()
   > n3 = Keys(TargetMap=Get DataBase_Category(self=@n2))
   > n4 = For Each Loop(Array=@n1)
   n5 = Branch(execute=@n4.Completed, Condition=@n1.Shuffle)
   > n6 = Shuffle!(execute, TargetArray:Array of St FXB Item Structures=Get tmpReadyCheckItemList() -> (tmpReadyCheckItemList:Array of St FXB Item Structures)) -> (then) {is=title K2Node_CallArrayFunction}
   n7 = For Each Loop(Exec=[@n6.then, @n5.else], Array:Array of St FXB Item Structures=Get tmpReadyCheckItemList()) -> (Array Element:St FXB Item Structure)
   Append Array!(
    execute=@n7.Completed
    TargetArray:Array of St FXB Item Structures=Get CurrentPendingLoadItemList()
    SourceArray:Array of St FXB Item Structures=Get tmpAppendItemList() -> (tmpAppendItemList:Array of St FXB Item Structures)
   ) -> (then)
   > PrintString(InString="========================EndF")
   > PrintString(InString=CurrentItemList)
   n8 = Get CurrentPendingLoadItemList()
   > PrintString(InString="To String (Integer)"(InInt=Length(TargetArray=@n8)))
   > n9 = Sequence!(execute) -> (then_0, then_1, then_2)
   n10 = Get tmps() -> (tmps:String)
   > Set tmps(tmps:String=Append!(A=@n10, B="Solo Tags : ") -> (ReturnValue)) -> (Output_Get:String)
   > n11 = For Each Loop(Array:Array of Strings=Get CurrentShowTag()) -> (Array Element:String)
   Set tmps(execute=@n11.Completed, tmps=Append!(A=@n10, B="     ||     ", C:String="Hide Tags : ") -> (ReturnValue))
   > n12 = For Each Loop(Array=Get CurrentHideTag())
   n13 = Replace(SourceString=Get tmps(), From="NeedCheck,")
   Set Text(
    execute=@n12.Completed
    Text="To Text (String)"(InString=Replace(SourceString=@n13, From=Favmark114514, To="❤❤❤❤❤Favorite❤❤❤❤❤"))
    self=Get TagShow() -> (TagShow:Text Object Reference)
   )
   Set tmps(execute=@n12.LoopBody, tmps=Append(A=@n10, B=@n12, C=", "))
   Set tmps(execute=@n11.LoopBody, tmps=Append(A=@n10, B=@n11, C=", "))
   UI_LoadCurrentPage(execute=@n9.then_1) {is=title K2Node_CallFunction}
   n14 = Break St FXB Item(St_FXB_Item=@n7) -> (HasValidFX_22_#9=~)
   n15 = Branch(execute=@n7.LoopBody, Condition="integer > integer"(A=Length(TargetArray:Array of Strings=@n14.Tags_54_#17)))
   > Set tmpFindInHideTag(tmpFindInHideTag:bool=false) -> (Output_Get:bool)
   > Set tmpFindInShowTag(tmpFindInShowTag:bool=false) -> (Output_Get:bool)
   n16 = Get CurrentShowTag()
   > n17 = Branch(Condition="integer > integer"(A=Length(TargetArray=@n16)))
   > n18 = For Each Loop(Array=@n14.Tags_54_#17)
   n19 = For Each Loop(Exec=[@n18.Completed, @n41.then], Array=@n14.Tags_54_#17)
   Branch(execute=@n19.Completed, Condition=Get tmpFindInShowTag() -> (tmpFindInShowTag:bool))
   > n20 = Branch(Condition=Get tmpFindInHideTag() -> (tmpFindInHideTag:bool))
   n21 = Get MinVersion() -> (MinVersion=0)
   n22 = Get MaxVersion() -> (MaxVersion=0)
   Branch(
    execute=[@n43.else, @n20.else]
    Condition=AND Boolean!(
     A="integer >= integer"(A=@n14.PublishVersion_20_#8, B=@n21)
     B="integer <= integer"(A=@n14.PublishVersion_20_#8, B=@n22)
    ) -> (ReturnValue)
   )
   > n23 = Branch(Condition="Not Equal (LinearColor)"(A:Linear Color Structure=Get ColorFApply(), B:Linear Color Structure=(R=0,G=0,B=0,A=1)) -> (ReturnValue:bool=~) {is=title K2Node_PromotableOperator})
   > Set bFoundColor(bFoundColor:bool=false) -> (Output_Get:bool)
   n24 = Break St FXB Item(St_FXB_Item=@n7) -> (HasValidFX_22_#9=~)
   n25 = Find(TargetMap=Get CurrentLoadItemRuler(), Key=@n24.Category_48_#15)
   n26 = Select!(
    Option 0:int=0
    Option 1:int=Find Item(TargetArray=@n24.VideoList_31_#12, ItemToFind=@n25)
    Index:bool=@n25.ReturnValue
   ) -> (ReturnValue:int)
   > n27 = For Each Loop(Array:Array of Linear Color Structures=Break St FXB Color Array(St_FXB_ColorArray="Get (a copy)"(Array:Array of St FXB Color Array Structures=@n24.ColorList_39_#13, Dimension 1=@n26) -> (Output:St FXB Color Array Structure))) -> (Array Element:Linear Color Structure)
   n28 = Branch(execute=@n27.Completed, Condition=Get bFoundColor() -> (bFoundColor:bool))
   Set CheckTagS(execute=[@n23.else, @n28.then], CheckTagS:String) -> (Output_Get:String)
   > n29 = For Each Loop(Array:Array of Strings=@n24.Tags_54_#17) -> (Array Element:String)
   n30 = Get CurrentLoadItemRuler()
   n31 = FileExists(InPath=Append!(
     A=@n24.FolderPath_50_#16
     B
     C=Get Key Value By Index!(TargetMap:Map of Strings to Strings=@n30, Index:int=0) -> (Key:String, Value:String) {is=title K2Node_CallFunction}.Value
     D:String=.mp4
    ) -> (ReturnValue))
   n32 = Replace(SourceString=Get Search(), From=" ", SearchCase=CaseSensitive)
   n33 = Get CheckTagS() -> (CheckTagS:String)
   n34 = FindSubstring(SearchIn=@n24.Depend_38_#14, Substring=@n32)
   n35 = OR Boolean!(
    A="integer >= integer"(A=FindSubstring(SearchIn=@n24.EditorName_14_#6, Substring=@n32))
    B="integer >= integer"(A=FindSubstring(SearchIn=@n24.PublishName_16_#7, Substring=@n32))
    C:bool=Select(Option 0:bool=false, Option 1:bool="integer >= integer"(A=@n34), Index=Get SearchDepend() -> (SearchDepend:bool)) -> (ReturnValue:bool)
    D:bool="integer >= integer"(A=FindSubstring(SearchIn=@n24.Category_48_#15, Substring=@n32))
    E:bool="integer >= integer"(A=FindSubstring(SearchIn=@n33, Substring=@n32))
    F:bool=IsEmpty(InString=@n32)
    G:bool
   ) -> (ReturnValue)
   n36 = Length!(TargetMap:Map of Strings to Strings=@n30) -> (ReturnValue:int=0) {is=title K2Node_CallFunction}
   Branch(execute=@n29.Completed, Condition=Select(Option 0=@n35, Option 1=AND Boolean(A=@n35, B=@n31), Index="integer > integer"(A=@n36)))
   > Add(TargetArray:Array of St FXB Item Structures=Get tmpAppendItemList(), NewItem:"St FXB Item Structure (by ref)"=@n7)
   Set CheckTagS(execute=@n29.LoopBody, CheckTagS=Append!(A=Get CheckTagS(), B=@n29) -> (ReturnValue))
   n37 = "RGB to HSV (LinearColor)"!(RGB:Linear Color Structure=Get ColorFApply()) -> (ReturnValue:Linear Color Structure)
   n38 = "RGB to HSV (LinearColor)"(RGB=@n27)
   Branch(
    execute=@n27.LoopBody
    Condition="Nearly Equal (Float)"(
     A=Break Linear Color!(LinearColor:"Linear Color Structure (by ref)"=@n38) -> (R:float=~, G:float=~, B:float=~, A:float=~) {is=title K2Node_BreakStruct}
     B=Break Linear Color(LinearColor=@n37) -> (R=~, G=~, B=~, A=~)
     ErrorTolerance=30
    )
   )
   > Set bFoundColor(bFoundColor=true)
   n39 = Get CurrentHideTag()
   Branch(execute=@n19.LoopBody, Condition="integer >= integer"(A=Find Item(TargetArray=@n39, ItemToFind=@n19)))
   > Set tmpFindInHideTag(tmpFindInHideTag=true)
   n40 = Get CurrentShowTag()
   Branch(execute=@n18.LoopBody, Condition="integer >= integer"(A=Find Item(TargetArray=@n40, ItemToFind=@n18)))
   > Set tmpFindInShowTag(tmpFindInShowTag=true)
   n41 = Set tmpFindInShowTag(execute=@n17.else, tmpFindInShowTag=true)
   n42 = Get CurrentShowTag()
   n43 = Branch(execute=@n15.else, Condition="integer > integer"(A=Length(TargetArray=@n42) {is=title K2Node_CallArrayFunction}))
   n44 = Branch(execute=@n4.LoopBody, Condition="Equal (Integer)"(A=@n4.Array Index))
   n45 = "Get (a ref)"(Array:Array of Booleans=@n1.CheckedList, Dimension 1=@n4.Array Index) -> (Output:"Boolean (by ref)")
   > Set tmpAll(tmpAll:bool=@n45) -> (Output_Get:bool)
   Branch(execute=@n44.else, Condition=OR Boolean!(A=Get tmpAll() -> (tmpAll:bool), B=@n45) -> (ReturnValue))
   > n46 = Branch(Condition="integer >= integer"(A=FindSubstring(SearchIn=@n4, Substring=-)))
   n47 = Split(SourceString=@n4, InStr=-, SearchDir=FromStart)
   > Add(TargetMap:Map of Strings to Strings=Get CurrentLoadItemRuler(), Key=@n47, Value:"String (by ref)"=@n47.RightS) {is=title K2Node_CallFunction}
   PrintString(execute=@n46.else, InString="!!!!!!!!!!!!!!!!!!StartF")
   > Clear(TargetArray=Get tmpAppendItemList()) {is=title K2Node_CallArrayFunction}
   n48 = Find Item(TargetArray=@n3, ItemToFind=@n4)
   n49 = Get DataBase_ItemList(self=@n2)
   > Append Array(
    TargetArray=Get tmpReadyCheckItemList()
    SourceArray=Break St FXB Item List(St_FXB_ItemList="Get (a ref)"(Array:Array of St FXB Item List Structures=@n49, Dimension 1=@n48) -> (Output:"St FXB Item List Structure (by ref)"))
   )
   Clear(TargetArray=Get tmpReadyCheckItemList())
  ]}
 UI_InitialCategory {nodes [
   UI_InitialCategory() {is=title K2Node_FunctionEntry}
   n1 = Get DataBase_Category(self=Get DataBaseSaveGame())
   > n2 = Keys(TargetMap=@n1)
   > n3 = Values!(execute, TargetMap:Map of Strings to Strings=@n1) -> (then, Values:Array of Strings)
   > n4 = For Each Loop(Array=@n2)
   n5 = Get NewLocalVar() -> (NewLocalVar:Array of Strings)
   Append Array(execute=@n4.Completed, TargetArray:Array of Strings=@n5, SourceArray:Array of Strings=Get CL() -> (CL:Array of Strings))
   > n6 = Create Widget!(
    execute:Exec
    WorldContextObject:Object Reference
    Class:User Widget Class Reference=&/BUW_Category.BUW_Category_C
    OwningPlayer:Player Controller Object Reference
    CategoryList:Array of Strings=@n5
   ) -> (then:Exec, ReturnValue:BUW Category Object Reference)
   > AddChild(self=Get CosmeticCategorySlot(), Content=@n6)
   > Bind Event to Category Changed(self:BUW Category Object Reference=@n6, Delegate=Create Event()) {is=title K2Node_AddDelegate}
   Add(execute=@n4.LoopBody, TargetArray:Array of Strings=Get CL(), NewItem:"String (by ref)"=@n4) {is=title K2Node_CallArrayFunction}
   n7 = ParseIntoArray(SourceString="Get (a ref)"(Array:Array of Strings=@n3, Dimension 1=@n4.Array Index) -> (Output:"String (by ref)"), Delimiter=%)
   > Branch(Condition="integer >= integer"(A=Length(TargetArray=@n7), B=2))
   > n8 = For Each Loop(Array=@n7)
   > Add(TargetArray=Get CL(), NewItem=Append!(A=@n4, B=-, C:String=@n8) -> (ReturnValue))
  ]}
 UI_InitialRightTagPanel {nodes [
   UI_InitialRightTagPanel() {is=title K2Node_FunctionEntry}
   > ClearListItems(self=Get RightTagButton() -> (RightTagButton:Tile View Object Reference))
   n1 = "To String (Text)"(InText=Get Text(self=Get EditorUtilityEditableTextBox_413()))
   > n2 = Branch(Condition=IsEmpty(InString=@n1))
   n3 = Break St FXB Item(St_FXB_Item=Get CurrentPlayItem()) -> (HasValidFX_22_#9=~)
   n4 = Branch(execute=[@n2.then, @n25.Completed], Condition=IsEmpty(InString=@n3.EditorName_14_#6))
   > n5 = For Each Loop(Array=Get DataBase_TagList(self=Get DataBaseSaveGame()))
   > n6 = Branch(Condition="Equal Exactly (String)"(A=@n5, B=Favmark114514))
   n7 = Get ProArray() -> (ProArray:Array of Strings)
   n8 = Branch(execute=@n6.else, Condition="integer >= integer"(A=Find Item(TargetArray=@n7, ItemToFind=@n5)))
   n9 = Self-Reference()
   n10 = Construct Object from Class!(
    execute=@n8.else
    Class=&/EUO_TagButtonWObject.EUO_TagButtonWObject_C
    self=@n9
    ParentWidget=@n9
    H:bool=False
    T:String=@n5
   ) -> (then, ReturnValue:EUO Tag Button WObject Object Reference)
   > AddItem(self=Get RightTagButton(), Item=@n10)
   n11 = For Each Loop(Exec=@n4.else, Array=@n3.Tags_54_#17)
   n12 = For Each Loop(Exec=@n11.Completed, Array=Get DataBase_TagList(self=Get DataBaseSaveGame()))
   > n13 = Branch(Condition="Equal Exactly (String)"(A=@n12, B=Favmark114514))
   n14 = Get ProArray()
   n15 = Branch(execute=@n13.else, Condition="integer >= integer"(A=Find Item(TargetArray=@n14, ItemToFind=@n12)))
   n16 = Get NewLocalVar()
   n17 = Branch(execute=@n15.else, Condition="integer >= integer"(A=Find Item(TargetArray=@n16, ItemToFind=@n12)))
   n18 = Self-Reference()
   n19 = Construct Object from Class(execute=@n17.else, Class=&EUO_TagButtonWObject_C, self=@n18, ParentWidget=@n18, H=False, T=@n12)
   > AddItem(self=Get RightTagButton(), Item=@n19)
   n20 = Branch(execute=@n11.LoopBody, Condition="Equal Exactly (String)"(A=@n11, B=Favmark114514))
   n21 = Get ProArray()
   n22 = Branch(execute=@n20.else, Condition="integer >= integer"(A=Find Item(TargetArray=@n21, ItemToFind=@n11)))
   Add(execute=@n22.else, TargetArray=@n16, NewItem=@n11)
   n23 = Self-Reference()
   > n24 = Construct Object from Class(Class=&EUO_TagButtonWObject_C, self=@n23, ParentWidget=@n23, H=true, T=@n11)
   > AddItem(self=Get RightTagButton(), Item=@n24)
   n25 = For Each Loop(Exec=@n2.else, Array=Get DataBase_TagList(self=Get DataBaseSaveGame()))
   > n26 = Branch(Condition="Equal Exactly (String)"(A=@n25, B=Favmark114514))
   Branch(execute=@n26.else, Condition="integer >= integer"(A=FindSubstring(SearchIn=@n25, Substring=@n1, bUseCase=true)))
   > Add(TargetArray=Get ProArray(), NewItem=@n25)
   n27 = Self-Reference()
   > n28 = Construct Object from Class(
    Class=&EUO_TagButtonWObject_C
    self=@n27
    ParentWidget=@n27
    H="integer >= integer"(A=Find Item(TargetArray=@n3.Tags_54_#17, ItemToFind=@n25))
    T=@n25
   )
   > AddItem(self=Get RightTagButton(), Item=@n28)
  ]}
 UI_InitialTag {nodes [
   UI_InitialTag() {is=title K2Node_FunctionEntry}
   > ClearChildren(self=Get TagList() -> (TagList:Vertical Box Object Reference))
   > n1 = Set TmpList(TmpList:Array of Strings=Get DataBase_TagList(self=Get DataBaseSaveGame())) -> (Output_Get:Array of Strings)
   > Remove Item(TargetArray=@n1, Item=Favmark114514)
   > n2 = Branch(Condition=Get SystemUser())
   Insert!(execute=[@n9.then, @n2.then], TargetArray=@n1, NewItem=Favmark114514, Index:int=0) -> (then) {is=title K2Node_CallArrayFunction}
   > n3 = For Each Loop(Array=Get TmpList() -> (TmpList:Array of Strings))
   n4 = Self-Reference()
   > n5 = Create Widget!(
    execute
    WorldContextObject=@n4
    Class=&/EUW_TagButton.EUW_TagButton_C
    OwningPlayer
    Tag:String=SelectString(A="❤❤❤❤❤Favorite❤❤❤❤❤", B=@n3, bPickA="Equal Exactly (String)"(A=@n3, B=Favmark114514))
    ParentFXB:BUW FXBrowser Object Reference=@n4
   ) -> (then, ReturnValue:EUW Tag Button Object Reference)
   > n6 = Sequence()
   n7 = Get CurrentShowTag()
   > Branch(Condition="integer >= integer"(A=Find Item(TargetArray=@n7, ItemToFind=@n3)))
   > SetIsChecked!(
    execute
    self:Check Box Object Reference=Get CheckBox_Show!(self:EUW Tag Button Object Reference=@n5) -> (CheckBox_Show:Check Box Object Reference)
    InIsChecked:bool=true
   ) -> (then)
   n8 = Get CurrentHideTag()
   Branch(execute=@n6.then_1, Condition="integer >= integer"(A=Find Item(TargetArray=@n8, ItemToFind=@n3)))
   > SetIsChecked(self=Get CheckBox_Hide!(self:EUW Tag Button Object Reference=@n5) -> (CheckBox_Hide:Check Box Object Reference), InIsChecked=true)
   AddChild(execute=@n6.then_2, self=Get TagList(), Content=@n5)
   Remove Item(execute=@n2.else, TargetArray=@n1, Item=NeedCheck)
   > n9 = Remove Item(TargetArray=@n1, Item=NeedCaptureAgain)
  ]}
 tmpArrangeTag {nodes [
   tmpArrangeTag() {is=title K2Node_FunctionEntry}
   n1 = Get DataBaseSaveGame()
   > n2 = For Each Loop(Array=Get DataBase_TagList(self=@n1))
   > Set tmpstring(tmpstring:String=Append!(A=@n2, B="%&&%") -> (ReturnValue)) -> (Output_Get:String)
   n3 = Get DataBase_Tag2ItemArray(self=@n1)
   > n4 = For Each Loop(Array=Break St FXB Tag 2Item Array(St_FXB_Tag2ItemArray="Get (a ref)"(Array:Array of St FXB Tag 2Item Array Structures=@n3, Dimension 1=@n2.Array Index) -> (Output:"St FXB Tag 2Item Array Structure (by ref)")).ItemNameList_6_#3)
   n5 = Get tmpstring() -> (tmpstring:String)
   WriteStringIntoFile!(
    execute=@n4.Completed
    WriteString:String=@n5
    FilePath:String=Append!(A=Get LibraryBasePath(), B=Tag/, C:String=yi.xie.tag.txt) -> (ReturnValue)
    Unique:bool=true
    Clear:bool=false
    __WorldContext:Object Reference=Self-Reference()
   ) -> (then)
   Set tmpstring(execute=@n4.LoopBody, tmpstring=Append(A=@n5, B=@n4, C="%@%"))
  ]}
]
variables [
 ColorF {type=LinearColor, default=<rgba 1 0 0 1>}
 ColorFApply {type=LinearColor, default=<rgba 0 0 0 1>}
 ColorFT {type=integer, default=30}
 CurrentCategoryList {type="array<string>", default []}
 CurrentCheckedList {type="array<boolean>", default []}
 CurrentHideTag {type="array<string>", default [NeedCheck]}
 CurrentLoadItemRuler {type="map<string>", default {}}
 CurrentPage {type=integer}
 CurrentPageObjectList {type="array<EUO_PreviewItemObject_C>", default []}
 CurrentPendingLoadItemList {type="array<St_FXB_Item>", default []}
 CurrentPlayItem {
  type=St_FXB_Item
  default {
   assetDataPath=""
   category=""
   depend=""
   editorName=""
   folderPath=""
   hasValidFX=false
   myCategoryIndexInDataBase=0
   myIndexInDatabase=0
   publishName=""
   publishVersion=0
   x=0
   y=0
   colorList []
   tags []
   videoList []
  }
 }
 CurrentShowTag {type="array<string>", default []}
 CurrentUser {type=string, default=User}
 DataBaseSaveGame {type=FXB_SaveGame_C, default=null}
 FolderList {type="array<string>", default []}
 LastShowSource {type=EUW_PreviewButton_C, default=null}
 LibraryBasePath {type=string, default=""}
 LoadPageItemNumber {type=integer, default=100}
 MaxVersion {type=integer, default=99}
 MediaPlayer {type=MediaPlayer, default=/FXBrowserUtilities/FXRefLib/Finder/FXRefLibMedia}
 MinVersion {type=integer}
 RelatedAssetsList {type="array<Object>", default []}
 Search {type=string, default=""}
 SearchDepend {type=boolean}
 SettingSaveObject {type=FXB_SaveGame_Path_C, default=null}
 SystemUser {type=boolean}
 tmpRefreshTag_ReadedTagItemList {type="array<St_FXB_StringArray>", default []}
 tmpRefreshTag_ReadedTagList {type="array<string>", default []}
 tmpRefreshTag_TagArray {type="array<string>", default []}
 tmpRefreshTag_TmpItem {
  type=St_FXB_Item
  default {
   assetDataPath=""
   category=""
   depend=""
   editorName=""
   folderPath=""
   hasValidFX=false
   myCategoryIndexInDataBase=0
   myIndexInDatabase=0
   publishName=""
   publishVersion=0
   x=0
   y=0
   colorList []
   tags []
   videoList []
  }
 }
 tmpRefreshTag_TmpItemListData {type="array<St_FXB_Item>", default []}
 tmpRefreshTag_tmpItemNameList {type="array<string>", default []}
 tmpRefreshTag_tmpTag2ItemArray {type="array<St_FXB_Tag2Item>", default []}
]
widgets {
 info {rootWidgetClass=&Overlay, widgetCount=158}
 widgets [
  Overlay_183 {widgetClassPath=&Overlay, widget=Overlay_183}
  Border_94 {
   widgetClassPath=&Border
   parent=Overlay_183
   widget=Border_94
   properties {brushColor=<rgba 0.0091 0.0097 0.0116 1>}
   slot {class=OverlaySlot, properties {horizontalAlignment=HAlign_Fill, verticalAlignment=VAlign_Fill}}
  }
  HorizontalBox_43 {
   widgetClassPath=&HorizontalBox
   parent=Overlay_183
   widget=HorizontalBox_43
   slot {class=OverlaySlot, properties {horizontalAlignment=HAlign_Fill, verticalAlignment=VAlign_Fill}}
  }
  LeftPanel {
   widgetClassPath=&Overlay
   bIsVariable=true
   parent=HorizontalBox_43
   widget=LeftPanel
   slot {class=HorizontalBoxSlot, properties {size {sizeRule=Fill, value=0.3}}}
  }
  Border_3 {
   widgetClassPath=&Border
   parent=LeftPanel
   widget=Border_3
   properties {brushColor=<rgba 0.0137 0.0144 0.0168 1>}
   slot {class=OverlaySlot, properties {horizontalAlignment=HAlign_Fill, verticalAlignment=VAlign_Fill}}
  }
  HorizontalBox_2 {
   widgetClassPath=&HorizontalBox
   parent=LeftPanel
   widget=HorizontalBox_2
   slot {class=OverlaySlot, properties {horizontalAlignment=HAlign_Fill, verticalAlignment=VAlign_Fill}}
  }
  EditorUtilityScrollBox_246 {
   widgetClassPath=&EditorUtilityScrollBox
   parent=HorizontalBox_2
   widget=EditorUtilityScrollBox_246
   slot {class=HorizontalBoxSlot, properties {size {sizeRule=Fill, value=1}}}
  }
  VerticalBox_77 {
   widgetClassPath=&VerticalBox
   parent=EditorUtilityScrollBox_246
   widget=VerticalBox_77
   properties {clipping=ClipToBoundsAlways}
   slot {class=ScrollBoxSlot, properties {padding=<ltrb 3 3 3 3>}}
  }
  HorizontalBox_5 {
   widgetClassPath=&HorizontalBox
   parent=VerticalBox_77
   widget=HorizontalBox_5
   slot {class=VerticalBoxSlot, properties {horizontalAlignment=HAlign_Left, verticalAlignment=VAlign_Center}}
  }
  TextBlock_180 {
   widgetClassPath=&TextBlock
   parent=HorizontalBox_5
   widget=TextBlock_180
   properties {
    text=FX Browser
    font {
     bForceMonospaced=false
     bMaterialIsStencil=false
     fontMaterial=null
     fontObject=/Engine/EngineFonts/RobotoTiny
     letterSpacing=0
     monospacedWidth=1
     size=24
     skewAmount=0
     typefaceFontName=Light
     outlineSettings {
      bApplyOutlineToDropShadows=false
      bMiteredCorners=false
      bSeparateFillAlpha=false
      outlineMaterial=null
      outlineSize=0
      outlineColor=<rgba 0 0 0 1>
     }
    }
   }
   slot {class=HorizontalBoxSlot}
  }
  EditorUtilityButton_10 {
   widgetClassPath=&EditorUtilityButton
   parent=HorizontalBox_5
   widget=EditorUtilityButton_10
   properties {backgroundColor=<rgba 0.5976 0.6299 0.733 1>}
   slot {class=HorizontalBoxSlot, properties {padding=<ltrb 5 7 5 7>, size {sizeRule=Fill, value=1}}}
  }
  ImageSize_11 {
   widgetClassPath=&TextBlock
   parent=EditorUtilityButton_10
   widget=ImageSize_11
   properties {
    justification=Center
    text=BugRep
    font {
     bForceMonospaced=false
     bMaterialIsStencil=false
     fontMaterial=null
     fontObject=&RobotoTiny
     letterSpacing=0
     monospacedWidth=1
     size=9
     skewAmount=0
     typefaceFontName=Light
     outlineSettings {
      bApplyOutlineToDropShadows=false
      bMiteredCorners=false
      bSeparateFillAlpha=false
      outlineMaterial=null
      outlineSize=0
      outlineColor=<rgba 0 0 0 1>
     }
    }
   }
   slot {class=ButtonSlot, properties {horizontalAlignment=HAlign_Fill, padding=<ltrb 10 0 10 0>}}
  }
  EditorUtilityButton_13 {
   widgetClassPath=&EditorUtilityButton
   parent=HorizontalBox_5
   widget=EditorUtilityButton_13
   properties {backgroundColor=<rgba 0.5976 0.6299 0.733 1>}
   slot {class=HorizontalBoxSlot, properties {padding=<ltrb 5 7 5 7>, size {sizeRule=Fill, value=1}}}
  }
  ImageSize_14 {
   widgetClassPath=&TextBlock
   parent=EditorUtilityButton_13
   widget=ImageSize_14
   properties {
    justification=Center
    text=TagTest
    font {
     bForceMonospaced=false
     bMaterialIsStencil=false
     fontMaterial=null
     fontObject=&RobotoTiny
     letterSpacing=0
     monospacedWidth=1
     size=9
     skewAmount=0
     typefaceFontName=Light
     outlineSettings {
      bApplyOutlineToDropShadows=false
      bMiteredCorners=false
      bSeparateFillAlpha=false
      outlineMaterial=null
      outlineSize=0
      outlineColor=<rgba 0 0 0 1>
     }
    }
   }
   slot {class=ButtonSlot, properties {horizontalAlignment=HAlign_Fill, padding=<ltrb 10 0 10 0>}}
  }
  CosmeticTypeButton {
   widgetClassPath=&EditorUtilityButton
   parent=VerticalBox_77
   widget=CosmeticTypeButton
   properties {backgroundColor=<rgba 0.8153 0.8594 1 1>}
   slot {class=VerticalBoxSlot}
  }
  TextBlock_121 {
   widgetClassPath=&TextBlock
   parent=CosmeticTypeButton
   widget=TextBlock_121
   properties {
    text=CosmeticType
    font {
     bForceMonospaced=false
     bMaterialIsStencil=false
     fontMaterial=null
     fontObject=&RobotoTiny
     letterSpacing=0
     monospacedWidth=1
     size=13.5
     skewAmount=0
     typefaceFontName=Light
     outlineSettings {
      bApplyOutlineToDropShadows=false
      bMiteredCorners=false
      bSeparateFillAlpha=false
      outlineMaterial=null
      outlineSize=0
      outlineColor=<rgba 0 0 0 1>
     }
    }
   }
   slot {class=ButtonSlot, properties {horizontalAlignment=HAlign_Fill, verticalAlignment=VAlign_Fill, padding=<ltrb 0 0 0 0>}}
  }
  CosmeticCategorySlot {widgetClassPath=&Overlay, bIsVariable=true, parent=VerticalBox_77, widget=CosmeticCategorySlot, slot {class=VerticalBoxSlot}}
  Image_76 {
   widgetClassPath=&Image
   parent=CosmeticCategorySlot
   widget=Image_76
   properties {colorAndOpacity=<rgba 0.0091 0.0097 0.0116 1>}
   slot {class=OverlaySlot, properties {horizontalAlignment=HAlign_Fill, verticalAlignment=VAlign_Fill}}
  }
  CosmeticCategoryBox {
   widgetClassPath=&/BUW_Category
   bIsVariable=true
   parent=CosmeticCategorySlot
   widget=CosmeticCategoryBox
   properties {categoryList []}
   slot {class=OverlaySlot}
  }
  AttributeFilter {
   widgetClassPath=&EditorUtilityButton
   parent=VerticalBox_77
   widget=AttributeFilter
   properties {backgroundColor=<rgba 0.8153 0.8594 1 1>}
   slot {class=VerticalBoxSlot, properties {padding=<ltrb 0 15 0 0>}}
  }
  TextBlock {
   widgetClassPath=&TextBlock
   parent=AttributeFilter
   widget=TextBlock
   properties {
    text=AttributeFilter
    font {
     bForceMonospaced=false
     bMaterialIsStencil=false
     fontMaterial=null
     fontObject=&RobotoTiny
     letterSpacing=0
     monospacedWidth=1
     size=13.5
     skewAmount=0
     typefaceFontName=Light
     outlineSettings {
      bApplyOutlineToDropShadows=false
      bMiteredCorners=false
      bSeparateFillAlpha=false
      outlineMaterial=null
      outlineSize=0
      outlineColor=<rgba 0 0 0 1>
     }
    }
   }
   slot {class=ButtonSlot, properties {horizontalAlignment=HAlign_Fill, verticalAlignment=VAlign_Fill, padding=<ltrb 0 0 0 0>}}
  }
  AttributeFilterBox {widgetClassPath=&Overlay, bIsVariable=true, parent=VerticalBox_77, widget=AttributeFilterBox, slot {class=VerticalBoxSlot}}
  Image_1 {
   widgetClassPath=&Image
   parent=AttributeFilterBox
   widget=Image_1
   properties {colorAndOpacity=<rgba 0.0091 0.0097 0.0116 1>}
   slot {class=OverlaySlot, properties {horizontalAlignment=HAlign_Fill, verticalAlignment=VAlign_Fill}}
  }
  VerticalBox_94 {
   widgetClassPath=&VerticalBox
   parent=AttributeFilterBox
   widget=VerticalBox_94
   slot {class=OverlaySlot, properties {horizontalAlignment=HAlign_Fill, verticalAlignment=VAlign_Fill}}
  }
  Border_0 {
   widgetClassPath=&Border
   parent=VerticalBox_94
   widget=Border_0
   properties {background {
     bIsDynamicallyLoaded=false
     drawAs=Border
     imageType=NoImage
     mirroring=NoMirror
     resourceName=None
     resourceObject=null
     tiling=NoTile
     imageSize=<32 32>
     margin=<ltrb 1 1 1 1>
     outlineSettings {
      bUseBrushTransparency=false
      roundingType=HalfHeightRadius
      width=0
      color {colorUseRule=UseColor_Specified, specifiedColor=<rgba 0 0 0 0>}
      cornerRadii=<0 0 0 0>
     }
     tintColor {colorUseRule=UseColor_Specified, specifiedColor=<rgba 0.0815 0.0859 0.1 1>}
     uVRegion {bIsValid=false, max=<0 0>, min=<0 0>}
    }}
   slot {class=VerticalBoxSlot, properties {padding=<ltrb 0 5 0 5>}}
  }
  VerticalBox_189 {widgetClassPath=&VerticalBox, parent=Border_0, widget=VerticalBox_189, slot {class=BorderSlot}}
  TextBlock_203 {
   widgetClassPath=&TextBlock
   parent=VerticalBox_189
   widget=TextBlock_203
   properties {
    text=Season range
    font {
     bForceMonospaced=false
     bMaterialIsStencil=false
     fontMaterial=null
     fontObject=&RobotoTiny
     letterSpacing=0
     monospacedWidth=1
     size=9
     skewAmount=0
     typefaceFontName=Light
     outlineSettings {
      bApplyOutlineToDropShadows=false
      bMiteredCorners=false
      bSeparateFillAlpha=false
      outlineMaterial=null
      outlineSize=0
      outlineColor=<rgba 0 0 0 1>
     }
    }
   }
   slot {class=VerticalBoxSlot, properties {padding=<ltrb 2 2 2 0>}}
  }
  HorizontalBox_0 {widgetClassPath=&HorizontalBox, parent=VerticalBox_189, widget=HorizontalBox_0, slot {class=VerticalBoxSlot}}
  MinVersionInput {
   widgetClassPath=&EditorUtilityEditableTextBox
   bIsVariable=true
   parent=HorizontalBox_0
   widget=MinVersionInput
   properties {hintText=Min Version}
   slot {class=HorizontalBoxSlot, properties {size {sizeRule=Fill, value=1}}}
  }
  MaxVersionInput {
   widgetClassPath=&EditorUtilityEditableTextBox
   bIsVariable=true
   parent=HorizontalBox_0
   widget=MaxVersionInput
   properties {hintText=Max Version}
   slot {class=HorizontalBoxSlot, properties {size {sizeRule=Fill, value=1}}}
  }
  HorizontalBox_3 {
   widgetClassPath=&HorizontalBox
   parent=VerticalBox_189
   widget=HorizontalBox_3
   slot {class=VerticalBoxSlot, properties {padding=<ltrb 50 2 50 2>}}
  }
  EditorUtilityButton_166 {
   widgetClassPath=&EditorUtilityButton
   parent=HorizontalBox_3
   widget=EditorUtilityButton_166
   properties {backgroundColor=<rgba 0.5976 0.6299 0.733 1>}
   slot {class=HorizontalBoxSlot, properties {size {sizeRule=Fill, value=1}}}
  }
  TextBlock_127 {
   widgetClassPath=&TextBlock
   parent=EditorUtilityButton_166
   widget=TextBlock_127
   properties {
    text=" Apply"
    font {
     bForceMonospaced=false
     bMaterialIsStencil=false
     fontMaterial=null
     fontObject=&RobotoTiny
     letterSpacing=0
     monospacedWidth=1
     size=9
     skewAmount=0
     typefaceFontName=Light
     outlineSettings {
      bApplyOutlineToDropShadows=false
      bMiteredCorners=false
      bSeparateFillAlpha=false
      outlineMaterial=null
      outlineSize=0
      outlineColor=<rgba 0 0 0 1>
     }
    }
   }
   slot {class=ButtonSlot}
  }
  EditorUtilityButton_9 {
   widgetClassPath=&EditorUtilityButton
   parent=HorizontalBox_3
   widget=EditorUtilityButton_9
   properties {backgroundColor=<rgba 0.5976 0.6299 0.733 1>}
   slot {class=HorizontalBoxSlot, properties {size {sizeRule=Fill, value=1}}}
  }
  TextBlock_9 {
   widgetClassPath=&TextBlock
   parent=EditorUtilityButton_9
   widget=TextBlock_9
   properties {
    text=Disable
    font {
     bForceMonospaced=false
     bMaterialIsStencil=false
     fontMaterial=null
     fontObject=&RobotoTiny
     letterSpacing=0
     monospacedWidth=1
     size=9
     skewAmount=0
     typefaceFontName=Light
     outlineSettings {
      bApplyOutlineToDropShadows=false
      bMiteredCorners=false
      bSeparateFillAlpha=false
      outlineMaterial=null
      outlineSize=0
      outlineColor=<rgba 0 0 0 1>
     }
    }
   }
   slot {class=ButtonSlot}
  }
  TextBlock_1 {
   widgetClassPath=&TextBlock
   parent=VerticalBox_189
   widget=TextBlock_1
   properties {
    text=Color
    font {
     bForceMonospaced=false
     bMaterialIsStencil=false
     fontMaterial=null
     fontObject=&RobotoTiny
     letterSpacing=0
     monospacedWidth=1
     size=9
     skewAmount=0
     typefaceFontName=Light
     outlineSettings {
      bApplyOutlineToDropShadows=false
      bMiteredCorners=false
      bSeparateFillAlpha=false
      outlineMaterial=null
      outlineSize=0
      outlineColor=<rgba 0 0 0 1>
     }
    }
   }
   slot {class=VerticalBoxSlot, properties {padding=<ltrb 2 2 2 0>}}
  }
  HorizontalBox_422 {widgetClassPath=&HorizontalBox, parent=VerticalBox_189, widget=HorizontalBox_422, slot {class=VerticalBoxSlot}}
  SizeBox_4 {
   widgetClassPath=&SizeBox
   parent=HorizontalBox_422
   widget=SizeBox_4
   properties {bOverride_HeightOverride=true, bOverride_WidthOverride=true, heightOverride=32, widthOverride=32}
   slot {class=HorizontalBoxSlot}
  }
  Border_328 {widgetClassPath=&Border, parent=SizeBox_4, widget=Border_328, slot {class=SizeBoxSlot}, properties=~}
  SizeBox_5 {
   widgetClassPath=&SizeBox
   parent=HorizontalBox_422
   widget=SizeBox_5
   properties {heightOverride=40, widthOverride=250}
   slot {class=HorizontalBoxSlot, properties {padding=<ltrb 5 5 5 5>, size {sizeRule=Fill, value=1}}}
  }
  AnalogSlider_1 {
   widgetClassPath=/Script/CommonUI.AnalogSlider
   bIsVariable=true
   parent=SizeBox_5
   widget=AnalogSlider_1
   properties {
    maxValue=360
    widgetStyle {
     barThickness=10
     disabledBarImage {
      bIsDynamicallyLoaded=false
      drawAs=RoundedBox
      imageType=NoImage
      mirroring=NoMirror
      resourceName=None
      resourceObject=null
      tiling=NoTile
      imageSize=<0 0>
      margin=<ltrb 0 0 0 0>
      outlineSettings {
       bUseBrushTransparency=false
       roundingType=FixedRadius
       width=1
       color {colorUseRule=UseColor_ColorTable, specifiedColor=<rgba 0 0 0 0>}
       cornerRadii=<2 2 2 2>
      }
      tintColor {colorUseRule=UseColor_Specified, specifiedColor=<rgba 0.0048 0.0048 0.0048 1>}
      uVRegion {bIsValid=false, max=<0 0>, min=<0 0>}
     }
     disabledThumbImage {
      bIsDynamicallyLoaded=false
      drawAs=RoundedBox
      imageType=NoImage
      mirroring=NoMirror
      resourceName=None
      resourceObject=null
      tiling=NoTile
      imageSize=<15 15>
      margin=<ltrb 0 0 0 0>
      outlineSettings {
       bUseBrushTransparency=false
       roundingType=HalfHeightRadius
       width=0
       color {colorUseRule=UseColor_Specified, specifiedColor=<rgba 0 0 0 0>}
       cornerRadii=<0 0 0 0>
      }
      tintColor {colorUseRule=UseColor_Specified, specifiedColor=<rgba 0.5271 0.5271 0.5271 1>}
      uVRegion {bIsValid=false, max=<0 0>, min=<0 0>}
     }
     hoveredBarImage {
      bIsDynamicallyLoaded=false
      drawAs=RoundedBox
      imageType=NoImage
      mirroring=NoMirror
      resourceName=None
      resourceObject=null
      tiling=NoTile
      imageSize=<0 0>
      margin=<ltrb 0 0 0 0>
      outlineSettings {
       bUseBrushTransparency=false
       roundingType=FixedRadius
       width=1
       color {colorUseRule=UseColor_ColorTable, specifiedColor=<rgba 0 0 0 0>}
       cornerRadii=<2 2 2 2>
      }
      tintColor {colorUseRule=UseColor_Specified, specifiedColor=<rgba 0.0048 0.0048 0.0048 1>}
      uVRegion {bIsValid=false, max=<0 0>, min=<0 0>}
     }
     hoveredThumbImage {
      bIsDynamicallyLoaded=false
      drawAs=RoundedBox
      imageType=NoImage
      mirroring=NoMirror
      resourceName=None
      resourceObject=null
      tiling=NoTile
      imageSize=<15 15>
      margin=<ltrb 0 0 0 0>
      outlineSettings {
       bUseBrushTransparency=false
       roundingType=HalfHeightRadius
       width=0
       color {colorUseRule=UseColor_Specified, specifiedColor=<rgba 0 0 0 0>}
       cornerRadii=<0 0 0 0>
      }
      tintColor {colorUseRule=UseColor_Specified, specifiedColor=<rgba 1 1 1 1>}
      uVRegion {bIsValid=false, max=<0 0>, min=<0 0>}
     }
     normalBarImage {
      bIsDynamicallyLoaded=false
      drawAs=RoundedBox
      imageType=NoImage
      mirroring=NoMirror
      resourceName=None
      resourceObject=null
      tiling=NoTile
      imageSize=<0 0>
      margin=<ltrb 0 0 0 0>
      outlineSettings {
       bUseBrushTransparency=false
       roundingType=FixedRadius
       width=1
       color {colorUseRule=UseColor_ColorTable, specifiedColor=<rgba 0 0 0 0>}
       cornerRadii=<2 2 2 2>
      }
      tintColor {colorUseRule=UseColor_Specified, specifiedColor=<rgba 0.0048 0.0048 0.0048 1>}
      uVRegion {bIsValid=false, max=<0 0>, min=<0 0>}
     }
     normalThumbImage {
      bIsDynamicallyLoaded=false
      drawAs=RoundedBox
      imageType=NoImage
      mirroring=NoMirror
      resourceName=None
      resourceObject=null
      tiling=NoTile
      imageSize=<15 15>
      margin=<ltrb 0 0 0 0>
      outlineSettings {
       bUseBrushTransparency=false
       roundingType=HalfHeightRadius
       width=0
       color {colorUseRule=UseColor_Specified, specifiedColor=<rgba 0 0 0 0>}
       cornerRadii=<0 0 0 0>
      }
      tintColor {colorUseRule=UseColor_Specified, specifiedColor=<rgba 0.5271 0.5271 0.5271 1>}
      uVRegion {bIsValid=false, max=<0 0>, min=<0 0>}
     }
    }
   }
   slot {class=SizeBoxSlot}
  }
  HorizontalBox_424 {
   widgetClassPath=&HorizontalBox
   parent=VerticalBox_189
   widget=HorizontalBox_424
   slot {class=VerticalBoxSlot, properties {padding=<ltrb 0 2 0 2>}}
  }
  TextBlock_7 {
   widgetClassPath=&TextBlock
   parent=HorizontalBox_424
   widget=TextBlock_7
   properties {
    text="Tolerance : "
    font {
     bForceMonospaced=false
     bMaterialIsStencil=false
     fontMaterial=null
     fontObject=&RobotoTiny
     letterSpacing=0
     monospacedWidth=1
     size=9
     skewAmount=0
     typefaceFontName=Light
     outlineSettings {
      bApplyOutlineToDropShadows=false
      bMiteredCorners=false
      bSeparateFillAlpha=false
      outlineMaterial=null
      outlineSize=0
      outlineColor=<rgba 0 0 0 1>
     }
    }
   }
   slot {class=HorizontalBoxSlot, properties {horizontalAlignment=HAlign_Center, verticalAlignment=VAlign_Center}}
  }
  EditorUtilityEditableTextBox {
   widgetClassPath=&EditorUtilityEditableTextBox
   parent=HorizontalBox_424
   widget=EditorUtilityEditableTextBox
   properties {hintText=Tolorence, text=30}
   slot {class=HorizontalBoxSlot}
  }
  EditorUtilityButton_6 {
   widgetClassPath=&EditorUtilityButton
   parent=HorizontalBox_424
   widget=EditorUtilityButton_6
   properties {backgroundColor=<rgba 0.5976 0.6299 0.733 1>}
   slot {class=HorizontalBoxSlot, properties {horizontalAlignment=HAlign_Right, size {sizeRule=Fill, value=1}}}
  }
  TextBlock_3 {
   widgetClassPath=&TextBlock
   parent=EditorUtilityButton_6
   widget=TextBlock_3
   properties {
    text=" Apply"
    font {
     bForceMonospaced=false
     bMaterialIsStencil=false
     fontMaterial=null
     fontObject=&RobotoTiny
     letterSpacing=0
     monospacedWidth=1
     size=9
     skewAmount=0
     typefaceFontName=Light
     outlineSettings {
      bApplyOutlineToDropShadows=false
      bMiteredCorners=false
      bSeparateFillAlpha=false
      outlineMaterial=null
      outlineSize=0
      outlineColor=<rgba 0 0 0 1>
     }
    }
   }
   slot {class=ButtonSlot}
  }
  EditorUtilityButton_7 {
   widgetClassPath=&EditorUtilityButton
   parent=HorizontalBox_424
   widget=EditorUtilityButton_7
   properties {backgroundColor=<rgba 0.5976 0.6299 0.733 1>}
   slot {class=HorizontalBoxSlot, properties {horizontalAlignment=HAlign_Right, size {sizeRule=Fill, value=1}}}
  }
  TextBlock_8 {
   widgetClassPath=&TextBlock
   parent=EditorUtilityButton_7
   widget=TextBlock_8
   properties {
    text=Disable
    font {
     bForceMonospaced=false
     bMaterialIsStencil=false
     fontMaterial=null
     fontObject=&RobotoTiny
     letterSpacing=0
     monospacedWidth=1
     size=9
     skewAmount=0
     typefaceFontName=Light
     outlineSettings {
      bApplyOutlineToDropShadows=false
      bMiteredCorners=false
      bSeparateFillAlpha=false
      outlineMaterial=null
      outlineSize=0
      outlineColor=<rgba 0 0 0 1>
     }
    }
   }
   slot {class=ButtonSlot}
  }
  Tag {
   widgetClassPath=&EditorUtilityButton
   parent=VerticalBox_77
   widget=Tag
   properties {backgroundColor=<rgba 0.8153 0.8594 1 1>}
   slot {class=VerticalBoxSlot, properties {padding=<ltrb 0 15 0 0>}}
  }
  TextBlock_2 {
   widgetClassPath=&TextBlock
   parent=Tag
   widget=TextBlock_2
   properties {
    text=Tag
    font {
     bForceMonospaced=false
     bMaterialIsStencil=false
     fontMaterial=null
     fontObject=&RobotoTiny
     letterSpacing=0
     monospacedWidth=1
     size=13.5
     skewAmount=0
     typefaceFontName=Light
     outlineSettings {
      bApplyOutlineToDropShadows=false
      bMiteredCorners=false
      bSeparateFillAlpha=false
      outlineMaterial=null
      outlineSize=0
      outlineColor=<rgba 0 0 0 1>
     }
    }
   }
   slot {class=ButtonSlot, properties {horizontalAlignment=HAlign_Fill, verticalAlignment=VAlign_Fill, padding=<ltrb 0 0 0 0>}}
  }
  Overlay_1 {
   widgetClassPath=&Overlay
   parent=VerticalBox_77
   widget=Overlay_1
   slot {class=VerticalBoxSlot, properties {padding=<ltrb 0 5 0 5>}}
  }
  Image_3 {
   widgetClassPath=&Image
   parent=Overlay_1
   widget=Image_3
   properties {colorAndOpacity=<rgba 0.0091 0.0097 0.0116 1>}
   slot {class=OverlaySlot, properties {horizontalAlignment=HAlign_Fill, verticalAlignment=VAlign_Fill}}
  }
  Border_1 {
   widgetClassPath=&Border
   parent=Overlay_1
   widget=Border_1
   properties {
    background {
     bIsDynamicallyLoaded=false
     drawAs=Border
     imageType=NoImage
     mirroring=NoMirror
     resourceName=None
     resourceObject=null
     tiling=NoTile
     imageSize=<32 32>
     margin=<ltrb 1 1 1 1>
     outlineSettings {
      bUseBrushTransparency=false
      roundingType=HalfHeightRadius
      width=0
      color {colorUseRule=UseColor_Specified, specifiedColor=<rgba 0 0 0 0>}
      cornerRadii=<0 0 0 0>
     }
     tintColor {colorUseRule=UseColor_Specified, specifiedColor=<rgba 0.0815 0.0859 0.1 1>}
     uVRegion {bIsValid=false, max=<0 0>, min=<0 0>}
    }
    padding=<ltrb 0 5 0 5>
   }
   slot {class=OverlaySlot, properties {horizontalAlignment=HAlign_Fill, verticalAlignment=VAlign_Fill}}
  }
  VerticalBox_261 {
   widgetClassPath=&VerticalBox
   parent=Border_1
   widget=VerticalBox_261
   slot {class=BorderSlot, properties {padding=<ltrb 0 5 0 5>}}
  }
  EditorUtilityEditableTextBox_229 {
   widgetClassPath=&EditorUtilityEditableTextBox
   parent=VerticalBox_261
   widget=EditorUtilityEditableTextBox_229
   properties {hintText=Search Tag}
   slot {class=VerticalBoxSlot, properties {padding=<ltrb 4 0 4 5>}}
  }
  HorizontalBox_178 {
   widgetClassPath=&HorizontalBox
   parent=VerticalBox_261
   widget=HorizontalBox_178
   slot {class=VerticalBoxSlot, properties {horizontalAlignment=HAlign_Right, padding=<ltrb 2 2 2 0>}}
  }
  TextBlock_4 {
   widgetClassPath=&TextBlock
   parent=HorizontalBox_178
   widget=TextBlock_4
   properties {
    text=Solo
    font {
     bForceMonospaced=false
     bMaterialIsStencil=false
     fontMaterial=null
     fontObject=&RobotoTiny
     letterSpacing=0
     monospacedWidth=1
     size=9
     skewAmount=0
     typefaceFontName=Light
     outlineSettings {
      bApplyOutlineToDropShadows=false
      bMiteredCorners=false
      bSeparateFillAlpha=false
      outlineMaterial=null
      outlineSize=0
      outlineColor=<rgba 0 0 0 1>
     }
    }
   }
   slot {class=HorizontalBoxSlot, properties {padding=<ltrb 0 0 10 0>}}
  }
  TextBlock_5 {
   widgetClassPath=&TextBlock
   parent=HorizontalBox_178
   widget=TextBlock_5
   properties {
    text=Hide
    font {
     bForceMonospaced=false
     bMaterialIsStencil=false
     fontMaterial=null
     fontObject=&RobotoTiny
     letterSpacing=0
     monospacedWidth=1
     size=9
     skewAmount=0
     typefaceFontName=Light
     outlineSettings {
      bApplyOutlineToDropShadows=false
      bMiteredCorners=false
      bSeparateFillAlpha=false
      outlineMaterial=null
      outlineSize=0
      outlineColor=<rgba 0 0 0 1>
     }
    }
   }
   slot {class=HorizontalBoxSlot}
  }
  SizeBox_0 {
   widgetClassPath=&SizeBox
   parent=VerticalBox_261
   widget=SizeBox_0
   properties {bOverride_MaxDesiredHeight=true, maxDesiredHeight=400}
   slot {class=VerticalBoxSlot}
  }
  EditorUtilityScrollBox_360 {
   widgetClassPath=&EditorUtilityScrollBox
   parent=SizeBox_0
   widget=EditorUtilityScrollBox_360
   slot {class=SizeBoxSlot, properties {padding=<ltrb 4 0 4 0>}}
  }
  TagList {widgetClassPath=&VerticalBox, bIsVariable=true, parent=EditorUtilityScrollBox_360, widget=TagList, slot {class=ScrollBoxSlot}}
  TagButton {widgetClassPath=&/TagButton, bIsVariable=true, parent=TagList, widget=TagButton, slot {class=VerticalBoxSlot}}
  EditorUtilityButton_5 {
   widgetClassPath=&EditorUtilityButton
   parent=VerticalBox_261
   widget=EditorUtilityButton_5
   properties {visibility=Collapsed, backgroundColor=<rgba 0.5976 0.6299 0.733 1>}
   slot {class=VerticalBoxSlot, properties {padding=<ltrb 50 2 50 2>}}
  }
  TextBlock_6 {
   widgetClassPath=&TextBlock
   parent=EditorUtilityButton_5
   widget=TextBlock_6
   properties {
    text=" Apply"
    font {
     bForceMonospaced=false
     bMaterialIsStencil=false
     fontMaterial=null
     fontObject=&RobotoTiny
     letterSpacing=0
     monospacedWidth=1
     size=9
     skewAmount=0
     typefaceFontName=Light
     outlineSettings {
      bApplyOutlineToDropShadows=false
      bMiteredCorners=false
      bSeparateFillAlpha=false
      outlineMaterial=null
      outlineSize=0
      outlineColor=<rgba 0 0 0 1>
     }
    }
   }
   slot {class=ButtonSlot}
  }
  EditorUtilityButton {
   widgetClassPath=&EditorUtilityButton
   parent=HorizontalBox_2
   widget=EditorUtilityButton
   properties {
    bIsEnabled=false
    visibility=Collapsed
    backgroundColor=<rgba 0.8153 0.8594 1 1>
    widgetStyle {
     clickedSlateSound {resourceObject=null}
     disabled {
      bIsDynamicallyLoaded=false
      drawAs=RoundedBox
      imageType=NoImage
      mirroring=NoMirror
      resourceName=None
      resourceObject=null
      tiling=NoTile
      imageSize=<0 0>
      margin=<ltrb 0 0 0 0>
      outlineSettings {
       bUseBrushTransparency=false
       roundingType=FixedRadius
       width=1
       color {colorUseRule=UseColor_ColorTable, specifiedColor=<rgba 0 0 0 0>}
       cornerRadii=<4 4 4 4>
      }
      tintColor {colorUseRule=UseColor_Specified, specifiedColor=<rgba 0.0395 0.0395 0.0395 1>}
      uVRegion {bIsValid=false, max=<0 0>, min=<0 0>}
     }
     disabledForeground {colorUseRule=UseColor_ColorTable, specifiedColor=<rgba 0 0 0 0>}
     hovered {
      bIsDynamicallyLoaded=false
      drawAs=RoundedBox
      imageType=NoImage
      mirroring=NoMirror
      resourceName=None
      resourceObject=null
      tiling=NoTile
      imageSize=<6 6>
      margin=<ltrb 0 0 0 0>
      outlineSettings {
       bUseBrushTransparency=false
       roundingType=FixedRadius
       width=1
       color {colorUseRule=UseColor_ColorTable, specifiedColor=<rgba 0 0 0 0>}
       cornerRadii=<4 4 4 4>
      }
      tintColor {colorUseRule=UseColor_Specified, specifiedColor=<rgba 0.0953 0.0953 0.0953 1>}
      uVRegion {bIsValid=false, max=<0 0>, min=<0 0>}
     }
     hoveredForeground {colorUseRule=UseColor_ColorTable, specifiedColor=<rgba 0 0 0 0>}
     hoveredSlateSound {resourceObject=null}
     normal {
      bIsDynamicallyLoaded=false
      drawAs=RoundedBox
      imageType=NoImage
      mirroring=NoMirror
      resourceName=None
      resourceObject=null
      tiling=NoTile
      imageSize=<6 6>
      margin=<ltrb 0 0 0 0>
      outlineSettings {
       bUseBrushTransparency=false
       roundingType=FixedRadius
       width=1
       color {colorUseRule=UseColor_ColorTable, specifiedColor=<rgba 0 0 0 0>}
       cornerRadii=<4 4 4 4>
      }
      tintColor {colorUseRule=UseColor_Specified, specifiedColor=<rgba 0.0395 0.0395 0.0395 1>}
      uVRegion {bIsValid=false, max=<0 0>, min=<0 0>}
     }
     normalForeground {colorUseRule=UseColor_ColorTable, specifiedColor=<rgba 0 0 0 0>}
     normalPadding=<ltrb 1 1 1 1>
     pressed {
      bIsDynamicallyLoaded=false
      drawAs=RoundedBox
      imageType=NoImage
      mirroring=NoMirror
      resourceName=None
      resourceObject=null
      tiling=NoTile
      imageSize=<6 6>
      margin=<ltrb 0 0 0 0>
      outlineSettings {
       bUseBrushTransparency=false
       roundingType=FixedRadius
       width=1
       color {colorUseRule=UseColor_ColorTable, specifiedColor=<rgba 0 0 0 0>}
       cornerRadii=<4 4 4 4>
      }
      tintColor {colorUseRule=UseColor_Specified, specifiedColor=<rgba 0.0284 0.0284 0.0284 1>}
      uVRegion {bIsValid=false, max=<0 0>, min=<0 0>}
     }
     pressedForeground {colorUseRule=UseColor_ColorTable, specifiedColor=<rgba 0 0 0 0>}
     pressedPadding=<ltrb 1 1 1 1>
     pressedSlateSound {resourceObject=null}
    }
   }
   slot {class=HorizontalBoxSlot}
  }
  EditorUtilityButton_62 {
   widgetClassPath=&EditorUtilityButton
   parent=HorizontalBox_43
   widget=EditorUtilityButton_62
   properties {
    backgroundColor=<rgba 0.8153 0.8594 1 1>
    widgetStyle {
     clickedSlateSound {resourceObject=null}
     disabled {
      bIsDynamicallyLoaded=false
      drawAs=RoundedBox
      imageType=NoImage
      mirroring=NoMirror
      resourceName=None
      resourceObject=null
      tiling=NoTile
      imageSize=<0 0>
      margin=<ltrb 0 0 0 0>
      outlineSettings {
       bUseBrushTransparency=false
       roundingType=FixedRadius
       width=1
       color {colorUseRule=UseColor_ColorTable, specifiedColor=<rgba 0 0 0 0>}
       cornerRadii=<4 4 4 4>
      }
      tintColor {colorUseRule=UseColor_Specified, specifiedColor=<rgba 0.0395 0.0395 0.0395 1>}
      uVRegion {bIsValid=false, max=<0 0>, min=<0 0>}
     }
     disabledForeground {colorUseRule=UseColor_ColorTable, specifiedColor=<rgba 0 0 0 0>}
     hovered {
      bIsDynamicallyLoaded=false
      drawAs=RoundedBox
      imageType=NoImage
      mirroring=NoMirror
      resourceName=None
      resourceObject=null
      tiling=NoTile
      imageSize=<6 6>
      margin=<ltrb 0 0 0 0>
      outlineSettings {
       bUseBrushTransparency=false
       roundingType=FixedRadius
       width=1
       color {colorUseRule=UseColor_ColorTable, specifiedColor=<rgba 0 0 0 0>}
       cornerRadii=<4 4 4 4>
      }
      tintColor {colorUseRule=UseColor_Specified, specifiedColor=<rgba 0.0953 0.0953 0.0953 1>}
      uVRegion {bIsValid=false, max=<0 0>, min=<0 0>}
     }
     hoveredForeground {colorUseRule=UseColor_ColorTable, specifiedColor=<rgba 0 0 0 0>}
     hoveredSlateSound {resourceObject=null}
     normal {
      bIsDynamicallyLoaded=false
      drawAs=RoundedBox
      imageType=NoImage
      mirroring=NoMirror
      resourceName=None
      resourceObject=null
      tiling=NoTile
      imageSize=<6 6>
      margin=<ltrb 0 0 0 0>
      outlineSettings {
       bUseBrushTransparency=false
       roundingType=FixedRadius
       width=1
       color {colorUseRule=UseColor_ColorTable, specifiedColor=<rgba 0 0 0 0>}
       cornerRadii=<4 4 4 4>
      }
      tintColor {colorUseRule=UseColor_Specified, specifiedColor=<rgba 0.0395 0.0395 0.0395 1>}
      uVRegion {bIsValid=false, max=<0 0>, min=<0 0>}
     }
     normalForeground {colorUseRule=UseColor_ColorTable, specifiedColor=<rgba 0 0 0 0>}
     normalPadding=<ltrb 1 1 1 1>
     pressed {
      bIsDynamicallyLoaded=false
      drawAs=RoundedBox
      imageType=NoImage
      mirroring=NoMirror
      resourceName=None
      resourceObject=null
      tiling=NoTile
      imageSize=<6 6>
      margin=<ltrb 0 0 0 0>
      outlineSettings {
       bUseBrushTransparency=false
       roundingType=FixedRadius
       width=1
       color {colorUseRule=UseColor_ColorTable, specifiedColor=<rgba 0 0 0 0>}
       cornerRadii=<4 4 4 4>
      }
      tintColor {colorUseRule=UseColor_Specified, specifiedColor=<rgba 0.0284 0.0284 0.0284 1>}
      uVRegion {bIsValid=false, max=<0 0>, min=<0 0>}
     }
     pressedForeground {colorUseRule=UseColor_ColorTable, specifiedColor=<rgba 0 0 0 0>}
     pressedPadding=<ltrb 1 1 1 1>
     pressedSlateSound {resourceObject=null}
    }
   }
   slot {class=HorizontalBoxSlot, properties {verticalAlignment=VAlign_Center}}
  }
  Fold {
   widgetClassPath=&TextBlock
   parent=EditorUtilityButton_62
   widget=Fold
   properties {
    text="<"
    font {
     bForceMonospaced=false
     bMaterialIsStencil=false
     fontMaterial=null
     fontObject=&Roboto
     letterSpacing=0
     monospacedWidth=1
     size=12
     skewAmount=0
     typefaceFontName=Bold
     outlineSettings {
      bApplyOutlineToDropShadows=false
      bMiteredCorners=false
      bSeparateFillAlpha=false
      outlineMaterial=null
      outlineSize=0
      outlineColor=<rgba 0 0 0 1>
     }
    }
   }
   slot {class=ButtonSlot, properties {padding=<ltrb 0 50 0 50>}}
  }
  Overlay_144 {
   widgetClassPath=&Overlay
   parent=HorizontalBox_43
   widget=Overlay_144
   slot {class=HorizontalBoxSlot, properties {size {sizeRule=Fill, value=1}}}
  }
  Border {
   widgetClassPath=&Border
   parent=Overlay_144
   widget=Border
   properties {brushColor=<rgba 0.0091 0.0097 0.0116 1>}
   slot {class=OverlaySlot, properties {horizontalAlignment=HAlign_Fill, verticalAlignment=VAlign_Fill}}
  }
  VerticalBox_259 {
   widgetClassPath=&VerticalBox
   parent=Overlay_144
   widget=VerticalBox_259
   properties {clipping=ClipToBoundsAlways}
   slot {class=OverlaySlot, properties {horizontalAlignment=HAlign_Fill, verticalAlignment=VAlign_Fill}}
  }
  HorizontalBox_302 {
   widgetClassPath=&HorizontalBox
   parent=VerticalBox_259
   widget=HorizontalBox_302
   slot {class=VerticalBoxSlot, properties {verticalAlignment=VAlign_Top, padding=<ltrb 10 10 10 1>}}
  }
  EditorUtilityButton_8 {
   widgetClassPath=&EditorUtilityButton
   parent=HorizontalBox_302
   widget=EditorUtilityButton_8
   properties {backgroundColor=<rgba 0.5976 0.6299 0.733 1>}
   slot {class=HorizontalBoxSlot, properties {padding=<ltrb 0 0 20 0>}}
  }
  ImageSize_8 {
   widgetClassPath=&TextBlock
   parent=EditorUtilityButton_8
   widget=ImageSize_8
   properties {
    text=Shuffle
    font {
     bForceMonospaced=false
     bMaterialIsStencil=false
     fontMaterial=null
     fontObject=&RobotoTiny
     letterSpacing=0
     monospacedWidth=1
     size=12
     skewAmount=0
     typefaceFontName=Light
     outlineSettings {
      bApplyOutlineToDropShadows=false
      bMiteredCorners=false
      bSeparateFillAlpha=false
      outlineMaterial=null
      outlineSize=0
      outlineColor=<rgba 0 0 0 1>
     }
    }
   }
   slot {class=ButtonSlot, properties {horizontalAlignment=HAlign_Left, padding=<ltrb 10 0 10 0>}}
  }
  EditorUtilityButton_584 {
   widgetClassPath=&EditorUtilityButton
   parent=HorizontalBox_302
   widget=EditorUtilityButton_584
   properties {backgroundColor=<rgba 0.8153 0.8594 1 1>}
   slot {class=HorizontalBoxSlot, properties {size {sizeRule=Fill, value=0.2}}}
  }
  ImageSize_7 {
   widgetClassPath=&TextBlock
   parent=EditorUtilityButton_584
   widget=ImageSize_7
   properties {
    text=Search
    font {
     bForceMonospaced=false
     bMaterialIsStencil=false
     fontMaterial=null
     fontObject=&RobotoTiny
     letterSpacing=0
     monospacedWidth=1
     size=12
     skewAmount=0
     typefaceFontName=Light
     outlineSettings {
      bApplyOutlineToDropShadows=false
      bMiteredCorners=false
      bSeparateFillAlpha=false
      outlineMaterial=null
      outlineSize=0
      outlineColor=<rgba 0 0 0 1>
     }
    }
   }
   slot {class=ButtonSlot, properties {horizontalAlignment=HAlign_Left, padding=<ltrb 10 0 10 0>}}
  }
  ClearSearchbt {
   widgetClassPath=&EditorUtilityButton
   parent=HorizontalBox_302
   widget=ClearSearchbt
   properties {backgroundColor=<rgba 0.8153 0.8594 1 1>}
   slot {class=HorizontalBoxSlot, properties {size {sizeRule=Fill, value=0.1}}}
  }
  ImageSize_10 {
   widgetClassPath=&TextBlock
   parent=ClearSearchbt
   widget=ImageSize_10
   properties {
    justification=Center
    text=Clear
    font {
     bForceMonospaced=false
     bMaterialIsStencil=false
     fontMaterial=null
     fontObject=&RobotoTiny
     letterSpacing=0
     monospacedWidth=1
     size=12
     skewAmount=0
     typefaceFontName=Light
     outlineSettings {
      bApplyOutlineToDropShadows=false
      bMiteredCorners=false
      bSeparateFillAlpha=false
      outlineMaterial=null
      outlineSize=0
      outlineColor=<rgba 0 0 0 1>
     }
    }
   }
   slot {class=ButtonSlot, properties {horizontalAlignment=HAlign_Left, padding=<ltrb 10 0 10 0>}}
  }
  EditorUtilityEditableTextBox_65 {
   widgetClassPath=&EditorUtilityEditableTextBox
   bIsVariable=true
   parent=HorizontalBox_302
   widget=EditorUtilityEditableTextBox_65
   properties {
    hintText=Search
    widgetStyle {
     backgroundColor {colorUseRule=UseColor_Specified, specifiedColor=<rgba 1 1 1 1>}
     backgroundImageFocused {
      bIsDynamicallyLoaded=false
      drawAs=RoundedBox
      imageType=NoImage
      mirroring=NoMirror
      resourceName=None
      resourceObject=null
      tiling=NoTile
      imageSize=<0 0>
      margin=<ltrb 0 0 0 0>
      outlineSettings {
       bUseBrushTransparency=false
       roundingType=FixedRadius
       width=1
       color {colorUseRule=UseColor_ColorTable, specifiedColor=<rgba 0 0 0 0>}
       cornerRadii=<4 4 4 4>
      }
      tintColor {colorUseRule=UseColor_Specified, specifiedColor=<rgba 0.0048 0.0048 0.0048 1>}
      uVRegion {bIsValid=false, max=<0 0>, min=<0 0>}
     }
     backgroundImageHovered {
      bIsDynamicallyLoaded=false
      drawAs=RoundedBox
      imageType=NoImage
      mirroring=NoMirror
      resourceName=None
      resourceObject=null
      tiling=NoTile
      imageSize=<0 0>
      margin=<ltrb 0 0 0 0>
      outlineSettings {
       bUseBrushTransparency=false
       roundingType=FixedRadius
       width=1
       color {colorUseRule=UseColor_ColorTable, specifiedColor=<rgba 0 0 0 0>}
       cornerRadii=<4 4 4 4>
      }
      tintColor {colorUseRule=UseColor_Specified, specifiedColor=<rgba 0.0048 0.0048 0.0048 1>}
      uVRegion {bIsValid=false, max=<0 0>, min=<0 0>}
     }
     backgroundImageNormal {
      bIsDynamicallyLoaded=false
      drawAs=RoundedBox
      imageType=NoImage
      mirroring=NoMirror
      resourceName=None
      resourceObject=null
      tiling=NoTile
      imageSize=<0 0>
      margin=<ltrb 0 0 0 0>
      outlineSettings {
       bUseBrushTransparency=false
       roundingType=FixedRadius
       width=1
       color {colorUseRule=UseColor_ColorTable, specifiedColor=<rgba 0 0 0 0>}
       cornerRadii=<4 4 4 4>
      }
      tintColor {colorUseRule=UseColor_Specified, specifiedColor=<rgba 0.0048 0.0048 0.0048 1>}
      uVRegion {bIsValid=false, max=<0 0>, min=<0 0>}
     }
     backgroundImageReadOnly {
      bIsDynamicallyLoaded=false
      drawAs=RoundedBox
      imageType=NoImage
      mirroring=NoMirror
      resourceName=None
      resourceObject=null
      tiling=NoTile
      imageSize=<0 0>
      margin=<ltrb 0 0 0 0>
      outlineSettings {
       bUseBrushTransparency=false
       roundingType=FixedRadius
       width=1
       color {colorUseRule=UseColor_ColorTable, specifiedColor=<rgba 0 0 0 0>}
       cornerRadii=<4 4 4 4>
      }
      tintColor {colorUseRule=UseColor_Specified, specifiedColor=<rgba 0.0284 0.0284 0.0284 1>}
      uVRegion {bIsValid=false, max=<0 0>, min=<0 0>}
     }
     focusedForegroundColor {colorUseRule=UseColor_Specified, specifiedColor=<rgba 1 1 1 1>}
     foregroundColor {colorUseRule=UseColor_Specified, specifiedColor=<rgba 0.5271 0.5271 0.5271 1>}
     hScrollBarPadding=<ltrb 0 0 0 0>
     padding=<ltrb 8 3 8 4>
     readOnlyForegroundColor {colorUseRule=UseColor_Foreground, specifiedColor=<rgba 1 0 1 1>}
     scrollBarStyle {
      thickness=8
      draggedThumbImage {
       bIsDynamicallyLoaded=false
       drawAs=RoundedBox
       imageType=NoImage
       mirroring=NoMirror
       resourceName=None
       resourceObject=null
       tiling=NoTile
       imageSize=<0 0>
       margin=<ltrb 0 0 0 0>
       outlineSettings {
        bUseBrushTransparency=false
        roundingType=FixedRadius
        width=0
        color {colorUseRule=UseColor_Specified, specifiedColor=<rgba 0 0 0 0>}
        cornerRadii=<4 4 4 4>
       }
       tintColor {colorUseRule=UseColor_Specified, specifiedColor=<rgba 0.2159 0.2159 0.2159 1>}
       uVRegion {bIsValid=false, max=<0 0>, min=<0 0>}
      }
      horizontalBackgroundImage {
       bIsDynamicallyLoaded=false
       drawAs=NoDrawType
       imageType=NoImage
       mirroring=NoMirror
       resourceName=None
       resourceObject=null
       tiling=NoTile
       imageSize=<0 0>
       margin=<ltrb 0 0 0 0>
       outlineSettings {
        bUseBrushTransparency=false
        roundingType=HalfHeightRadius
        width=0
        color {colorUseRule=UseColor_Specified, specifiedColor=<rgba 0 0 0 0>}
        cornerRadii=<0 0 0 0>
       }
       tintColor {colorUseRule=UseColor_Specified, specifiedColor=<rgba 1 1 1 1>}
       uVRegion {bIsValid=false, max=<0 0>, min=<0 0>}
      }
      horizontalBottomSlotImage {
       bIsDynamicallyLoaded=false
       drawAs=NoDrawType
       imageType=NoImage
       mirroring=NoMirror
       resourceName=None
       resourceObject=null
       tiling=NoTile
       imageSize=<0 0>
       margin=<ltrb 0 0 0 0>
       outlineSettings {
        bUseBrushTransparency=false
        roundingType=HalfHeightRadius
        width=0
        color {colorUseRule=UseColor_Specified, specifiedColor=<rgba 0 0 0 0>}
        cornerRadii=<0 0 0 0>
       }
       tintColor {colorUseRule=UseColor_Specified, specifiedColor=<rgba 1 1 1 1>}
       uVRegion {bIsValid=false, max=<0 0>, min=<0 0>}
      }
      horizontalTopSlotImage {
       bIsDynamicallyLoaded=false
       drawAs=NoDrawType
       imageType=NoImage
       mirroring=NoMirror
       resourceName=None
       resourceObject=null
       tiling=NoTile
       imageSize=<0 0>
       margin=<ltrb 0 0 0 0>
       outlineSettings {
        bUseBrushTransparency=false
        roundingType=HalfHeightRadius
        width=0
        color {colorUseRule=UseColor_Specified, specifiedColor=<rgba 0 0 0 0>}
        cornerRadii=<0 0 0 0>
       }
       tintColor {colorUseRule=UseColor_Specified, specifiedColor=<rgba 1 1 1 1>}
       uVRegion {bIsValid=false, max=<0 0>, min=<0 0>}
      }
      hoveredThumbImage {
       bIsDynamicallyLoaded=false
       drawAs=RoundedBox
       imageType=NoImage
       mirroring=NoMirror
       resourceName=None
       resourceObject=null
       tiling=NoTile
       imageSize=<0 0>
       margin=<ltrb 0 0 0 0>
       outlineSettings {
        bUseBrushTransparency=false
        roundingType=FixedRadius
        width=0
        color {colorUseRule=UseColor_Specified, specifiedColor=<rgba 0 0 0 0>}
        cornerRadii=<4 4 4 4>
       }
       tintColor {colorUseRule=UseColor_Specified, specifiedColor=<rgba 0.2159 0.2159 0.2159 1>}
       uVRegion {bIsValid=false, max=<0 0>, min=<0 0>}
      }
      normalThumbImage {
       bIsDynamicallyLoaded=false
       drawAs=RoundedBox
       imageType=NoImage
       mirroring=NoMirror
       resourceName=None
       resourceObject=null
       tiling=NoTile
       imageSize=<0 0>
       margin=<ltrb 0 0 0 0>
       outlineSettings {
        bUseBrushTransparency=false
        roundingType=FixedRadius
        width=0
        color {colorUseRule=UseColor_Specified, specifiedColor=<rgba 0 0 0 0>}
        cornerRadii=<4 4 4 4>
       }
       tintColor {colorUseRule=UseColor_Specified, specifiedColor=<rgba 0.0953 0.0953 0.0953 1>}
       uVRegion {bIsValid=false, max=<0 0>, min=<0 0>}
      }
      verticalBackgroundImage {
       bIsDynamicallyLoaded=false
       drawAs=NoDrawType
       imageType=NoImage
       mirroring=NoMirror
       resourceName=None
       resourceObject=null
       tiling=NoTile
       imageSize=<0 0>
       margin=<ltrb 0 0 0 0>
       outlineSettings {
        bUseBrushTransparency=false
        roundingType=HalfHeightRadius
        width=0
        color {colorUseRule=UseColor_Specified, specifiedColor=<rgba 0 0 0 0>}
        cornerRadii=<0 0 0 0>
       }
       tintColor {colorUseRule=UseColor_Specified, specifiedColor=<rgba 1 1 1 1>}
       uVRegion {bIsValid=false, max=<0 0>, min=<0 0>}
      }
      verticalBottomSlotImage {
       bIsDynamicallyLoaded=false
       drawAs=NoDrawType
       imageType=NoImage
       mirroring=NoMirror
       resourceName=None
       resourceObject=null
       tiling=NoTile
       imageSize=<0 0>
       margin=<ltrb 0 0 0 0>
       outlineSettings {
        bUseBrushTransparency=false
        roundingType=HalfHeightRadius
        width=0
        color {colorUseRule=UseColor_Specified, specifiedColor=<rgba 0 0 0 0>}
        cornerRadii=<0 0 0 0>
       }
       tintColor {colorUseRule=UseColor_Specified, specifiedColor=<rgba 1 1 1 1>}
       uVRegion {bIsValid=false, max=<0 0>, min=<0 0>}
      }
      verticalTopSlotImage {
       bIsDynamicallyLoaded=false
       drawAs=NoDrawType
       imageType=NoImage
       mirroring=NoMirror
       resourceName=None
       resourceObject=null
       tiling=NoTile
       imageSize=<0 0>
       margin=<ltrb 0 0 0 0>
       outlineSettings {
        bUseBrushTransparency=false
        roundingType=HalfHeightRadius
        width=0
        color {colorUseRule=UseColor_Specified, specifiedColor=<rgba 0 0 0 0>}
        cornerRadii=<0 0 0 0>
       }
       tintColor {colorUseRule=UseColor_Specified, specifiedColor=<rgba 1 1 1 1>}
       uVRegion {bIsValid=false, max=<0 0>, min=<0 0>}
      }
     }
     textStyle {
      overflowPolicy=Clip
      transformPolicy=None
      colorAndOpacity {colorUseRule=UseColor_Foreground, specifiedColor=<rgba 1 0 1 1>}
      font {
       bForceMonospaced=false
       bMaterialIsStencil=false
       fontMaterial=null
       fontObject=&RobotoTiny
       letterSpacing=0
       monospacedWidth=1
       size=12
       skewAmount=0
       typefaceFontName=Light
       outlineSettings {
        bApplyOutlineToDropShadows=false
        bMiteredCorners=false
        bSeparateFillAlpha=false
        outlineMaterial=null
        outlineSize=0
        outlineColor=<rgba 0 0 0 1>
       }
      }
      highlightColor {colorUseRule=UseColor_ColorTable, specifiedColor=<rgba 0 0 0 0>}
      highlightShape {
       bIsDynamicallyLoaded=false
       drawAs=Image
       imageType=NoImage
       mirroring=NoMirror
       resourceName=None
       resourceObject=null
       tiling=NoTile
       imageSize=<0 0>
       margin=<ltrb 0 0 0 0>
       outlineSettings {
        bUseBrushTransparency=false
        roundingType=HalfHeightRadius
        width=0
        color {colorUseRule=UseColor_Specified, specifiedColor=<rgba 0 0 0 0>}
        cornerRadii=<0 0 0 0>
       }
       tintColor {colorUseRule=UseColor_ColorTable, specifiedColor=<rgba 0 0 0 0>}
       uVRegion {bIsValid=false, max=<0 0>, min=<0 0>}
      }
      selectedBackgroundColor {colorUseRule=UseColor_ColorTable, specifiedColor=<rgba 0 0 0 0>}
      shadowColorAndOpacity=<rgba 0 0 0 1>
      shadowOffset=<0 0>
      strikeBrush {
       bIsDynamicallyLoaded=false
       drawAs=Image
       imageType=NoImage
       mirroring=NoMirror
       resourceName=None
       resourceObject=null
       tiling=NoTile
       imageSize=<32 32>
       margin=<ltrb 0 0 0 0>
       outlineSettings {
        bUseBrushTransparency=false
        roundingType=HalfHeightRadius
        width=0
        color {colorUseRule=UseColor_Specified, specifiedColor=<rgba 0 0 0 0>}
        cornerRadii=<0 0 0 0>
       }
       tintColor {colorUseRule=UseColor_Specified, specifiedColor=<rgba 1 1 1 1>}
       uVRegion {bIsValid=false, max=<0 0>, min=<0 0>}
      }
      underlineBrush {
       bIsDynamicallyLoaded=false
       drawAs=Image
       imageType=NoImage
       mirroring=NoMirror
       resourceName=None
       resourceObject=null
       tiling=NoTile
       imageSize=<32 32>
       margin=<ltrb 0 0 0 0>
       outlineSettings {
        bUseBrushTransparency=false
        roundingType=HalfHeightRadius
        width=0
        color {colorUseRule=UseColor_Specified, specifiedColor=<rgba 0 0 0 0>}
        cornerRadii=<0 0 0 0>
       }
       tintColor {colorUseRule=UseColor_Specified, specifiedColor=<rgba 1 1 1 1>}
       uVRegion {bIsValid=false, max=<0 0>, min=<0 0>}
      }
     }
     vScrollBarPadding=<ltrb 0 0 0 0>
    }
   }
   slot {class=HorizontalBoxSlot, properties {size {sizeRule=Fill, value=1}}}
  }
  EditorUtilityCheckBox_1 {
   widgetClassPath=&EditorUtilityCheckBox
   bIsVariable=true
   parent=HorizontalBox_302
   widget=EditorUtilityCheckBox_1
   slot {class=HorizontalBoxSlot}
  }
  ImageSize_9 {
   widgetClassPath=&TextBlock
   parent=EditorUtilityCheckBox_1
   widget=ImageSize_9
   properties {
    text=SearchDepend
    font {
     bForceMonospaced=false
     bMaterialIsStencil=false
     fontMaterial=null
     fontObject=&RobotoTiny
     letterSpacing=0
     monospacedWidth=1
     size=7.5
     skewAmount=0
     typefaceFontName=Light
     outlineSettings {
      bApplyOutlineToDropShadows=false
      bMiteredCorners=false
      bSeparateFillAlpha=false
      outlineMaterial=null
      outlineSize=0
      outlineColor=<rgba 0 0 0 1>
     }
    }
   }
   slot {class=PanelSlot}
  }
  HorizontalBox_425 {
   widgetClassPath=&HorizontalBox
   parent=VerticalBox_259
   widget=HorizontalBox_425
   slot {class=VerticalBoxSlot, properties {horizontalAlignment=HAlign_Left, verticalAlignment=VAlign_Center, padding=<ltrb 10 3 0 2>}}
  }
  ImageSize_6 {
   widgetClassPath=&TextBlock
   parent=HorizontalBox_425
   widget=ImageSize_6
   properties {
    text="VersionRange,Color,Search"
    font {
     bForceMonospaced=false
     bMaterialIsStencil=false
     fontMaterial=null
     fontObject=&Burbank
     letterSpacing=0
     monospacedWidth=1
     size=12
     skewAmount=0
     typefaceFontName=Burbank-Black
     outlineSettings {
      bApplyOutlineToDropShadows=false
      bMiteredCorners=false
      bSeparateFillAlpha=false
      outlineMaterial=null
      outlineSize=0
      outlineColor=<rgba 0 0 0 1>
     }
    }
   }
   slot {class=HorizontalBoxSlot}
  }
  SizeBox_6 {
   widgetClassPath=&SizeBox
   parent=HorizontalBox_425
   widget=SizeBox_6
   properties {bOverride_MaxAspectRatio=true, bOverride_MinAspectRatio=true, bOverride_WidthOverride=true, widthOverride=20}
   slot {class=HorizontalBoxSlot}
  }
  Border_4 {widgetClassPath=&Border, parent=SizeBox_6, widget=Border_4, slot {class=SizeBoxSlot}, properties=~}
  EditorUtilityButton_14 {
   widgetClassPath=&EditorUtilityButton
   parent=HorizontalBox_425
   widget=EditorUtilityButton_14
   properties {backgroundColor=<rgba 0.5976 0.6299 0.733 1>}
   slot {class=HorizontalBoxSlot, properties {padding=<ltrb 50 0 20 0>}}
  }
  ImageSize_15 {
   widgetClassPath=&TextBlock
   parent=EditorUtilityButton_14
   widget=ImageSize_15
   properties {
    text=Disable Filter
    font {
     bForceMonospaced=false
     bMaterialIsStencil=false
     fontMaterial=null
     fontObject=&RobotoTiny
     letterSpacing=0
     monospacedWidth=1
     size=12
     skewAmount=0
     typefaceFontName=Light
     outlineSettings {
      bApplyOutlineToDropShadows=false
      bMiteredCorners=false
      bSeparateFillAlpha=false
      outlineMaterial=null
      outlineSize=0
      outlineColor=<rgba 0 0 0 1>
     }
    }
   }
   slot {class=ButtonSlot, properties {horizontalAlignment=HAlign_Left, padding=<ltrb 10 0 10 0>}}
  }
  TagShow {
   widgetClassPath=&TextBlock
   bIsVariable=true
   parent=VerticalBox_259
   widget=TagShow
   properties {
    text=Tags
    colorAndOpacity {colorUseRule=UseColor_Specified, specifiedColor=<rgba 1 0 0 1>}
    font {
     bForceMonospaced=false
     bMaterialIsStencil=false
     fontMaterial=null
     fontObject=&Burbank
     letterSpacing=0
     monospacedWidth=1
     size=12
     skewAmount=0
     typefaceFontName=Burbank-Black
     outlineSettings {
      bApplyOutlineToDropShadows=false
      bMiteredCorners=false
      bSeparateFillAlpha=false
      outlineMaterial=null
      outlineSize=0
      outlineColor=<rgba 0 0 0 1>
     }
    }
   }
   slot {class=VerticalBoxSlot, properties {padding=<ltrb 10 3 0 2>}}
  }
  TileView_218 {
   widgetClassPath=/Script/UMG.TileView
   bIsVariable=true
   parent=VerticalBox_259
   widget=TileView_218
   properties {
    accessibleBehavior=Auto
    bEnableScrollAnimation=true
    bOverrideAccessibleDefaults=true
    bReturnFocusToSelection=true
    entryHeight=200
    entryWidgetClass=&/EUW_PreviewButton.EUW_PreviewButton_C
    entryWidth=200
    horizontalEntrySpacing=5
    numDesignerPreviewEntries=0
    scrollbarDisabledVisibility=Visible
    tileAlignment=LeftAligned
    verticalEntrySpacing=5
    scrollBarStyle {
     thickness=8
     draggedThumbImage {
      bIsDynamicallyLoaded=false
      drawAs=RoundedBox
      imageType=NoImage
      mirroring=NoMirror
      resourceName=None
      resourceObject=null
      tiling=NoTile
      imageSize=<0 0>
      margin=<ltrb 0 0 0 0>
      outlineSettings {
       bUseBrushTransparency=false
       roundingType=FixedRadius
       width=0
       color {colorUseRule=UseColor_Specified, specifiedColor=<rgba 0 0 0 0>}
       cornerRadii=<4 4 4 4>
      }
      tintColor {colorUseRule=UseColor_Specified, specifiedColor=<rgba 0.022 0.022 0.022 1>}
      uVRegion {bIsValid=false, max=<0 0>, min=<0 0>}
     }
     horizontalBackgroundImage {
      bIsDynamicallyLoaded=false
      drawAs=NoDrawType
      imageType=NoImage
      mirroring=NoMirror
      resourceName=None
      resourceObject=null
      tiling=NoTile
      imageSize=<0 0>
      margin=<ltrb 0 0 0 0>
      outlineSettings {
       bUseBrushTransparency=false
       roundingType=HalfHeightRadius
       width=0
       color {colorUseRule=UseColor_Specified, specifiedColor=<rgba 0 0 0 0>}
       cornerRadii=<0 0 0 0>
      }
      tintColor {colorUseRule=UseColor_Specified, specifiedColor=<rgba 1 1 1 1>}
      uVRegion {bIsValid=false, max=<0 0>, min=<0 0>}
     }
     horizontalBottomSlotImage {
      bIsDynamicallyLoaded=false
      drawAs=NoDrawType
      imageType=NoImage
      mirroring=NoMirror
      resourceName=None
      resourceObject=null
      tiling=NoTile
      imageSize=<0 0>
      margin=<ltrb 0 0 0 0>
      outlineSettings {
       bUseBrushTransparency=false
       roundingType=HalfHeightRadius
       width=0
       color {colorUseRule=UseColor_Specified, specifiedColor=<rgba 0 0 0 0>}
       cornerRadii=<0 0 0 0>
      }
      tintColor {colorUseRule=UseColor_Specified, specifiedColor=<rgba 1 1 1 1>}
      uVRegion {bIsValid=false, max=<0 0>, min=<0 0>}
     }
     horizontalTopSlotImage {
      bIsDynamicallyLoaded=false
      drawAs=NoDrawType
      imageType=NoImage
      mirroring=NoMirror
      resourceName=None
      resourceObject=null
      tiling=NoTile
      imageSize=<0 0>
      margin=<ltrb 0 0 0 0>
      outlineSettings {
       bUseBrushTransparency=false
       roundingType=HalfHeightRadius
       width=0
       color {colorUseRule=UseColor_Specified, specifiedColor=<rgba 0 0 0 0>}
       cornerRadii=<0 0 0 0>
      }
      tintColor {colorUseRule=UseColor_Specified, specifiedColor=<rgba 1 1 1 1>}
      uVRegion {bIsValid=false, max=<0 0>, min=<0 0>}
     }
     hoveredThumbImage {
      bIsDynamicallyLoaded=false
      drawAs=RoundedBox
      imageType=NoImage
      mirroring=NoMirror
      resourceName=None
      resourceObject=null
      tiling=NoTile
      imageSize=<0 0>
      margin=<ltrb 0 0 0 0>
      outlineSettings {
       bUseBrushTransparency=false
       roundingType=FixedRadius
       width=0
       color {colorUseRule=UseColor_Specified, specifiedColor=<rgba 0 0 0 0>}
       cornerRadii=<4 4 4 4>
      }
      tintColor {colorUseRule=UseColor_Specified, specifiedColor=<rgba 0.057 0.057 0.057 1>}
      uVRegion {bIsValid=false, max=<0 0>, min=<0 0>}
     }
     normalThumbImage {
      bIsDynamicallyLoaded=false
      drawAs=RoundedBox
      imageType=NoImage
      mirroring=NoMirror
      resourceName=None
      resourceObject=null
      tiling=NoTile
      imageSize=<0 0>
      margin=<ltrb 0 0 0 0>
      outlineSettings {
       bUseBrushTransparency=false
       roundingType=FixedRadius
       width=0
       color {colorUseRule=UseColor_Specified, specifiedColor=<rgba 0 0 0 0>}
       cornerRadii=<4 4 4 4>
      }
      tintColor {colorUseRule=UseColor_Specified, specifiedColor=<rgba 0.0395 0.0395 0.0395 1>}
      uVRegion {bIsValid=false, max=<0 0>, min=<0 0>}
     }
     verticalBackgroundImage {
      bIsDynamicallyLoaded=false
      drawAs=NoDrawType
      imageType=NoImage
      mirroring=NoMirror
      resourceName=None
      resourceObject=null
      tiling=NoTile
      imageSize=<0 0>
      margin=<ltrb 0 0 0 0>
      outlineSettings {
       bUseBrushTransparency=false
       roundingType=HalfHeightRadius
       width=0
       color {colorUseRule=UseColor_Specified, specifiedColor=<rgba 0 0 0 0>}
       cornerRadii=<0 0 0 0>
      }
      tintColor {colorUseRule=UseColor_Specified, specifiedColor=<rgba 1 1 1 1>}
      uVRegion {bIsValid=false, max=<0 0>, min=<0 0>}
     }
     verticalBottomSlotImage {
      bIsDynamicallyLoaded=false
      drawAs=NoDrawType
      imageType=NoImage
      mirroring=NoMirror
      resourceName=None
      resourceObject=null
      tiling=NoTile
      imageSize=<0 0>
      margin=<ltrb 0 0 0 0>
      outlineSettings {
       bUseBrushTransparency=false
       roundingType=HalfHeightRadius
       width=0
       color {colorUseRule=UseColor_Specified, specifiedColor=<rgba 0 0 0 0>}
       cornerRadii=<0 0 0 0>
      }
      tintColor {colorUseRule=UseColor_Specified, specifiedColor=<rgba 1 1 1 1>}
      uVRegion {bIsValid=false, max=<0 0>, min=<0 0>}
     }
     verticalTopSlotImage {
      bIsDynamicallyLoaded=false
      drawAs=NoDrawType
      imageType=NoImage
      mirroring=NoMirror
      resourceName=None
      resourceObject=null
      tiling=NoTile
      imageSize=<0 0>
      margin=<ltrb 0 0 0 0>
      outlineSettings {
       bUseBrushTransparency=false
       roundingType=HalfHeightRadius
       width=0
       color {colorUseRule=UseColor_Specified, specifiedColor=<rgba 0 0 0 0>}
       cornerRadii=<0 0 0 0>
      }
      tintColor {colorUseRule=UseColor_Specified, specifiedColor=<rgba 1 1 1 1>}
      uVRegion {bIsValid=false, max=<0 0>, min=<0 0>}
     }
    }
   }
   slot {class=VerticalBoxSlot, properties {padding=<ltrb 10 5 10 5>, size {sizeRule=Fill, value=1}}}
  }
  HorizontalBox_300 {widgetClassPath=&HorizontalBox, parent=VerticalBox_259, widget=HorizontalBox_300, slot {class=VerticalBoxSlot}}
  EditorUtilityButton_3 {
   widgetClassPath=&EditorUtilityButton
   parent=HorizontalBox_300
   widget=EditorUtilityButton_3
   properties {backgroundColor=<rgba 0.8153 0.8594 1 1>}
   slot {class=HorizontalBoxSlot, properties {size {sizeRule=Fill, value=1}}}
  }
  ImageSize_3 {
   widgetClassPath=&TextBlock
   parent=EditorUtilityButton_3
   widget=ImageSize_3
   properties {
    text=-
    font {
     bForceMonospaced=false
     bMaterialIsStencil=false
     fontMaterial=null
     fontObject=&HeadingNow
     letterSpacing=0
     monospacedWidth=1
     size=24
     skewAmount=0
     typefaceFontName=86 Bold
     outlineSettings {
      bApplyOutlineToDropShadows=false
      bMiteredCorners=false
      bSeparateFillAlpha=false
      outlineMaterial=null
      outlineSize=0
      outlineColor=<rgba 0 0 0 1>
     }
    }
   }
   slot {class=ButtonSlot, properties {padding=<ltrb 0 0 0 0>}}
  }
  PageNum {
   widgetClassPath=&TextBlock
   bIsVariable=true
   parent=HorizontalBox_300
   widget=PageNum
   properties {
    justification=Center
    text=Page
    font {
     bForceMonospaced=false
     bMaterialIsStencil=false
     fontMaterial=null
     fontObject=&Burbank
     letterSpacing=0
     monospacedWidth=1
     size=24
     skewAmount=0
     typefaceFontName=Burbank-Black
     outlineSettings {
      bApplyOutlineToDropShadows=false
      bMiteredCorners=false
      bSeparateFillAlpha=false
      outlineMaterial=null
      outlineSize=0
      outlineColor=<rgba 0 0 0 1>
     }
    }
   }
   slot {class=HorizontalBoxSlot, properties {size {sizeRule=Fill, value=1}}}
  }
  EditorUtilityButton_4 {
   widgetClassPath=&EditorUtilityButton
   parent=HorizontalBox_300
   widget=EditorUtilityButton_4
   properties {backgroundColor=<rgba 0.8153 0.8594 1 1>}
   slot {class=HorizontalBoxSlot, properties {size {sizeRule=Fill, value=1}}}
  }
  ImageSize_4 {
   widgetClassPath=&TextBlock
   parent=EditorUtilityButton_4
   widget=ImageSize_4
   properties {
    text=+
    font {
     bForceMonospaced=false
     bMaterialIsStencil=false
     fontMaterial=null
     fontObject=&HeadingNow
     letterSpacing=0
     monospacedWidth=1
     size=24
     skewAmount=0
     typefaceFontName=86 Bold
     outlineSettings {
      bApplyOutlineToDropShadows=false
      bMiteredCorners=false
      bSeparateFillAlpha=false
      outlineMaterial=null
      outlineSize=0
      outlineColor=<rgba 0 0 0 1>
     }
    }
   }
   slot {class=ButtonSlot, properties {padding=<ltrb 0 0 0 0>}}
  }
  EditorUtilityButton_91 {
   widgetClassPath=&EditorUtilityButton
   parent=HorizontalBox_300
   widget=EditorUtilityButton_91
   properties {backgroundColor=<rgba 0.8153 0.8594 1 1>}
   slot {class=HorizontalBoxSlot}
  }
  TextBlock_117 {
   widgetClassPath=&TextBlock
   parent=EditorUtilityButton_91
   widget=TextBlock_117
   properties {
    text=Refresh
    font {
     bForceMonospaced=false
     bMaterialIsStencil=false
     fontMaterial=null
     fontObject=&Burbank
     letterSpacing=0
     monospacedWidth=1
     size=24
     skewAmount=0
     typefaceFontName=Burbank-Black
     outlineSettings {
      bApplyOutlineToDropShadows=false
      bMiteredCorners=false
      bSeparateFillAlpha=false
      outlineMaterial=null
      outlineSize=0
      outlineColor=<rgba 0 0 0 1>
     }
    }
   }
   slot {class=ButtonSlot}
  }
  HorizontalBox_208 {widgetClassPath=&HorizontalBox, parent=VerticalBox_259, widget=HorizontalBox_208, slot {class=VerticalBoxSlot}}
  ImageSize {
   widgetClassPath=&TextBlock
   parent=HorizontalBox_208
   widget=ImageSize
   properties {
    text=ImageSize
    font {
     bForceMonospaced=false
     bMaterialIsStencil=false
     fontMaterial=null
     fontObject=&RobotoTiny
     letterSpacing=0
     monospacedWidth=1
     size=12
     skewAmount=0
     typefaceFontName=Light
     outlineSettings {
      bApplyOutlineToDropShadows=false
      bMiteredCorners=false
      bSeparateFillAlpha=false
      outlineMaterial=null
      outlineSize=0
      outlineColor=<rgba 0 0 0 1>
     }
    }
   }
   slot {class=HorizontalBoxSlot, properties {horizontalAlignment=HAlign_Left, verticalAlignment=VAlign_Center}}
  }
  SizeBox_1 {
   widgetClassPath=&SizeBox
   parent=HorizontalBox_208
   widget=SizeBox_1
   properties {heightOverride=40, widthOverride=250}
   slot {class=HorizontalBoxSlot, properties {padding=<ltrb 5 5 5 5>, size {sizeRule=Fill, value=1}}}
  }
  AnalogSlider_219 {
   widgetClassPath=&AnalogSlider
   bIsVariable=true
   parent=SizeBox_1
   widget=AnalogSlider_219
   properties {
    value=0.5
    widgetStyle {
     barThickness=10
     disabledBarImage {
      bIsDynamicallyLoaded=false
      drawAs=RoundedBox
      imageType=NoImage
      mirroring=NoMirror
      resourceName=None
      resourceObject=null
      tiling=NoTile
      imageSize=<0 0>
      margin=<ltrb 0 0 0 0>
      outlineSettings {
       bUseBrushTransparency=false
       roundingType=FixedRadius
       width=1
       color {colorUseRule=UseColor_ColorTable, specifiedColor=<rgba 0 0 0 0>}
       cornerRadii=<2 2 2 2>
      }
      tintColor {colorUseRule=UseColor_Specified, specifiedColor=<rgba 0.0048 0.0048 0.0048 1>}
      uVRegion {bIsValid=false, max=<0 0>, min=<0 0>}
     }
     disabledThumbImage {
      bIsDynamicallyLoaded=false
      drawAs=RoundedBox
      imageType=NoImage
      mirroring=NoMirror
      resourceName=None
      resourceObject=null
      tiling=NoTile
      imageSize=<15 15>
      margin=<ltrb 0 0 0 0>
      outlineSettings {
       bUseBrushTransparency=false
       roundingType=HalfHeightRadius
       width=0
       color {colorUseRule=UseColor_Specified, specifiedColor=<rgba 0 0 0 0>}
       cornerRadii=<0 0 0 0>
      }
      tintColor {colorUseRule=UseColor_Specified, specifiedColor=<rgba 0.5271 0.5271 0.5271 1>}
      uVRegion {bIsValid=false, max=<0 0>, min=<0 0>}
     }
     hoveredBarImage {
      bIsDynamicallyLoaded=false
      drawAs=RoundedBox
      imageType=NoImage
      mirroring=NoMirror
      resourceName=None
      resourceObject=null
      tiling=NoTile
      imageSize=<0 0>
      margin=<ltrb 0 0 0 0>
      outlineSettings {
       bUseBrushTransparency=false
       roundingType=FixedRadius
       width=1
       color {colorUseRule=UseColor_ColorTable, specifiedColor=<rgba 0 0 0 0>}
       cornerRadii=<2 2 2 2>
      }
      tintColor {colorUseRule=UseColor_Specified, specifiedColor=<rgba 0.0048 0.0048 0.0048 1>}
      uVRegion {bIsValid=false, max=<0 0>, min=<0 0>}
     }
     hoveredThumbImage {
      bIsDynamicallyLoaded=false
      drawAs=RoundedBox
      imageType=NoImage
      mirroring=NoMirror
      resourceName=None
      resourceObject=null
      tiling=NoTile
      imageSize=<15 15>
      margin=<ltrb 0 0 0 0>
      outlineSettings {
       bUseBrushTransparency=false
       roundingType=HalfHeightRadius
       width=0
       color {colorUseRule=UseColor_Specified, specifiedColor=<rgba 0 0 0 0>}
       cornerRadii=<0 0 0 0>
      }
      tintColor {colorUseRule=UseColor_Specified, specifiedColor=<rgba 1 1 1 1>}
      uVRegion {bIsValid=false, max=<0 0>, min=<0 0>}
     }
     normalBarImage {
      bIsDynamicallyLoaded=false
      drawAs=RoundedBox
      imageType=NoImage
      mirroring=NoMirror
      resourceName=None
      resourceObject=null
      tiling=NoTile
      imageSize=<0 0>
      margin=<ltrb 0 0 0 0>
      outlineSettings {
       bUseBrushTransparency=false
       roundingType=FixedRadius
       width=1
       color {colorUseRule=UseColor_ColorTable, specifiedColor=<rgba 0 0 0 0>}
       cornerRadii=<2 2 2 2>
      }
      tintColor {colorUseRule=UseColor_Specified, specifiedColor=<rgba 0.0048 0.0048 0.0048 1>}
      uVRegion {bIsValid=false, max=<0 0>, min=<0 0>}
     }
     normalThumbImage {
      bIsDynamicallyLoaded=false
      drawAs=RoundedBox
      imageType=NoImage
      mirroring=NoMirror
      resourceName=None
      resourceObject=null
      tiling=NoTile
      imageSize=<15 15>
      margin=<ltrb 0 0 0 0>
      outlineSettings {
       bUseBrushTransparency=false
       roundingType=HalfHeightRadius
       width=0
       color {colorUseRule=UseColor_Specified, specifiedColor=<rgba 0 0 0 0>}
       cornerRadii=<0 0 0 0>
      }
      tintColor {colorUseRule=UseColor_Specified, specifiedColor=<rgba 0.5271 0.5271 0.5271 1>}
      uVRegion {bIsValid=false, max=<0 0>, min=<0 0>}
     }
    }
   }
   slot {class=SizeBoxSlot}
  }
  ImageSize_1 {
   widgetClassPath=&TextBlock
   parent=HorizontalBox_208
   widget=ImageSize_1
   properties {
    text=PreviewSpeed
    visibility=Hidden
    font {
     bForceMonospaced=false
     bMaterialIsStencil=false
     fontMaterial=null
     fontObject=&RobotoTiny
     letterSpacing=0
     monospacedWidth=1
     size=12
     skewAmount=0
     typefaceFontName=Light
     outlineSettings {
      bApplyOutlineToDropShadows=false
      bMiteredCorners=false
      bSeparateFillAlpha=false
      outlineMaterial=null
      outlineSize=0
      outlineColor=<rgba 0 0 0 1>
     }
    }
   }
   slot {class=HorizontalBoxSlot, properties {horizontalAlignment=HAlign_Left, verticalAlignment=VAlign_Center}}
  }
  SizeBox {
   widgetClassPath=&SizeBox
   parent=HorizontalBox_208
   widget=SizeBox
   properties {heightOverride=40, widthOverride=250}
   slot {class=HorizontalBoxSlot, properties {padding=<ltrb 5 5 5 5>, size {sizeRule=Fill, value=1}}}
  }
  AnalogSlider {
   widgetClassPath=&AnalogSlider
   bIsVariable=true
   parent=SizeBox
   widget=AnalogSlider
   properties {
    value=0.5
    visibility=Hidden
    widgetStyle {
     barThickness=10
     disabledBarImage {
      bIsDynamicallyLoaded=false
      drawAs=RoundedBox
      imageType=NoImage
      mirroring=NoMirror
      resourceName=None
      resourceObject=null
      tiling=NoTile
      imageSize=<0 0>
      margin=<ltrb 0 0 0 0>
      outlineSettings {
       bUseBrushTransparency=false
       roundingType=FixedRadius
       width=1
       color {colorUseRule=UseColor_ColorTable, specifiedColor=<rgba 0 0 0 0>}
       cornerRadii=<2 2 2 2>
      }
      tintColor {colorUseRule=UseColor_Specified, specifiedColor=<rgba 0.0048 0.0048 0.0048 1>}
      uVRegion {bIsValid=false, max=<0 0>, min=<0 0>}
     }
     disabledThumbImage {
      bIsDynamicallyLoaded=false
      drawAs=RoundedBox
      imageType=NoImage
      mirroring=NoMirror
      resourceName=None
      resourceObject=null
      tiling=NoTile
      imageSize=<15 15>
      margin=<ltrb 0 0 0 0>
      outlineSettings {
       bUseBrushTransparency=false
       roundingType=HalfHeightRadius
       width=0
       color {colorUseRule=UseColor_Specified, specifiedColor=<rgba 0 0 0 0>}
       cornerRadii=<0 0 0 0>
      }
      tintColor {colorUseRule=UseColor_Specified, specifiedColor=<rgba 0.5271 0.5271 0.5271 1>}
      uVRegion {bIsValid=false, max=<0 0>, min=<0 0>}
     }
     hoveredBarImage {
      bIsDynamicallyLoaded=false
      drawAs=RoundedBox
      imageType=NoImage
      mirroring=NoMirror
      resourceName=None
      resourceObject=null
      tiling=NoTile
      imageSize=<0 0>
      margin=<ltrb 0 0 0 0>
      outlineSettings {
       bUseBrushTransparency=false
       roundingType=FixedRadius
       width=1
       color {colorUseRule=UseColor_ColorTable, specifiedColor=<rgba 0 0 0 0>}
       cornerRadii=<2 2 2 2>
      }
      tintColor {colorUseRule=UseColor_Specified, specifiedColor=<rgba 0.0048 0.0048 0.0048 1>}
      uVRegion {bIsValid=false, max=<0 0>, min=<0 0>}
     }
     hoveredThumbImage {
      bIsDynamicallyLoaded=false
      drawAs=RoundedBox
      imageType=NoImage
      mirroring=NoMirror
      resourceName=None
      resourceObject=null
      tiling=NoTile
      imageSize=<15 15>
      margin=<ltrb 0 0 0 0>
      outlineSettings {
       bUseBrushTransparency=false
       roundingType=HalfHeightRadius
       width=0
       color {colorUseRule=UseColor_Specified, specifiedColor=<rgba 0 0 0 0>}
       cornerRadii=<0 0 0 0>
      }
      tintColor {colorUseRule=UseColor_Specified, specifiedColor=<rgba 1 1 1 1>}
      uVRegion {bIsValid=false, max=<0 0>, min=<0 0>}
     }
     normalBarImage {
      bIsDynamicallyLoaded=false
      drawAs=RoundedBox
      imageType=NoImage
      mirroring=NoMirror
      resourceName=None
      resourceObject=null
      tiling=NoTile
      imageSize=<0 0>
      margin=<ltrb 0 0 0 0>
      outlineSettings {
       bUseBrushTransparency=false
       roundingType=FixedRadius
       width=1
       color {colorUseRule=UseColor_ColorTable, specifiedColor=<rgba 0 0 0 0>}
       cornerRadii=<2 2 2 2>
      }
      tintColor {colorUseRule=UseColor_Specified, specifiedColor=<rgba 0.0048 0.0048 0.0048 1>}
      uVRegion {bIsValid=false, max=<0 0>, min=<0 0>}
     }
     normalThumbImage {
      bIsDynamicallyLoaded=false
      drawAs=RoundedBox
      imageType=NoImage
      mirroring=NoMirror
      resourceName=None
      resourceObject=null
      tiling=NoTile
      imageSize=<15 15>
      margin=<ltrb 0 0 0 0>
      outlineSettings {
       bUseBrushTransparency=false
       roundingType=HalfHeightRadius
       width=0
       color {colorUseRule=UseColor_Specified, specifiedColor=<rgba 0 0 0 0>}
       cornerRadii=<0 0 0 0>
      }
      tintColor {colorUseRule=UseColor_Specified, specifiedColor=<rgba 0.5271 0.5271 0.5271 1>}
      uVRegion {bIsValid=false, max=<0 0>, min=<0 0>}
     }
    }
   }
   slot {class=SizeBoxSlot}
  }
  ImageSize_5 {
   widgetClassPath=&TextBlock
   parent=HorizontalBox_208
   widget=ImageSize_5
   properties {
    text=CurrentUser
    font {
     bForceMonospaced=false
     bMaterialIsStencil=false
     fontMaterial=null
     fontObject=&RobotoTiny
     letterSpacing=0
     monospacedWidth=1
     size=12
     skewAmount=0
     typefaceFontName=Light
     outlineSettings {
      bApplyOutlineToDropShadows=false
      bMiteredCorners=false
      bSeparateFillAlpha=false
      outlineMaterial=null
      outlineSize=0
      outlineColor=<rgba 0 0 0 1>
     }
    }
   }
   slot {class=HorizontalBoxSlot, properties {horizontalAlignment=HAlign_Left, verticalAlignment=VAlign_Center}}
  }
  EditorUtilityEditableTextBox_25 {
   widgetClassPath=&EditorUtilityEditableTextBox
   bIsVariable=true
   parent=HorizontalBox_208
   widget=EditorUtilityEditableTextBox_25
   properties {bIsEnabled=false, text=yi.xie}
   slot {class=HorizontalBoxSlot, properties {size {sizeRule=Fill, value=1}}}
  }
  HorizontalBox_4 {widgetClassPath=&HorizontalBox, parent=VerticalBox_259, widget=HorizontalBox_4, slot {class=VerticalBoxSlot}}
  EditorUtilityButton_299 {
   widgetClassPath=&EditorUtilityButton
   parent=HorizontalBox_4
   widget=EditorUtilityButton_299
   slot {class=HorizontalBoxSlot, properties {horizontalAlignment=HAlign_Left, verticalAlignment=VAlign_Bottom}}
  }
  ImageSize_13 {
   widgetClassPath=&TextBlock
   parent=EditorUtilityButton_299
   widget=ImageSize_13
   properties {
    text=Rebuild DataBase
    font {
     bForceMonospaced=false
     bMaterialIsStencil=false
     fontMaterial=null
     fontObject=&RobotoTiny
     letterSpacing=0
     monospacedWidth=1
     size=12
     skewAmount=0
     typefaceFontName=Light
     outlineSettings {
      bApplyOutlineToDropShadows=false
      bMiteredCorners=false
      bSeparateFillAlpha=false
      outlineMaterial=null
      outlineSize=0
      outlineColor=<rgba 0 0 0 1>
     }
    }
   }
   slot {class=ButtonSlot, properties {horizontalAlignment=HAlign_Left, padding=<ltrb 0 0 0 0>}}
  }
  LibPath {
   widgetClassPath=&EditorUtilityEditableTextBox
   bIsVariable=true
   parent=HorizontalBox_4
   widget=LibPath
   properties {
    hintText=Input your library path here
    justification=Center
    visibility=Collapsed
    widgetStyle {
     backgroundColor {colorUseRule=UseColor_Specified, specifiedColor=<rgba 1 1 1 1>}
     backgroundImageFocused {
      bIsDynamicallyLoaded=false
      drawAs=RoundedBox
      imageType=NoImage
      mirroring=NoMirror
      resourceName=None
      resourceObject=null
      tiling=NoTile
      imageSize=<0 0>
      margin=<ltrb 0 0 0 0>
      outlineSettings {
       bUseBrushTransparency=false
       roundingType=FixedRadius
       width=1
       color {colorUseRule=UseColor_ColorTable, specifiedColor=<rgba 0 0 0 0>}
       cornerRadii=<4 4 4 4>
      }
      tintColor {colorUseRule=UseColor_Specified, specifiedColor=<rgba 0.0048 0.0048 0.0048 1>}
      uVRegion {bIsValid=false, max=<0 0>, min=<0 0>}
     }
     backgroundImageHovered {
      bIsDynamicallyLoaded=false
      drawAs=RoundedBox
      imageType=NoImage
      mirroring=NoMirror
      resourceName=None
      resourceObject=null
      tiling=NoTile
      imageSize=<0 0>
      margin=<ltrb 0 0 0 0>
      outlineSettings {
       bUseBrushTransparency=false
       roundingType=FixedRadius
       width=1
       color {colorUseRule=UseColor_ColorTable, specifiedColor=<rgba 0 0 0 0>}
       cornerRadii=<4 4 4 4>
      }
      tintColor {colorUseRule=UseColor_Specified, specifiedColor=<rgba 0.0048 0.0048 0.0048 1>}
      uVRegion {bIsValid=false, max=<0 0>, min=<0 0>}
     }
     backgroundImageNormal {
      bIsDynamicallyLoaded=false
      drawAs=RoundedBox
      imageType=NoImage
      mirroring=NoMirror
      resourceName=None
      resourceObject=null
      tiling=NoTile
      imageSize=<0 0>
      margin=<ltrb 0 0 0 0>
      outlineSettings {
       bUseBrushTransparency=false
       roundingType=FixedRadius
       width=1
       color {colorUseRule=UseColor_ColorTable, specifiedColor=<rgba 0 0 0 0>}
       cornerRadii=<4 4 4 4>
      }
      tintColor {colorUseRule=UseColor_Specified, specifiedColor=<rgba 0.0048 0.0048 0.0048 1>}
      uVRegion {bIsValid=false, max=<0 0>, min=<0 0>}
     }
     backgroundImageReadOnly {
      bIsDynamicallyLoaded=false
      drawAs=RoundedBox
      imageType=NoImage
      mirroring=NoMirror
      resourceName=None
      resourceObject=null
      tiling=NoTile
      imageSize=<0 0>
      margin=<ltrb 0 0 0 0>
      outlineSettings {
       bUseBrushTransparency=false
       roundingType=FixedRadius
       width=1
       color {colorUseRule=UseColor_ColorTable, specifiedColor=<rgba 0 0 0 0>}
       cornerRadii=<4 4 4 4>
      }
      tintColor {colorUseRule=UseColor_Specified, specifiedColor=<rgba 0.0284 0.0284 0.0284 1>}
      uVRegion {bIsValid=false, max=<0 0>, min=<0 0>}
     }
     focusedForegroundColor {colorUseRule=UseColor_Specified, specifiedColor=<rgba 1 1 1 1>}
     foregroundColor {colorUseRule=UseColor_Specified, specifiedColor=<rgba 0.5271 0.5271 0.5271 1>}
     hScrollBarPadding=<ltrb 0 0 0 0>
     padding=<ltrb 0 0 0 0>
     readOnlyForegroundColor {colorUseRule=UseColor_Foreground, specifiedColor=<rgba 1 0 1 1>}
     scrollBarStyle {
      thickness=8
      draggedThumbImage {
       bIsDynamicallyLoaded=false
       drawAs=RoundedBox
       imageType=NoImage
       mirroring=NoMirror
       resourceName=None
       resourceObject=null
       tiling=NoTile
       imageSize=<0 0>
       margin=<ltrb 0 0 0 0>
       outlineSettings {
        bUseBrushTransparency=false
        roundingType=FixedRadius
        width=0
        color {colorUseRule=UseColor_Specified, specifiedColor=<rgba 0 0 0 0>}
        cornerRadii=<4 4 4 4>
       }
       tintColor {colorUseRule=UseColor_Specified, specifiedColor=<rgba 0.2159 0.2159 0.2159 1>}
       uVRegion {bIsValid=false, max=<0 0>, min=<0 0>}
      }
      horizontalBackgroundImage {
       bIsDynamicallyLoaded=false
       drawAs=NoDrawType
       imageType=NoImage
       mirroring=NoMirror
       resourceName=None
       resourceObject=null
       tiling=NoTile
       imageSize=<0 0>
       margin=<ltrb 0 0 0 0>
       outlineSettings {
        bUseBrushTransparency=false
        roundingType=HalfHeightRadius
        width=0
        color {colorUseRule=UseColor_Specified, specifiedColor=<rgba 0 0 0 0>}
        cornerRadii=<0 0 0 0>
       }
       tintColor {colorUseRule=UseColor_Specified, specifiedColor=<rgba 1 1 1 1>}
       uVRegion {bIsValid=false, max=<0 0>, min=<0 0>}
      }
      horizontalBottomSlotImage {
       bIsDynamicallyLoaded=false
       drawAs=NoDrawType
       imageType=NoImage
       mirroring=NoMirror
       resourceName=None
       resourceObject=null
       tiling=NoTile
       imageSize=<0 0>
       margin=<ltrb 0 0 0 0>
       outlineSettings {
        bUseBrushTransparency=false
        roundingType=HalfHeightRadius
        width=0
        color {colorUseRule=UseColor_Specified, specifiedColor=<rgba 0 0 0 0>}
        cornerRadii=<0 0 0 0>
       }
       tintColor {colorUseRule=UseColor_Specified, specifiedColor=<rgba 1 1 1 1>}
       uVRegion {bIsValid=false, max=<0 0>, min=<0 0>}
      }
      horizontalTopSlotImage {
       bIsDynamicallyLoaded=false
       drawAs=NoDrawType
       imageType=NoImage
       mirroring=NoMirror
       resourceName=None
       resourceObject=null
       tiling=NoTile
       imageSize=<0 0>
       margin=<ltrb 0 0 0 0>
       outlineSettings {
        bUseBrushTransparency=false
        roundingType=HalfHeightRadius
        width=0
        color {colorUseRule=UseColor_Specified, specifiedColor=<rgba 0 0 0 0>}
        cornerRadii=<0 0 0 0>
       }
       tintColor {colorUseRule=UseColor_Specified, specifiedColor=<rgba 1 1 1 1>}
       uVRegion {bIsValid=false, max=<0 0>, min=<0 0>}
      }
      hoveredThumbImage {
       bIsDynamicallyLoaded=false
       drawAs=RoundedBox
       imageType=NoImage
       mirroring=NoMirror
       resourceName=None
       resourceObject=null
       tiling=NoTile
       imageSize=<0 0>
       margin=<ltrb 0 0 0 0>
       outlineSettings {
        bUseBrushTransparency=false
        roundingType=FixedRadius
        width=0
        color {colorUseRule=UseColor_Specified, specifiedColor=<rgba 0 0 0 0>}
        cornerRadii=<4 4 4 4>
       }
       tintColor {colorUseRule=UseColor_Specified, specifiedColor=<rgba 0.2159 0.2159 0.2159 1>}
       uVRegion {bIsValid=false, max=<0 0>, min=<0 0>}
      }
      normalThumbImage {
       bIsDynamicallyLoaded=false
       drawAs=RoundedBox
       imageType=NoImage
       mirroring=NoMirror
       resourceName=None
       resourceObject=null
       tiling=NoTile
       imageSize=<0 0>
       margin=<ltrb 0 0 0 0>
       outlineSettings {
        bUseBrushTransparency=false
        roundingType=FixedRadius
        width=0
        color {colorUseRule=UseColor_Specified, specifiedColor=<rgba 0 0 0 0>}
        cornerRadii=<4 4 4 4>
       }
       tintColor {colorUseRule=UseColor_Specified, specifiedColor=<rgba 0.0953 0.0953 0.0953 1>}
       uVRegion {bIsValid=false, max=<0 0>, min=<0 0>}
      }
      verticalBackgroundImage {
       bIsDynamicallyLoaded=false
       drawAs=NoDrawType
       imageType=NoImage
       mirroring=NoMirror
       resourceName=None
       resourceObject=null
       tiling=NoTile
       imageSize=<0 0>
       margin=<ltrb 0 0 0 0>
       outlineSettings {
        bUseBrushTransparency=false
        roundingType=HalfHeightRadius
        width=0
        color {colorUseRule=UseColor_Specified, specifiedColor=<rgba 0 0 0 0>}
        cornerRadii=<0 0 0 0>
       }
       tintColor {colorUseRule=UseColor_Specified, specifiedColor=<rgba 1 1 1 1>}
       uVRegion {bIsValid=false, max=<0 0>, min=<0 0>}
      }
      verticalBottomSlotImage {
       bIsDynamicallyLoaded=false
       drawAs=NoDrawType
       imageType=NoImage
       mirroring=NoMirror
       resourceName=None
       resourceObject=null
       tiling=NoTile
       imageSize=<0 0>
       margin=<ltrb 0 0 0 0>
       outlineSettings {
        bUseBrushTransparency=false
        roundingType=HalfHeightRadius
        width=0
        color {colorUseRule=UseColor_Specified, specifiedColor=<rgba 0 0 0 0>}
        cornerRadii=<0 0 0 0>
       }
       tintColor {colorUseRule=UseColor_Specified, specifiedColor=<rgba 1 1 1 1>}
       uVRegion {bIsValid=false, max=<0 0>, min=<0 0>}
      }
      verticalTopSlotImage {
       bIsDynamicallyLoaded=false
       drawAs=NoDrawType
       imageType=NoImage
       mirroring=NoMirror
       resourceName=None
       resourceObject=null
       tiling=NoTile
       imageSize=<0 0>
       margin=<ltrb 0 0 0 0>
       outlineSettings {
        bUseBrushTransparency=false
        roundingType=HalfHeightRadius
        width=0
        color {colorUseRule=UseColor_Specified, specifiedColor=<rgba 0 0 0 0>}
        cornerRadii=<0 0 0 0>
       }
       tintColor {colorUseRule=UseColor_Specified, specifiedColor=<rgba 1 1 1 1>}
       uVRegion {bIsValid=false, max=<0 0>, min=<0 0>}
      }
     }
     textStyle {
      overflowPolicy=Clip
      transformPolicy=None
      colorAndOpacity {colorUseRule=UseColor_Foreground, specifiedColor=<rgba 1 0 1 1>}
      font {
       bForceMonospaced=false
       bMaterialIsStencil=false
       fontMaterial=null
       fontObject=null
       letterSpacing=0
       monospacedWidth=1
       size=12
       skewAmount=0
       typefaceFontName=Regular
       outlineSettings {
        bApplyOutlineToDropShadows=false
        bMiteredCorners=false
        bSeparateFillAlpha=false
        outlineMaterial=null
        outlineSize=0
        outlineColor=<rgba 0 0 0 1>
       }
      }
      highlightColor {colorUseRule=UseColor_ColorTable, specifiedColor=<rgba 0 0 0 0>}
      highlightShape {
       bIsDynamicallyLoaded=false
       drawAs=Image
       imageType=NoImage
       mirroring=NoMirror
       resourceName=None
       resourceObject=null
       tiling=NoTile
       imageSize=<0 0>
       margin=<ltrb 0 0 0 0>
       outlineSettings {
        bUseBrushTransparency=false
        roundingType=HalfHeightRadius
        width=0
        color {colorUseRule=UseColor_Specified, specifiedColor=<rgba 0 0 0 0>}
        cornerRadii=<0 0 0 0>
       }
       tintColor {colorUseRule=UseColor_ColorTable, specifiedColor=<rgba 0 0 0 0>}
       uVRegion {bIsValid=false, max=<0 0>, min=<0 0>}
      }
      selectedBackgroundColor {colorUseRule=UseColor_ColorTable, specifiedColor=<rgba 0 0 0 0>}
      shadowColorAndOpacity=<rgba 0 0 0 1>
      shadowOffset=<0 0>
      strikeBrush {
       bIsDynamicallyLoaded=false
       drawAs=Image
       imageType=NoImage
       mirroring=NoMirror
       resourceName=None
       resourceObject=null
       tiling=NoTile
       imageSize=<32 32>
       margin=<ltrb 0 0 0 0>
       outlineSettings {
        bUseBrushTransparency=false
        roundingType=HalfHeightRadius
        width=0
        color {colorUseRule=UseColor_Specified, specifiedColor=<rgba 0 0 0 0>}
        cornerRadii=<0 0 0 0>
       }
       tintColor {colorUseRule=UseColor_Specified, specifiedColor=<rgba 1 1 1 1>}
       uVRegion {bIsValid=false, max=<0 0>, min=<0 0>}
      }
      underlineBrush {
       bIsDynamicallyLoaded=false
       drawAs=Image
       imageType=NoImage
       mirroring=NoMirror
       resourceName=None
       resourceObject=null
       tiling=NoTile
       imageSize=<32 32>
       margin=<ltrb 0 0 0 0>
       outlineSettings {
        bUseBrushTransparency=false
        roundingType=HalfHeightRadius
        width=0
        color {colorUseRule=UseColor_Specified, specifiedColor=<rgba 0 0 0 0>}
        cornerRadii=<0 0 0 0>
       }
       tintColor {colorUseRule=UseColor_Specified, specifiedColor=<rgba 1 1 1 1>}
       uVRegion {bIsValid=false, max=<0 0>, min=<0 0>}
      }
     }
     vScrollBarPadding=<ltrb 0 0 0 0>
    }
   }
   slot {class=HorizontalBoxSlot, properties {size {sizeRule=Fill, value=1}}}
  }
  EditorUtilityButton_2 {
   widgetClassPath=&EditorUtilityButton
   parent=HorizontalBox_4
   widget=EditorUtilityButton_2
   properties {visibility=Collapsed, backgroundColor=<rgba 0.8153 0.8594 1 1>}
   slot {class=HorizontalBoxSlot}
  }
  ImageSize_2 {
   widgetClassPath=&TextBlock
   parent=EditorUtilityButton_2
   widget=ImageSize_2
   properties {
    text=Rebuild DataBase
    font {
     bForceMonospaced=false
     bMaterialIsStencil=false
     fontMaterial=null
     fontObject=&RobotoTiny
     letterSpacing=0
     monospacedWidth=1
     size=12
     skewAmount=0
     typefaceFontName=Light
     outlineSettings {
      bApplyOutlineToDropShadows=false
      bMiteredCorners=false
      bSeparateFillAlpha=false
      outlineMaterial=null
      outlineSize=0
      outlineColor=<rgba 0 0 0 1>
     }
    }
   }
   slot {class=ButtonSlot, properties {horizontalAlignment=HAlign_Left, padding=<ltrb 0 0 0 0>}}
  }
  RebuildTagButton {
   widgetClassPath=&EditorUtilityButton
   parent=HorizontalBox_4
   widget=RebuildTagButton
   properties {backgroundColor=<rgba 0.8153 0.8594 1 1>}
   slot {class=HorizontalBoxSlot}
  }
  ImageSize_12 {
   widgetClassPath=&TextBlock
   parent=RebuildTagButton
   widget=ImageSize_12
   properties {
    text=Rebuild Tag
    font {
     bForceMonospaced=false
     bMaterialIsStencil=false
     fontMaterial=null
     fontObject=&RobotoTiny
     letterSpacing=0
     monospacedWidth=1
     size=12
     skewAmount=0
     typefaceFontName=Light
     outlineSettings {
      bApplyOutlineToDropShadows=false
      bMiteredCorners=false
      bSeparateFillAlpha=false
      outlineMaterial=null
      outlineSize=0
      outlineColor=<rgba 0 0 0 1>
     }
    }
   }
   slot {class=ButtonSlot, properties {horizontalAlignment=HAlign_Left, padding=<ltrb 0 0 0 0>}}
  }
  EditorUtilityButton_1 {
   widgetClassPath=&EditorUtilityButton
   parent=HorizontalBox_43
   widget=EditorUtilityButton_1
   properties {
    backgroundColor=<rgba 0.8153 0.8594 1 1>
    widgetStyle {
     clickedSlateSound {resourceObject=null}
     disabled {
      bIsDynamicallyLoaded=false
      drawAs=RoundedBox
      imageType=NoImage
      mirroring=NoMirror
      resourceName=None
      resourceObject=null
      tiling=NoTile
      imageSize=<0 0>
      margin=<ltrb 0 0 0 0>
      outlineSettings {
       bUseBrushTransparency=false
       roundingType=FixedRadius
       width=1
       color {colorUseRule=UseColor_ColorTable, specifiedColor=<rgba 0 0 0 0>}
       cornerRadii=<4 4 4 4>
      }
      tintColor {colorUseRule=UseColor_Specified, specifiedColor=<rgba 0.0395 0.0395 0.0395 1>}
      uVRegion {bIsValid=false, max=<0 0>, min=<0 0>}
     }
     disabledForeground {colorUseRule=UseColor_ColorTable, specifiedColor=<rgba 0 0 0 0>}
     hovered {
      bIsDynamicallyLoaded=false
      drawAs=RoundedBox
      imageType=NoImage
      mirroring=NoMirror
      resourceName=None
      resourceObject=null
      tiling=NoTile
      imageSize=<6 6>
      margin=<ltrb 0 0 0 0>
      outlineSettings {
       bUseBrushTransparency=false
       roundingType=FixedRadius
       width=1
       color {colorUseRule=UseColor_ColorTable, specifiedColor=<rgba 0 0 0 0>}
       cornerRadii=<4 4 4 4>
      }
      tintColor {colorUseRule=UseColor_Specified, specifiedColor=<rgba 0.0953 0.0953 0.0953 1>}
      uVRegion {bIsValid=false, max=<0 0>, min=<0 0>}
     }
     hoveredForeground {colorUseRule=UseColor_ColorTable, specifiedColor=<rgba 0 0 0 0>}
     hoveredSlateSound {resourceObject=null}
     normal {
      bIsDynamicallyLoaded=false
      drawAs=RoundedBox
      imageType=NoImage
      mirroring=NoMirror
      resourceName=None
      resourceObject=null
      tiling=NoTile
      imageSize=<6 6>
      margin=<ltrb 0 0 0 0>
      outlineSettings {
       bUseBrushTransparency=false
       roundingType=FixedRadius
       width=1
       color {colorUseRule=UseColor_ColorTable, specifiedColor=<rgba 0 0 0 0>}
       cornerRadii=<4 4 4 4>
      }
      tintColor {colorUseRule=UseColor_Specified, specifiedColor=<rgba 0.0395 0.0395 0.0395 1>}
      uVRegion {bIsValid=false, max=<0 0>, min=<0 0>}
     }
     normalForeground {colorUseRule=UseColor_ColorTable, specifiedColor=<rgba 0 0 0 0>}
     normalPadding=<ltrb 1 1 1 1>
     pressed {
      bIsDynamicallyLoaded=false
      drawAs=RoundedBox
      imageType=NoImage
      mirroring=NoMirror
      resourceName=None
      resourceObject=null
      tiling=NoTile
      imageSize=<6 6>
      margin=<ltrb 0 0 0 0>
      outlineSettings {
       bUseBrushTransparency=false
       roundingType=FixedRadius
       width=1
       color {colorUseRule=UseColor_ColorTable, specifiedColor=<rgba 0 0 0 0>}
       cornerRadii=<4 4 4 4>
      }
      tintColor {colorUseRule=UseColor_Specified, specifiedColor=<rgba 0.0284 0.0284 0.0284 1>}
      uVRegion {bIsValid=false, max=<0 0>, min=<0 0>}
     }
     pressedForeground {colorUseRule=UseColor_ColorTable, specifiedColor=<rgba 0 0 0 0>}
     pressedPadding=<ltrb 1 1 1 1>
     pressedSlateSound {resourceObject=null}
    }
   }
   slot {class=HorizontalBoxSlot, properties {verticalAlignment=VAlign_Center}}
  }
  Fold_1 {
   widgetClassPath=&TextBlock
   parent=EditorUtilityButton_1
   widget=Fold_1
   properties {
    text=">"
    font {
     bForceMonospaced=false
     bMaterialIsStencil=false
     fontMaterial=null
     fontObject=&Roboto
     letterSpacing=0
     monospacedWidth=1
     size=12
     skewAmount=0
     typefaceFontName=Bold
     outlineSettings {
      bApplyOutlineToDropShadows=false
      bMiteredCorners=false
      bSeparateFillAlpha=false
      outlineMaterial=null
      outlineSize=0
      outlineColor=<rgba 0 0 0 1>
     }
    }
   }
   slot {class=ButtonSlot, properties {padding=<ltrb 0 50 0 50>}}
  }
  RightPanel {
   widgetClassPath=&Overlay
   bIsVariable=true
   parent=HorizontalBox_43
   widget=RightPanel
   properties {visibility=Collapsed}
   slot {class=HorizontalBoxSlot, properties {size {sizeRule=Fill, value=0.6}}}
  }
  Border_2 {
   widgetClassPath=&Border
   parent=RightPanel
   widget=Border_2
   properties {brushColor=<rgba 0.0137 0.0144 0.0168 1>}
   slot {class=OverlaySlot, properties {horizontalAlignment=HAlign_Fill, verticalAlignment=VAlign_Fill}}
  }
  EditorUtilityScrollBox_184 {
   widgetClassPath=&EditorUtilityScrollBox
   parent=RightPanel
   widget=EditorUtilityScrollBox_184
   slot {class=OverlaySlot, properties {horizontalAlignment=HAlign_Fill, verticalAlignment=VAlign_Fill}}
  }
  ItemNameBox {
   widgetClassPath=&TextBlock
   bIsVariable=true
   parent=EditorUtilityScrollBox_184
   widget=ItemNameBox
   properties {
    text=ItemName
    font {
     bForceMonospaced=false
     bMaterialIsStencil=false
     fontMaterial=null
     fontObject=&RobotoTiny
     letterSpacing=0
     monospacedWidth=1
     size=24
     skewAmount=0
     typefaceFontName=Light
     outlineSettings {
      bApplyOutlineToDropShadows=false
      bMiteredCorners=false
      bSeparateFillAlpha=false
      outlineMaterial=null
      outlineSize=0
      outlineColor=<rgba 0 0 0 1>
     }
    }
   }
   slot {class=ScrollBoxSlot}
  }
  PublishName {
   widgetClassPath=&TextBlock
   bIsVariable=true
   parent=EditorUtilityScrollBox_184
   widget=PublishName
   properties {
    text=PublishName
    font {
     bForceMonospaced=false
     bMaterialIsStencil=false
     fontMaterial=null
     fontObject=&RobotoTiny
     letterSpacing=0
     monospacedWidth=1
     size=24
     skewAmount=0
     typefaceFontName=Light
     outlineSettings {
      bApplyOutlineToDropShadows=false
      bMiteredCorners=false
      bSeparateFillAlpha=false
      outlineMaterial=null
      outlineSize=0
      outlineColor=<rgba 0 0 0 1>
     }
    }
   }
   slot {class=ScrollBoxSlot}
  }
  SizeBox_3 {
   widgetClassPath=&SizeBox
   bIsVariable=true
   parent=EditorUtilityScrollBox_184
   widget=SizeBox_3
   properties {bOverride_HeightOverride=true, bOverride_MaxAspectRatio=true, bOverride_MinAspectRatio=true, heightOverride=589.9555}
   slot {class=ScrollBoxSlot}
  }
  EditorUtilityButton_0 {
   widgetClassPath=&EditorUtilityButton
   parent=SizeBox_3
   widget=EditorUtilityButton_0
   properties {widgetStyle {
     clickedSlateSound {resourceObject=null}
     disabled {
      bIsDynamicallyLoaded=false
      drawAs=RoundedBox
      imageType=NoImage
      mirroring=NoMirror
      resourceName=None
      resourceObject=null
      tiling=NoTile
      imageSize=<0 0>
      margin=<ltrb 0 0 0 0>
      outlineSettings {
       bUseBrushTransparency=false
       roundingType=FixedRadius
       width=1
       color {colorUseRule=UseColor_ColorTable, specifiedColor=<rgba 0 0 0 0>}
       cornerRadii=<4 4 4 4>
      }
      tintColor {colorUseRule=UseColor_Specified, specifiedColor=<rgba 0.0395 0.0395 0.0395 1>}
      uVRegion {bIsValid=false, max=<0 0>, min=<0 0>}
     }
     disabledForeground {colorUseRule=UseColor_ColorTable, specifiedColor=<rgba 0 0 0 0>}
     hovered {
      bIsDynamicallyLoaded=false
      drawAs=RoundedBox
      imageType=NoImage
      mirroring=NoMirror
      resourceName=None
      resourceObject=null
      tiling=NoTile
      imageSize=<32 32>
      margin=<ltrb 0 0 0 0>
      outlineSettings {
       bUseBrushTransparency=false
       roundingType=FixedRadius
       width=1
       color {colorUseRule=UseColor_ColorTable, specifiedColor=<rgba 0 0 0 0>}
       cornerRadii=<4 4 4 4>
      }
      tintColor {colorUseRule=UseColor_Specified, specifiedColor=<rgba 0.0953 0.0953 0.0953 1>}
      uVRegion {bIsValid=false, max=<0 0>, min=<0 0>}
     }
     hoveredForeground {colorUseRule=UseColor_ColorTable, specifiedColor=<rgba 0 0 0 0>}
     hoveredSlateSound {resourceObject=null}
     normal {
      bIsDynamicallyLoaded=false
      drawAs=RoundedBox
      imageType=NoImage
      mirroring=NoMirror
      resourceName=None
      resourceObject=null
      tiling=NoTile
      imageSize=<32 32>
      margin=<ltrb 0 0 0 0>
      outlineSettings {
       bUseBrushTransparency=false
       roundingType=FixedRadius
       width=1
       color {colorUseRule=UseColor_ColorTable, specifiedColor=<rgba 0 0 0 0>}
       cornerRadii=<4 4 4 4>
      }
      tintColor {colorUseRule=UseColor_Specified, specifiedColor=<rgba 0.0395 0.0395 0.0395 1>}
      uVRegion {bIsValid=false, max=<0 0>, min=<0 0>}
     }
     normalForeground {colorUseRule=UseColor_ColorTable, specifiedColor=<rgba 0 0 0 0>}
     normalPadding=<ltrb 0 0 0 0>
     pressed {
      bIsDynamicallyLoaded=false
      drawAs=RoundedBox
      imageType=NoImage
      mirroring=NoMirror
      resourceName=None
      resourceObject=null
      tiling=NoTile
      imageSize=<32 32>
      margin=<ltrb 0 0 0 0>
      outlineSettings {
       bUseBrushTransparency=false
       roundingType=FixedRadius
       width=1
       color {colorUseRule=UseColor_ColorTable, specifiedColor=<rgba 0 0 0 0>}
       cornerRadii=<4 4 4 4>
      }
      tintColor {colorUseRule=UseColor_Specified, specifiedColor=<rgba 0.0284 0.0284 0.0284 1>}
      uVRegion {bIsValid=false, max=<0 0>, min=<0 0>}
     }
     pressedForeground {colorUseRule=UseColor_ColorTable, specifiedColor=<rgba 0 0 0 0>}
     pressedPadding=<ltrb 0 0 0 0>
     pressedSlateSound {resourceObject=null}
    }}
   slot {class=SizeBoxSlot, properties {horizontalAlignment=HAlign_Center, verticalAlignment=VAlign_Center}}
  }
  Overlay_0 {
   widgetClassPath=&Overlay
   parent=EditorUtilityButton_0
   widget=Overlay_0
   slot {class=ButtonSlot, properties {padding=<ltrb 0 0 0 0>}}
  }
  VideoImage {
   widgetClassPath=&Image
   bIsVariable=true
   parent=Overlay_0
   widget=VideoImage
   properties {brush {
     bIsDynamicallyLoaded=false
     drawAs=Image
     imageType=FullColor
     mirroring=NoMirror
     resourceName=None
     resourceObject=&FXRefLibMedia_Video_Mat
     tiling=NoTile
     imageSize=<600 600>
     margin=<ltrb 0 0 0 0>
     outlineSettings {
      bUseBrushTransparency=false
      roundingType=HalfHeightRadius
      width=0
      color {colorUseRule=UseColor_Specified, specifiedColor=<rgba 0 0 0 0>}
      cornerRadii=<0 0 0 0>
     }
     tintColor {colorUseRule=UseColor_Specified, specifiedColor=<rgba 1 1 1 1>}
     uVRegion {bIsValid=false, max=<0 0>, min=<0 0>}
    }}
   slot {class=OverlaySlot}
  }
  ItemNameBox_4 {
   widgetClassPath=&TextBlock
   bIsVariable=true
   parent=EditorUtilityScrollBox_184
   widget=ItemNameBox_4
   properties {
    text=Click video to replay
    colorAndOpacity {colorUseRule=UseColor_Specified, specifiedColor=<rgba 0.574 0.574 0.574 1>}
    font {
     bForceMonospaced=false
     bMaterialIsStencil=false
     fontMaterial=null
     fontObject=&RobotoTiny
     letterSpacing=0
     monospacedWidth=1
     size=9.75
     skewAmount=0
     typefaceFontName=Light
     outlineSettings {
      bApplyOutlineToDropShadows=false
      bMiteredCorners=false
      bSeparateFillAlpha=false
      outlineMaterial=null
      outlineSize=0
      outlineColor=<rgba 0 0 0 1>
     }
    }
   }
   slot {class=ScrollBoxSlot, properties {horizontalAlignment=HAlign_Center, verticalAlignment=VAlign_Center}}
  }
  VerticalBox_1 {widgetClassPath=&VerticalBox, parent=EditorUtilityScrollBox_184, widget=VerticalBox_1, slot {class=ScrollBoxSlot}}
  HorizontalBox_1 {widgetClassPath=&HorizontalBox, parent=VerticalBox_1, widget=HorizontalBox_1, slot {class=VerticalBoxSlot}}
  ItemNameBox_2 {
   widgetClassPath=&TextBlock
   bIsVariable=true
   parent=HorizontalBox_1
   widget=ItemNameBox_2
   properties {
    text=MainColor
    font {
     bForceMonospaced=false
     bMaterialIsStencil=false
     fontMaterial=null
     fontObject=&RobotoTiny
     letterSpacing=0
     monospacedWidth=1
     size=12
     skewAmount=0
     typefaceFontName=Light
     outlineSettings {
      bApplyOutlineToDropShadows=false
      bMiteredCorners=false
      bSeparateFillAlpha=false
      outlineMaterial=null
      outlineSize=0
      outlineColor=<rgba 0 0 0 1>
     }
    }
   }
   slot {class=HorizontalBoxSlot}
  }
  HorizontalBox_174 {
   widgetClassPath=&HorizontalBox
   parent=HorizontalBox_1
   widget=HorizontalBox_174
   slot {class=HorizontalBoxSlot, properties {size {sizeRule=Fill, value=1}}}
  }
  1stColor {
   widgetClassPath=&Border
   bIsVariable=true
   parent=HorizontalBox_174
   widget=1stColor
   properties {visibility=Collapsed}
   slot {class=HorizontalBoxSlot, properties {padding=<ltrb 5 5 5 5>, size {sizeRule=Fill, value=1}}}
  }
  2ndColor {
   widgetClassPath=&Border
   bIsVariable=true
   parent=HorizontalBox_174
   widget=2ndColor
   slot {class=HorizontalBoxSlot, properties {padding=<ltrb 5 5 5 5>, size {sizeRule=Fill, value=1}}}
   properties=~
  }
  3rdColor {
   widgetClassPath=&Border
   bIsVariable=true
   parent=HorizontalBox_174
   widget=3rdColor
   slot {class=HorizontalBoxSlot, properties {padding=<ltrb 5 5 5 5>, size {sizeRule=Fill, value=1}}}
   properties=~
  }
  4thColor {
   widgetClassPath=&Border
   bIsVariable=true
   parent=HorizontalBox_174
   widget=4thColor
   slot {class=HorizontalBoxSlot, properties {padding=<ltrb 5 5 5 5>, size {sizeRule=Fill, value=1}}}
   properties=~
  }
  ItemNameBox_3 {
   widgetClassPath=&TextBlock
   bIsVariable=true
   parent=VerticalBox_1
   widget=ItemNameBox_3
   properties {
    text=AddTags
    visibility=Collapsed
    font {
     bForceMonospaced=false
     bMaterialIsStencil=false
     fontMaterial=null
     fontObject=&RobotoTiny
     letterSpacing=0
     monospacedWidth=1
     size=12
     skewAmount=0
     typefaceFontName=Light
     outlineSettings {
      bApplyOutlineToDropShadows=false
      bMiteredCorners=false
      bSeparateFillAlpha=false
      outlineMaterial=null
      outlineSize=0
      outlineColor=<rgba 0 0 0 1>
     }
    }
   }
   slot {class=VerticalBoxSlot, properties {padding=<ltrb 0 20 0 0>}}
  }
  HorizontalBox_177 {
   widgetClassPath=&HorizontalBox
   parent=VerticalBox_1
   widget=HorizontalBox_177
   properties {visibility=Collapsed}
   slot {class=VerticalBoxSlot, properties {padding=<ltrb 0 5 0 0>}}
  }
  EditorUtilityEditableTextBox_413 {
   widgetClassPath=&EditorUtilityEditableTextBox
   bIsVariable=true
   parent=HorizontalBox_177
   widget=EditorUtilityEditableTextBox_413
   slot {class=HorizontalBoxSlot, properties {size {sizeRule=Fill, value=1}}}
  }
  EditorUtilityButton_485 {
   widgetClassPath=&EditorUtilityButton
   parent=HorizontalBox_177
   widget=EditorUtilityButton_485
   properties {backgroundColor=<rgba 0.8153 0.8594 1 1>}
   slot {class=HorizontalBoxSlot, properties {size {sizeRule=Fill, value=0.3}}}
  }
  ItemNameBox_1 {
   widgetClassPath=&TextBlock
   bIsVariable=true
   parent=EditorUtilityButton_485
   widget=ItemNameBox_1
   properties {
    text=AddTag
    font {
     bForceMonospaced=false
     bMaterialIsStencil=false
     fontMaterial=null
     fontObject=&RobotoTiny
     letterSpacing=0
     monospacedWidth=1
     size=12
     skewAmount=0
     typefaceFontName=Light
     outlineSettings {
      bApplyOutlineToDropShadows=false
      bMiteredCorners=false
      bSeparateFillAlpha=false
      outlineMaterial=null
      outlineSize=0
      outlineColor=<rgba 0 0 0 1>
     }
    }
   }
   slot {class=ButtonSlot, properties {horizontalAlignment=HAlign_Left, padding=<ltrb 0 0 0 0>}}
  }
  SizeBox_2 {
   widgetClassPath=&SizeBox
   parent=VerticalBox_1
   widget=SizeBox_2
   properties {bOverride_MaxDesiredHeight=true, heightOverride=200, maxDesiredHeight=400, visibility=Collapsed}
   slot {class=VerticalBoxSlot}
  }
  RightTagButton {
   widgetClassPath=&TileView
   bIsVariable=true
   parent=SizeBox_2
   widget=RightTagButton
   properties {
    entryHeight=25
    entryWidgetClass=&/EUW_TagButtonW.EUW_TagButtonW_C
    entryWidth=110
    numDesignerPreviewEntries=20
    scrollBarStyle {
     thickness=8
     draggedThumbImage {
      bIsDynamicallyLoaded=false
      drawAs=RoundedBox
      imageType=NoImage
      mirroring=NoMirror
      resourceName=None
      resourceObject=null
      tiling=NoTile
      imageSize=<0 0>
      margin=<ltrb 0 0 0 0>
      outlineSettings {
       bUseBrushTransparency=false
       roundingType=FixedRadius
       width=0
       color {colorUseRule=UseColor_Specified, specifiedColor=<rgba 0 0 0 0>}
       cornerRadii=<4 4 4 4>
      }
      tintColor {colorUseRule=UseColor_Specified, specifiedColor=<rgba 0.2159 0.2159 0.2159 1>}
      uVRegion {bIsValid=false, max=<0 0>, min=<0 0>}
     }
     horizontalBackgroundImage {
      bIsDynamicallyLoaded=false
      drawAs=NoDrawType
      imageType=NoImage
      mirroring=NoMirror
      resourceName=None
      resourceObject=null
      tiling=NoTile
      imageSize=<0 0>
      margin=<ltrb 0 0 0 0>
      outlineSettings {
       bUseBrushTransparency=false
       roundingType=HalfHeightRadius
       width=0
       color {colorUseRule=UseColor_Specified, specifiedColor=<rgba 0 0 0 0>}
       cornerRadii=<0 0 0 0>
      }
      tintColor {colorUseRule=UseColor_Specified, specifiedColor=<rgba 1 1 1 1>}
      uVRegion {bIsValid=false, max=<0 0>, min=<0 0>}
     }
     horizontalBottomSlotImage {
      bIsDynamicallyLoaded=false
      drawAs=NoDrawType
      imageType=NoImage
      mirroring=NoMirror
      resourceName=None
      resourceObject=null
      tiling=NoTile
      imageSize=<0 0>
      margin=<ltrb 0 0 0 0>
      outlineSettings {
       bUseBrushTransparency=false
       roundingType=HalfHeightRadius
       width=0
       color {colorUseRule=UseColor_Specified, specifiedColor=<rgba 0 0 0 0>}
       cornerRadii=<0 0 0 0>
      }
      tintColor {colorUseRule=UseColor_Specified, specifiedColor=<rgba 1 1 1 1>}
      uVRegion {bIsValid=false, max=<0 0>, min=<0 0>}
     }
     horizontalTopSlotImage {
      bIsDynamicallyLoaded=false
      drawAs=NoDrawType
      imageType=NoImage
      mirroring=NoMirror
      resourceName=None
      resourceObject=null
      tiling=NoTile
      imageSize=<0 0>
      margin=<ltrb 0 0 0 0>
      outlineSettings {
       bUseBrushTransparency=false
       roundingType=HalfHeightRadius
       width=0
       color {colorUseRule=UseColor_Specified, specifiedColor=<rgba 0 0 0 0>}
       cornerRadii=<0 0 0 0>
      }
      tintColor {colorUseRule=UseColor_Specified, specifiedColor=<rgba 1 1 1 1>}
      uVRegion {bIsValid=false, max=<0 0>, min=<0 0>}
     }
     hoveredThumbImage {
      bIsDynamicallyLoaded=false
      drawAs=RoundedBox
      imageType=NoImage
      mirroring=NoMirror
      resourceName=None
      resourceObject=null
      tiling=NoTile
      imageSize=<0 0>
      margin=<ltrb 0 0 0 0>
      outlineSettings {
       bUseBrushTransparency=false
       roundingType=FixedRadius
       width=0
       color {colorUseRule=UseColor_Specified, specifiedColor=<rgba 0 0 0 0>}
       cornerRadii=<4 4 4 4>
      }
      tintColor {colorUseRule=UseColor_Specified, specifiedColor=<rgba 0.2159 0.2159 0.2159 1>}
      uVRegion {bIsValid=false, max=<0 0>, min=<0 0>}
     }
     normalThumbImage {
      bIsDynamicallyLoaded=false
      drawAs=RoundedBox
      imageType=NoImage
      mirroring=NoMirror
      resourceName=None
      resourceObject=null
      tiling=NoTile
      imageSize=<0 0>
      margin=<ltrb 0 0 0 0>
      outlineSettings {
       bUseBrushTransparency=false
       roundingType=FixedRadius
       width=0
       color {colorUseRule=UseColor_Specified, specifiedColor=<rgba 0 0 0 0>}
       cornerRadii=<4 4 4 4>
      }
      tintColor {colorUseRule=UseColor_Specified, specifiedColor=<rgba 0.0953 0.0953 0.0953 1>}
      uVRegion {bIsValid=false, max=<0 0>, min=<0 0>}
     }
     verticalBackgroundImage {
      bIsDynamicallyLoaded=false
      drawAs=NoDrawType
      imageType=NoImage
      mirroring=NoMirror
      resourceName=None
      resourceObject=null
      tiling=NoTile
      imageSize=<0 0>
      margin=<ltrb 0 0 0 0>
      outlineSettings {
       bUseBrushTransparency=false
       roundingType=HalfHeightRadius
       width=0
       color {colorUseRule=UseColor_Specified, specifiedColor=<rgba 0 0 0 0>}
       cornerRadii=<0 0 0 0>
      }
      tintColor {colorUseRule=UseColor_Specified, specifiedColor=<rgba 1 1 1 1>}
      uVRegion {bIsValid=false, max=<0 0>, min=<0 0>}
     }
     verticalBottomSlotImage {
      bIsDynamicallyLoaded=false
      drawAs=NoDrawType
      imageType=NoImage
      mirroring=NoMirror
      resourceName=None
      resourceObject=null
      tiling=NoTile
      imageSize=<0 0>
      margin=<ltrb 0 0 0 0>
      outlineSettings {
       bUseBrushTransparency=false
       roundingType=HalfHeightRadius
       width=0
       color {colorUseRule=UseColor_Specified, specifiedColor=<rgba 0 0 0 0>}
       cornerRadii=<0 0 0 0>
      }
      tintColor {colorUseRule=UseColor_Specified, specifiedColor=<rgba 1 1 1 1>}
      uVRegion {bIsValid=false, max=<0 0>, min=<0 0>}
     }
     verticalTopSlotImage {
      bIsDynamicallyLoaded=false
      drawAs=NoDrawType
      imageType=NoImage
      mirroring=NoMirror
      resourceName=None
      resourceObject=null
      tiling=NoTile
      imageSize=<0 0>
      margin=<ltrb 0 0 0 0>
      outlineSettings {
       bUseBrushTransparency=false
       roundingType=HalfHeightRadius
       width=0
       color {colorUseRule=UseColor_Specified, specifiedColor=<rgba 0 0 0 0>}
       cornerRadii=<0 0 0 0>
      }
      tintColor {colorUseRule=UseColor_Specified, specifiedColor=<rgba 1 1 1 1>}
      uVRegion {bIsValid=false, max=<0 0>, min=<0 0>}
     }
    }
   }
   slot {class=SizeBoxSlot, properties {padding=<ltrb 0 5 0 0>}}
  }
  EditorUtilityButton_203 {
   widgetClassPath=&EditorUtilityButton
   parent=VerticalBox_1
   widget=EditorUtilityButton_203
   properties {backgroundColor=<rgba 0.8153 0.8594 1 1>}
   slot {class=VerticalBoxSlot, properties {padding=<ltrb 0 20 0 0>}}
  }
  RelatedAssets {
   widgetClassPath=&TextBlock
   bIsVariable=true
   parent=EditorUtilityButton_203
   widget=RelatedAssets
   properties {
    text=Related Assets
    colorAndOpacity {colorUseRule=UseColor_Specified, specifiedColor=<rgba 1 0 0.1604 1>}
    font {
     bForceMonospaced=false
     bMaterialIsStencil=false
     fontMaterial=null
     fontObject=null
     letterSpacing=0
     monospacedWidth=1
     size=15
     skewAmount=0
     typefaceFontName=76 Bold
     outlineSettings {
      bApplyOutlineToDropShadows=false
      bMiteredCorners=false
      bSeparateFillAlpha=false
      outlineMaterial=null
      outlineSize=0
      outlineColor=<rgba 0 0 0 1>
     }
    }
   }
   slot {class=ButtonSlot, properties {horizontalAlignment=HAlign_Fill, verticalAlignment=VAlign_Fill, padding=<ltrb 0 0 0 0>}}
  }
  DetailsView_172 {
   widgetClassPath=&DetailsView
   parent=VerticalBox_1
   widget=DetailsView_172
   properties {
    bShowAnimatedPropertiesOption=false
    bShowKeyablePropertiesOption=false
    visibility=Collapsed
    categoriesToShow [Show]
    propertiesToShow [RelatedAssetsList]
   }
   slot {class=VerticalBoxSlot}
  }
  TutorialCanvas {
   widgetClassPath=&CanvasPanel
   bIsVariable=true
   parent=Overlay_183
   widget=TutorialCanvas
   properties {visibility=HitTestInvisible}
   slot {class=OverlaySlot, properties {horizontalAlignment=HAlign_Fill, verticalAlignment=VAlign_Fill}}
  }
  Tutorial {
   widgetClassPath=&Border
   bIsVariable=true
   parent=TutorialCanvas
   widget=Tutorial
   properties {brushColor=<rgba 0.0137 0.0144 0.0168 1>, padding=<ltrb 0 0 0 0>}
   slot {class=CanvasPanelSlot, properties {layoutData {alignment=<0 0>, anchors {maximum=<1 1>, minimum=<0 0>}, offsets=<ltrb 0 0 0 23.2232>}}}
  }
  CanvasPanel_156 {
   widgetClassPath=&CanvasPanel
   parent=Tutorial
   widget=CanvasPanel_156
   slot {class=BorderSlot, properties {padding=<ltrb 0 0 0 0>}}
  }
  Image_0 {
   widgetClassPath=&Image
   bIsVariable=true
   parent=CanvasPanel_156
   widget=Image_0
   properties {
    visibility=Hidden
    brush {
     bIsDynamicallyLoaded=false
     drawAs=Image
     imageType=NoImage
     mirroring=NoMirror
     resourceName=None
     resourceObject=&/FXLibTutorial
     tiling=NoTile
     imageSize=<32 32>
     margin=<ltrb 0 0 0 0>
     outlineSettings {
      bUseBrushTransparency=false
      roundingType=HalfHeightRadius
      width=0
      color {colorUseRule=UseColor_Specified, specifiedColor=<rgba 0 0 0 0>}
      cornerRadii=<0 0 0 0>
     }
     tintColor {colorUseRule=UseColor_Specified, specifiedColor=<rgba 1 1 1 1>}
     uVRegion {bIsValid=false, max=<0 0>, min=<0 0>}
    }
   }
   slot {class=CanvasPanelSlot, properties {layoutData {alignment=<0 0>, anchors {maximum=<1 1>, minimum=<0 0>}, offsets=<ltrb 0 0 0 0>}}}
  }
  TextBlock_138 {
   widgetClassPath=&TextBlock
   bIsVariable=true
   parent=CanvasPanel_156
   widget=TextBlock_138
   properties {justification=Right, text="↓↓↓Here is the library path input block", visibility=Collapsed}
   slot {class=CanvasPanelSlot, properties {layoutData {alignment=<0 0>, anchors {maximum=<1 1>, minimum=<0 1>}, offsets=<ltrb 0 -50 0 30>}}}
  }
  TextBlock_11 {
   widgetClassPath=&TextBlock
   bIsVariable=true
   parent=CanvasPanel_156
   widget=TextBlock_11
   properties {
    justification=Center
    text="Setup Steps:\r\n\r\n1.Enable \"Fortnite:FXBrowser Library\" UGS sync filter and sync to newest.(Ignore this step if you already did before)\r\n\r\n2.Click Build\r\n"
    font {
     bForceMonospaced=false
     bMaterialIsStencil=false
     fontMaterial=null
     fontObject=&Roboto
     letterSpacing=0
     monospacedWidth=1
     size=15
     skewAmount=0
     typefaceFontName=Light
     outlineSettings {
      bApplyOutlineToDropShadows=false
      bMiteredCorners=false
      bSeparateFillAlpha=false
      outlineMaterial=null
      outlineSize=0
      outlineColor=<rgba 0 0 0 1>
     }
    }
   }
   slot {class=CanvasPanelSlot, properties {layoutData {alignment=<0 0>, anchors {maximum=<0.5 0.5>, minimum=<0.5 0.5>}, offsets=<ltrb 0 0 0 0>}}}
  }
  EditorUtilityButton_408 {
   widgetClassPath=&EditorUtilityButton
   parent=CanvasPanel_156
   widget=EditorUtilityButton_408
   slot {
    class=CanvasPanelSlot
    properties {bAutoSize=true, layoutData {alignment=<0.5 0>, anchors {maximum=<0.5 0.5>, minimum=<0.5 0.5>}, offsets=<ltrb 0 130 100 30>}}
   }
  }
  TextBlock_12 {
   widgetClassPath=&TextBlock
   parent=EditorUtilityButton_408
   widget=TextBlock_12
   properties {
    justification=Center
    text=Build
    font {
     bForceMonospaced=false
     bMaterialIsStencil=false
     fontMaterial=null
     fontObject=&Roboto
     letterSpacing=0
     monospacedWidth=1
     size=15
     skewAmount=0
     typefaceFontName=Light
     outlineSettings {
      bApplyOutlineToDropShadows=false
      bMiteredCorners=false
      bSeparateFillAlpha=false
      outlineMaterial=null
      outlineSize=0
      outlineColor=<rgba 0 0 0 1>
     }
    }
   }
   slot {class=ButtonSlot}
  }
  EditorUtilityButton_11 {
   widgetClassPath=&EditorUtilityButton
   parent=CanvasPanel_156
   widget=EditorUtilityButton_11
   properties {visibility=Collapsed}
   slot {
    class=CanvasPanelSlot
    properties {bAutoSize=true, layoutData {alignment=<0.5 0>, anchors {maximum=<0.5 0.5>, minimum=<0.5 0.5>}, offsets=<ltrb 0 170 100 30>}}
   }
  }
  TextBlock_13 {
   widgetClassPath=&TextBlock
   parent=EditorUtilityButton_11
   widget=TextBlock_13
   properties {
    justification=Center
    text=Check the instructions
    font {
     bForceMonospaced=false
     bMaterialIsStencil=false
     fontMaterial=null
     fontObject=&Roboto
     letterSpacing=0
     monospacedWidth=1
     size=15
     skewAmount=0
     typefaceFontName=Light
     outlineSettings {
      bApplyOutlineToDropShadows=false
      bMiteredCorners=false
      bSeparateFillAlpha=false
      outlineMaterial=null
      outlineSize=0
      outlineColor=<rgba 0 0 0 1>
     }
    }
   }
   slot {class=ButtonSlot}
  }
  EditorUtilityButton_12 {
   widgetClassPath=&EditorUtilityButton
   parent=CanvasPanel_156
   widget=EditorUtilityButton_12
   properties {visibility=Collapsed}
   slot {
    class=CanvasPanelSlot
    properties {bAutoSize=true, layoutData {alignment=<1 0>, anchors {maximum=<1 0>, minimum=<1 0>}, offsets=<ltrb 0 0 100 30>}}
   }
  }
  TextBlock_14 {
   widgetClassPath=&TextBlock
   parent=EditorUtilityButton_12
   widget=TextBlock_14
   properties {
    justification=Right
    text=Back to first page
    font {
     bForceMonospaced=false
     bMaterialIsStencil=false
     fontMaterial=null
     fontObject=&Roboto
     letterSpacing=0
     monospacedWidth=1
     size=15
     skewAmount=0
     typefaceFontName=Light
     outlineSettings {
      bApplyOutlineToDropShadows=false
      bMiteredCorners=false
      bSeparateFillAlpha=false
      outlineMaterial=null
      outlineSize=0
      outlineColor=<rgba 0 0 0 1>
     }
    }
   }
   slot {class=ButtonSlot}
  }
  WaitBuildCanvas {
   widgetClassPath=&CanvasPanel
   bIsVariable=true
   parent=Overlay_183
   widget=WaitBuildCanvas
   properties {visibility=Collapsed}
   slot {class=OverlaySlot, properties {horizontalAlignment=HAlign_Fill, verticalAlignment=VAlign_Fill}}
  }
  Border_240 {
   widgetClassPath=&Border
   parent=WaitBuildCanvas
   widget=Border_240
   properties {horizontalAlignment=HAlign_Center, verticalAlignment=VAlign_Center, brushColor=<rgba 0.0137 0.0144 0.0168 1>, padding=<ltrb 0 0 0 0>}
   slot {class=CanvasPanelSlot, properties {layoutData {alignment=<0 0>, anchors {maximum=<1 1>, minimum=<0 0>}, offsets=<ltrb 0 0 0 0>}}}
  }
  TextBlock_10 {
   widgetClassPath=&TextBlock
   parent=Border_240
   widget=TextBlock_10
   properties {justification=Center, text="Please wait the database updating, do not close the editor.\r\nThis action might cost 1-5 mins"}
   slot {class=BorderSlot, properties {horizontalAlignment=HAlign_Center, verticalAlignment=VAlign_Center, padding=<ltrb 0 0 0 0>}}
  }
 ]
}
